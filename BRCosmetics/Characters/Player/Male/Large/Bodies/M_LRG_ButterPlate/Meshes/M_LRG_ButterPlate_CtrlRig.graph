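# Control Rig for the large male ButterPlate player body, adding procedural corrective motion to secondary bones and optional finger retargeting.
# Spherical pose readers on limb bones weight transform offsets for hips, knees, ankles, finger bases and shoulder armor.
# FE, FE_Idle, is_crouch and is_fly modulate hand correctives, while weighted rotation constraints make toe bones follow the balls.
# Map_Fingers functions transfer weighted rotation deltas from initial pose between three-bone source and target chains.
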ControlRigBlueprint /BRCosmetics/Characters/Player/Male/Large/Bodies/M_LRG_ButterPlate/Meshes/M_LRG_ButterPlate_CtrlRig
graphs [
 New Function_ContainedGraph {nodes [
   n1 = RigVMFunctionEntryNode()
   > RigVMFunction_Sequence = RigVMAggregateNode()
   n2 = RigVMDispatchNode(Index=0, Array=@n1.Source) {template="DISPATCH_RigVMDispatch_ArrayGetAtIndex(in Array,in Index,out Element)"}
   n3 = &RigUnit_GetTransform(Space=LocalSpace, bInitial=true, Item=@n2.Element)
   n4 = /Script/RigVM.RigVMFunction_MathQuaternionInverse(Value=/Script/RigVM.RigVMFunction_MathQuaternionMul(
     A=&RigUnit_GetTransform(Space=LocalSpace, bInitial=False, Item=@n2.Element).Transform.Rotation
     B=&RigVMFunction_MathQuaternionInverse(Value=@n3.Transform.Rotation) {template="Inverse::Execute(in Value,out Result)"}.Result
    ) {is=struct RigVMUnitNode, template="Multiply::Execute(in A,in B,out Result)"}.Result) {is=struct RigVMUnitNode, template="Inverse::Execute(in Value,out Result)"}
   &RigUnit_ModifyTransforms(
    ItemToModify=((Item=(Type=Bone,Name="None"),Transform=(Rotation=(X=0,Y=0,Z=0,W=1),Translation=(X=0,Y=0,Z=0),Scale3D=(X=1,Y=1,Z=1))))
    Weight=0.3
    WeightMinimum=0
    ItemToModify.0.Item=RigVMDispatchNode(Index=0, Array=@n1.Target) {template="DISPATCH_RigVMDispatch_ArrayGetAtIndex(in Array,in Index,out Element)"}.Element
    ItemToModify.0.Transform.Rotation=@n4.Result
    ExecutePin=@RigVMFunction_Sequence.A
   )
   RigVMFunctionReturnNode(ExecuteContext=@RigVMFunction_Sequence.D)
   n5 = RigVMDispatchNode(Index=2, Array=@n1.Source) {template="DISPATCH_RigVMDispatch_ArrayGetAtIndex(in Array,in Index,out Element)"}
   n6 = &RigVMFunction_MathQuaternionMul(
    A=&RigUnit_GetTransform(Space=LocalSpace, bInitial=False, Item=@n5.Element).Transform.Rotation
    B=&RigVMFunction_MathQuaternionInverse(Value=&RigUnit_GetTransform(Space=LocalSpace, bInitial=true, Item=@n5.Element).Transform.Rotation) {template="Inverse::Execute(in Value,out Result)"}.Result
   ) {template="Multiply::Execute(in A,in B,out Result)"}
   &RigUnit_ModifyTransforms(
    ItemToModify=((Item=(Type=Bone,Name="None"),Transform=(Rotation=(X=0,Y=0,Z=0,W=1),Translation=(X=0,Y=0,Z=0),Scale3D=(X=1,Y=1,Z=1))))
    Weight=0.5
    WeightMinimum=0
    ItemToModify.0.Item=RigVMDispatchNode(Index=2, Array=@n1.Target) {template="DISPATCH_RigVMDispatch_ArrayGetAtIndex(in Array,in Index,out Element)"}.Element
    ItemToModify.0.Transform.Rotation=&RigVMFunction_MathQuaternionInverse(Value=&RigVMFunction_MathQuaternionSwingTwist(TwistAxis=(X=0,Y=0,Z=1), Input=@n6.Result).Twist) {template="Inverse::Execute(in Value,out Result)"}.Result
    ExecutePin=@RigVMFunction_Sequence.C
   )
   n7 = RigVMDispatchNode(Index=1, Array=@n1.Source) {template="DISPATCH_RigVMDispatch_ArrayGetAtIndex(in Array,in Index,out Element)"}
   n8 = &RigUnit_GetTransform(Space=LocalSpace, bInitial=true, Item=@n7.Element)
   n9 = &RigVMFunction_MathQuaternionMul(
    B=@n8.Transform.Rotation
    A=&RigVMFunction_MathQuaternionInverse(Value=&RigUnit_GetTransform(Space=LocalSpace, bInitial=false, Item=@n7.Element).Transform.Rotation) {template="Inverse::Execute(in Value,out Result)"}.Result
   ) {template="Multiply::Execute(in A,in B,out Result)"}
   RigVMFunction_MathQuaternionToEuler = &RigVMFunction_MathQuaternionToEuler(
    RotationOrder=ZYX
    Value=&RigVMFunction_MathQuaternionInverse(Value=@n9.Result) {template="Inverse::Execute(in Value,out Result)"}.Result
   )
   Greater = /Script/RigVM.RigVMFunction_MathDoubleGreater(B=0, A=@RigVMFunction_MathQuaternionToEuler.Result.Z) {is=struct RigVMUnitNode, template="Greater::Execute(in A,in B,out Result)"}
   n10 = &RigVMFunction_ControlFlowBranch(Condition=@Greater.Result, ExecuteContext=@RigVMFunction_Sequence.B)
   Scale = /Script/RigVM.RigVMFunction_MathQuaternionScaleV2(Factor=1, Value=@n9.Result) {is=struct RigVMUnitNode, template="Scale::Execute(in Value,in Factor,out Result)"}
   n11 = RigVMDispatchNode(Index=1, Array=@n1.Target) {template="DISPATCH_RigVMDispatch_ArrayGetAtIndex(in Array,in Index,out Element)"}
   &RigUnit_ModifyTransforms(
    ItemToModify=((Item=(Type=Bone,Name="None"),Transform=(Rotation=(X=0,Y=0,Z=0,W=1),Translation=(X=0,Y=0,Z=0),Scale3D=(X=1,Y=1,Z=1))))
    Weight=1
    WeightMinimum=0
    ExecutePin=@n10.False
    ItemToModify.0.Transform.Rotation=@Scale.Result
    ItemToModify.0.Item=@n11.Element
   )
   &RigUnit_ModifyTransforms(
    ItemToModify=((Item=(Type=Bone,Name="None"),Transform=(Rotation=(X=0,Y=0,Z=0,W=1),Translation=(X=0,Y=0,Z=0),Scale3D=(X=1,Y=1,Z=1))))
    Weight=1
    WeightMinimum=0
    ExecutePin=@n10.True
    ItemToModify.0.Transform.Rotation=&RigVMFunction_MathQuaternionScaleV2(Factor=1.1, Value=@n9.Result) {template="Scale::Execute(in Value,in Factor,out Result)"}.Result
    ItemToModify.0.Item=@n11.Element
   )
   Interpolate = /Script/RigVM.RigVMFunction_MathQuaternionSlerp(B=(X=0,Y=0,Z=0,W=1), T=0, A=@n8.Transform.Rotation) {is=struct RigVMUnitNode, template="Interpolate::Execute(in A,in B,in T,out Result)"}
   &RigVMFunction_MathQuaternionInverse(Value=@n8.Transform.Rotation) {template="Inverse::Execute(in Value,out Result)"}
   &RigVMFunction_MathQuaternionScaleV2(Factor=1.05, Value=@n4.Result) {template="Scale::Execute(in Value,in Factor,out Result)"}
   &RigVMFunction_MathQuaternionScaleV2(Factor=0.875, Value=(X=0,Y=0,Z=0,W=1)) {template="Scale::Execute(in Value,in Factor,out Result)"}
  ]}
 New Function_ContainedGraph {nodes [
   n1 = RigVMFunctionEntryNode()
   n2 = &RigVMFunction_Sequence(ExecuteContext=@n1.ExecuteContext)
   RigVMFunction_Sequence = RigVMAggregateNode(ExecuteContext=@n2.A)
   n3 = RigVMDispatchNode(Index=0, Array=@n1.Source) {template="DISPATCH_RigVMDispatch_ArrayGetAtIndex(in Array,in Index,out Element)"}
   n4 = &RigUnit_GetTransform(Space=LocalSpace, bInitial=true, Item=@n3.Element)
   &RigUnit_ModifyTransforms(
    ItemToModify=((Item=(Type=Bone,Name="None"),Transform=(Rotation=(X=0,Y=0,Z=0,W=1),Translation=(X=0,Y=0,Z=0),Scale3D=(X=1,Y=1,Z=1))))
    Weight=0.3
    WeightMinimum=0
    ItemToModify.0.Item=RigVMDispatchNode(Index=0, Array=@n1.Target) {template="DISPATCH_RigVMDispatch_ArrayGetAtIndex(in Array,in Index,out Element)"}.Element
    ItemToModify.0.Transform.Rotation=&RigVMFunction_MathQuaternionMul(
     A=&RigUnit_GetTransform(Space=LocalSpace, bInitial=False, Item=@n3.Element).Transform.Rotation
     B=&RigVMFunction_MathQuaternionInverse(Value=@n4.Transform.Rotation) {template="Inverse::Execute(in Value,out Result)"}.Result
    ) {template="Multiply::Execute(in A,in B,out Result)"}.Result
    ExecutePin=@RigVMFunction_Sequence.A
   )
   RigVMFunctionReturnNode(ExecuteContext=@RigVMFunction_Sequence.D)
   n5 = RigVMDispatchNode(Index=2, Array=@n1.Source) {template="DISPATCH_RigVMDispatch_ArrayGetAtIndex(in Array,in Index,out Element)"}
   n6 = &RigVMFunction_MathQuaternionMul(
    A=&RigUnit_GetTransform(Space=LocalSpace, bInitial=False, Item=@n5.Element).Transform.Rotation
    B=&RigVMFunction_MathQuaternionInverse(Value=&RigUnit_GetTransform(Space=LocalSpace, bInitial=true, Item=@n5.Element).Transform.Rotation) {template="Inverse::Execute(in Value,out Result)"}.Result
   ) {template="Multiply::Execute(in A,in B,out Result)"}
   &RigUnit_ModifyTransforms(
    ItemToModify=((Item=(Type=Bone,Name="None"),Transform=(Rotation=(X=0,Y=0,Z=0,W=1),Translation=(X=0,Y=0,Z=0),Scale3D=(X=1,Y=1,Z=1))))
    Weight=0.5
    WeightMinimum=0
    ItemToModify.0.Transform.Rotation=@n6.Result
    ExecutePin=@RigVMFunction_Sequence.C
    ItemToModify.0.Item=RigVMDispatchNode(Index=2, Array=@n1.Target) {template="DISPATCH_RigVMDispatch_ArrayGetAtIndex(in Array,in Index,out Element)"}.Element
   )
   n7 = RigVMDispatchNode(Index=1, Array=@n1.Source) {template="DISPATCH_RigVMDispatch_ArrayGetAtIndex(in Array,in Index,out Element)"}
   n8 = &RigUnit_GetTransform(Space=LocalSpace, bInitial=false, Item=@n7.Element)
   n9 = &RigVMFunction_MathQuaternionMul(
    A=@n8.Transform.Rotation
    B=&RigVMFunction_MathQuaternionInverse(Value=&RigUnit_GetTransform(Space=LocalSpace, bInitial=true, Item=@n7.Element).Transform.Rotation) {template="Inverse::Execute(in Value,out Result)"}.Result
   ) {template="Multiply::Execute(in A,in B,out Result)"}
   RigVMFunction_MathQuaternionToEuler = &RigVMFunction_MathQuaternionToEuler(
    RotationOrder=ZYX
    Value=&RigVMFunction_MathQuaternionInverse(Value=@n9.Result) {template="Inverse::Execute(in Value,out Result)"}.Result
   )
   Greater = &RigVMFunction_MathDoubleGreater(B=0, A=@RigVMFunction_MathQuaternionToEuler.Result.Z) {template="Greater::Execute(in A,in B,out Result)"}
   n10 = &RigVMFunction_ControlFlowBranch(Condition=@Greater.Result, ExecuteContext=@RigVMFunction_Sequence.B)
   Scale = &RigVMFunction_MathQuaternionScaleV2(Factor=1, Value=@n9.Result) {template="Scale::Execute(in Value,in Factor,out Result)"}
   n11 = RigVMDispatchNode(Index=1, Array=@n1.Target) {template="DISPATCH_RigVMDispatch_ArrayGetAtIndex(in Array,in Index,out Element)"}
   &RigUnit_ModifyTransforms(
    ItemToModify=((Item=(Type=Bone,Name="None"),Transform=(Rotation=(X=0,Y=0,Z=0,W=1),Translation=(X=0,Y=0,Z=0),Scale3D=(X=1,Y=1,Z=1))))
    Weight=1
    WeightMinimum=0
    ExecutePin=@n10.False
    ItemToModify.0.Transform.Rotation=@Scale.Result
    ItemToModify.0.Item=@n11.Element
   )
   &RigUnit_ModifyTransforms(
    ItemToModify=((Item=(Type=Bone,Name="None"),Transform=(Rotation=(X=0,Y=0,Z=0,W=1),Translation=(X=0,Y=0,Z=0),Scale3D=(X=1,Y=1,Z=1))))
    Weight=1
    WeightMinimum=0
    ExecutePin=@n10.True
    ItemToModify.0.Transform.Rotation=&RigVMFunction_MathQuaternionScaleV2(Factor=1.05, Value=@n9.Result) {template="Scale::Execute(in Value,in Factor,out Result)"}.Result
    ItemToModify.0.Item=@n11.Element
   )
   &RigVMFunction_MathQuaternionInverse(Value=@n8.Transform.Rotation) {template="Inverse::Execute(in Value,out Result)"}
   &RigVMFunction_MathQuaternionInverse(Value=&RigVMFunction_MathQuaternionSwingTwist(TwistAxis=(X=0,Y=0,Z=1), Input=@n6.Result).Twist) {template="Inverse::Execute(in Value,out Result)"}
   &RigVMFunction_MathQuaternionScaleV2(Factor=0.875, Value=(X=0,Y=0,Z=0,W=1)) {template="Scale::Execute(in Value,in Factor,out Result)"}
  ]}
 RigVMFunctionLibrary {nodes [
   Map_Fingers_Down = RigVMCollapseNode(ExecuteContext=(), Inverse=False, Source=(), Target=())
   Map_Fingers_Up = RigVMCollapseNode(ExecuteContext=(), Inverse=False, Source=(), Target=())
  ]}
 RigVMFunction_Sequence_1_1_1_1_ContainedGraph {nodes [
   n1 = RigVMFunctionEntryNode()
   n2 = &RigVMFunction_Sequence(ExecuteContext=@n1.ExecuteContext)
   n3 = &RigVMFunction_Sequence(ExecuteContext=@n2.B)
   n4 = &RigVMFunction_Sequence(ExecuteContext=@n3.B)
   RigVMFunctionReturnNode(A=@n2.A, B=@n3.A, C=@n4.A, D=@n4.B)
  ]}
 RigVMFunction_Sequence_1_1_1_1_ContainedGraph {nodes [
   n1 = RigVMFunctionEntryNode()
   n2 = &RigVMFunction_Sequence(ExecuteContext=@n1.ExecuteContext)
   n3 = &RigVMFunction_Sequence(ExecuteContext=@n2.B)
   n4 = &RigVMFunction_Sequence(ExecuteContext=@n3.B)
   RigVMFunctionReturnNode(A=@n2.A, B=@n3.A, C=@n4.A, D=@n4.B)
  ]}
 RigVMFunction_Sequence_1_1_2_1_2_1_ContainedGraph {nodes [
   n1 = RigVMFunctionEntryNode()
   n2 = &RigVMFunction_Sequence(ExecuteContext=@n1.ExecuteContext)
   n3 = &RigVMFunction_Sequence(ExecuteContext=@n2.B)
   n4 = &RigVMFunction_Sequence(ExecuteContext=@n3.B)
   n5 = &RigVMFunction_Sequence(ExecuteContext=@n4.B)
   n6 = &RigVMFunction_Sequence(ExecuteContext=@n5.B)
   RigVMFunctionReturnNode(A=@n2.A, B=@n3.A, C=@n4.A, D=@n5.A, E=@n6.A, F=@n6.B)
  ]}
 RigVMFunction_Sequence_1_2_1_ContainedGraph {nodes [
   n1 = RigVMFunctionEntryNode()
   n2 = &RigVMFunction_Sequence(ExecuteContext=@n1.ExecuteContext)
   n3 = &RigVMFunction_Sequence(ExecuteContext=@n2.B)
   n4 = &RigVMFunction_Sequence(ExecuteContext=@n3.B)
   n5 = &RigVMFunction_Sequence(ExecuteContext=@n4.B)
   RigVMFunctionReturnNode(A=@n2.A, B=@n3.A, C=@n4.A, D=@n5.A, E=@n5.B)
  ]}
 RigVMFunction_Sequence_1_3_ContainedGraph {nodes [
   n1 = RigVMFunctionEntryNode()
   n2 = &RigVMFunction_Sequence(ExecuteContext=@n1.ExecuteContext)
   n3 = &RigVMFunction_Sequence(ExecuteContext=@n2.B)
   n4 = &RigVMFunction_Sequence(ExecuteContext=@n3.B)
   RigVMFunctionReturnNode(A=@n2.A, B=@n3.A, C=@n4.A, D=@n4.B)
  ]}
 RigVMFunction_Sequence_1_ContainedGraph {nodes [
   n1 = RigVMFunctionEntryNode()
   RigVMFunction_Sequence = &RigVMFunction_Sequence(ExecuteContext=@n1.ExecuteContext)
   n2 = &RigVMFunction_Sequence(ExecuteContext=@RigVMFunction_Sequence.B)
   n3 = &RigVMFunction_Sequence(ExecuteContext=@n2.B)
   RigVMFunctionReturnNode(A=@RigVMFunction_Sequence.A, B=@n2.A, C=@n3.A, D=@n3.B)
  ]}
 RigVMFunction_Sequence_1_ContainedGraph {nodes [
   n1 = RigVMFunctionEntryNode()
   RigVMFunction_Sequence = &RigVMFunction_Sequence(ExecuteContext=@n1.ExecuteContext)
   n2 = &RigVMFunction_Sequence(ExecuteContext=@RigVMFunction_Sequence.B)
   n3 = &RigVMFunction_Sequence(ExecuteContext=@n2.B)
   RigVMFunctionReturnNode(A=@RigVMFunction_Sequence.A, B=@n2.A, C=@n3.A, D=@n3.B)
  ]}
 RigVMFunction_Sequence_2_1_2_ContainedGraph {nodes [
   n1 = RigVMFunctionEntryNode()
   n2 = &RigVMFunction_Sequence(ExecuteContext=@n1.ExecuteContext)
   n3 = &RigVMFunction_Sequence(ExecuteContext=@n2.B)
   RigVMFunctionReturnNode(A=@n2.A, B=@n3.A, C=@n3.B)
  ]}
 RigVMFunction_Sequence_2_2_ContainedGraph {nodes [
   n1 = RigVMFunctionEntryNode()
   n2 = &RigVMFunction_Sequence(ExecuteContext=@n1.ExecuteContext)
   n3 = &RigVMFunction_Sequence(ExecuteContext=@n2.B)
   RigVMFunctionReturnNode(A=@n2.A, B=@n3.A, C=@n3.B)
  ]}
 RigVMFunction_Sequence_3_1_ContainedGraph {nodes [
   n1 = RigVMFunctionEntryNode()
   n2 = &RigVMFunction_Sequence(ExecuteContext=@n1.ExecuteContext)
   n3 = &RigVMFunction_Sequence(ExecuteContext=@n2.B)
   n4 = &RigVMFunction_Sequence(ExecuteContext=@n3.B)
   RigVMFunctionReturnNode(A=@n2.A, B=@n3.A, C=@n4.A, D=@n4.B)
  ]}
 RigVMFunction_Sequence_5_1_ContainedGraph {nodes [
   n1 = RigVMFunctionEntryNode()
   n2 = &RigVMFunction_Sequence(ExecuteContext=@n1.ExecuteContext)
   n3 = &RigVMFunction_Sequence(ExecuteContext=@n2.B)
   RigVMFunctionReturnNode(A=@n2.A, B=@n3.A, C=@n3.B)
  ]}
 RigVMModel {nodes [
   n1 = RigVMRerouteNode(Value=())
   n2 = &RigVMFunction_Sequence(ExecuteContext=@n1.Value)
   n3 = &RigVMFunction_ControlFlowBranch(ExecuteContext=@n2.B, Condition=RigVMVariableNode() {variable=Enable Finger Retarget}.Value)
   n4 = RigVMDispatchNode(ExecuteContext=@n3.True, Array=RigVMVariableNode() {variable=Fingers}.Value) {template="DISPATCH_RigVMDispatch_ArrayIterator(in Array,out Element,out Index,out Count,out Ratio)"}
   RigUnit_BeginExecution = &RigUnit_BeginExecution()
   RigVMFunction_Sequence_1_2 = RigVMAggregateNode(ExecuteContext=@RigUnit_BeginExecution.ExecutePin)
   RigVMFunction_Sequence_1 = RigVMAggregateNode(ExecuteContext=@RigVMFunction_Sequence_1_2.A)
   n5 = &RigVMFunction_Sequence(ExecuteContext=@RigVMFunction_Sequence_1.A)
   n6 = &RigUnit_SphericalPoseReader(
    ActiveRegionScaleFactors=(PositiveWidth=0.5,NegativeWidth=0.5,PositiveHeight=0.5,NegativeHeight=0.5)
    ActiveRegionSize=0.1
    Debug=(bDrawDebug=false,bDraw2D=False,bDrawLocalAxes=False,DebugScale=25,DebugSegments=20,DebugThickness=0.25)
    DriverAxis=(X=-1,Y=0,Z=0)
    DriverItem=(Type=Bone,Name="thigh_l")
    FalloffRegionScaleFactors=(PositiveWidth=1,NegativeWidth=1,PositiveHeight=1,NegativeHeight=1)
    FalloffSize=0.3
    FlipHeightScaling=False
    FlipWidthScaling=False
    OptionalParentItem=(Type=Bone,Name="None")
    RotationOffset=(X=115,Y=0,Z=10)
    ExecutePin=@n5.A
   )
   > &RigUnit_ModifyTransforms(
    ItemToModify=((Item=(Type=Bone,Name="dyn_hip_l"),Transform=(Rotation=(X=0,Y=0,Z=0,W=1),Translation=(X=3,Y=0,Z=-4),Scale3D=(X=1,Y=1,Z=1))))
    ItemToModify.0.Transform.Rotation=&RigVMFunction_MathQuaternionFromEuler(Euler=(X=10,Y=17.5,Z=0), RotationOrder=ZXY).Result
    Weight=@n6.OutputParam
   )
   n7 = &RigUnit_SphericalPoseReader(
    ActiveRegionScaleFactors=(PositiveWidth=0.5,NegativeWidth=1.5,PositiveHeight=0.5,NegativeHeight=0.5)
    ActiveRegionSize=0.1
    Debug=(bDrawDebug=false,bDraw2D=False,bDrawLocalAxes=False,DebugScale=25,DebugSegments=20,DebugThickness=0.25)
    DriverAxis=(X=1,Y=0,Z=0)
    DriverItem=(Type=Bone,Name="thigh_r")
    FalloffRegionScaleFactors=(PositiveWidth=1,NegativeWidth=1,PositiveHeight=1,NegativeHeight=1)
    FalloffSize=0.25
    FlipHeightScaling=False
    FlipWidthScaling=False
    OptionalParentItem=(Type=Bone,Name="None")
    RotationOffset=(X=-80,Y=0,Z=0)
    ExecutePin=@n5.B
   )
   > &RigUnit_ModifyTransforms(
    ItemToModify=((Item=(Type=Bone,Name="dyn_hip_r"),Transform=(Rotation=(X=0,Y=0,Z=0,W=1),Translation=(X=-3,Y=0,Z=4),Scale3D=(X=1,Y=1,Z=1))))
    ItemToModify.0.Transform.Rotation=&RigVMFunction_MathQuaternionFromEuler(Euler=(X=-10,Y=17.5,Z=0), RotationOrder=ZXY).Result
    Weight=@n7.OutputParam
   )
   RigVMFunction_Sequence = &RigVMFunction_Sequence(ExecuteContext=@RigVMFunction_Sequence_1.B)
   n8 = &RigUnit_SphericalPoseReader(
    ActiveRegionScaleFactors=(PositiveWidth=0.5,NegativeWidth=0.5,PositiveHeight=0.5,NegativeHeight=0.5)
    ActiveRegionSize=0.1
    Debug=(bDrawDebug=false,bDraw2D=False,bDrawLocalAxes=False,DebugScale=25,DebugSegments=20,DebugThickness=0.25)
    DriverAxis=(X=-1,Y=0,Z=0)
    DriverItem=(Type=Bone,Name="calf_l")
    FalloffRegionScaleFactors=(PositiveWidth=1,NegativeWidth=1,PositiveHeight=1,NegativeHeight=1)
    FalloffSize=0.5
    FlipHeightScaling=False
    FlipWidthScaling=False
    OptionalParentItem=(Type=Bone,Name="None")
    RotationOffset=(X=-90,Y=0,Z=25)
    ExecutePin=@RigVMFunction_Sequence.A
   )
   RigVMFunction_MathQuaternionFromEuler = &RigVMFunction_MathQuaternionFromEuler(Euler=(X=0,Y=0,Z=60), RotationOrder=ZXY)
   > &RigUnit_ModifyTransforms(
    ItemToModify=((Item=(Type=Bone,Name="dyn_knee_l"),Transform=(Rotation=(X=0,Y=0,Z=0,W=1),Translation=(X=3,Y=0,Z=0),Scale3D=(X=1,Y=1,Z=1))))
    ItemToModify.0.Transform.Rotation=@RigVMFunction_MathQuaternionFromEuler.Result
    Weight=@n8.OutputParam
   )
   n9 = &RigUnit_SphericalPoseReader(
    ActiveRegionScaleFactors=(PositiveWidth=0.5,NegativeWidth=1.5,PositiveHeight=0.5,NegativeHeight=0.5)
    ActiveRegionSize=0.1
    Debug=(bDrawDebug=false,bDraw2D=False,bDrawLocalAxes=False,DebugScale=25,DebugSegments=20,DebugThickness=0.25)
    DriverAxis=(X=1,Y=0,Z=0)
    DriverItem=(Type=Bone,Name="calf_r")
    FalloffRegionScaleFactors=(PositiveWidth=1,NegativeWidth=1,PositiveHeight=1,NegativeHeight=1)
    FalloffSize=0.5
    FlipHeightScaling=False
    FlipWidthScaling=False
    OptionalParentItem=(Type=Bone,Name="None")
    RotationOffset=(X=90,Y=0,Z=25)
    ExecutePin=@RigVMFunction_Sequence.B
   )
   > &RigUnit_ModifyTransforms(
    ItemToModify=((Item=(Type=Bone,Name="dyn_knee_r"),Transform=(Rotation=(X=0,Y=0,Z=0,W=1),Translation=(X=-3,Y=0,Z=0),Scale3D=(X=1,Y=1,Z=1))))
    ItemToModify.0.Transform.Rotation=&RigVMFunction_MathQuaternionFromEuler(Euler=(X=0,Y=0,Z=60), RotationOrder=ZXY).Result
    Weight=@n9.OutputParam
   )
   n10 = &RigVMFunction_Sequence(ExecuteContext=@RigVMFunction_Sequence_1.C)
   n11 = &RigUnit_SphericalPoseReader(
    ActiveRegionScaleFactors=(PositiveWidth=0.5,NegativeWidth=0.5,PositiveHeight=0.5,NegativeHeight=0.5)
    ActiveRegionSize=0.1
    Debug=(bDrawDebug=false,bDraw2D=False,bDrawLocalAxes=False,DebugScale=25,DebugSegments=20,DebugThickness=0.25)
    DriverAxis=(X=-1,Y=0,Z=0)
    DriverItem=(Type=Bone,Name="foot_l")
    FalloffRegionScaleFactors=(PositiveWidth=1,NegativeWidth=1,PositiveHeight=1,NegativeHeight=1)
    FalloffSize=0.3
    FlipHeightScaling=False
    FlipWidthScaling=False
    OptionalParentItem=(Type=Bone,Name="None")
    RotationOffset=(X=80,Y=0,Z=45)
    ExecutePin=@n10.A
   )
   > &RigUnit_ModifyTransforms(
    ItemToModify=((Item=(Type=Bone,Name="calf_twist_01_l"),Transform=(Rotation=(X=0,Y=0,Z=0,W=1),Translation=(X=1.5,Y=0,Z=0),Scale3D=(X=1,Y=1,Z=1))))
    ItemToModify.0.Transform.Rotation=&RigVMFunction_MathQuaternionFromEuler(Euler=(X=-12.5,Y=-5,Z=-15), RotationOrder=ZXY).Result
    Weight=@n11.OutputParam
   )
   > &RigUnit_ModifyTransforms(
    ItemToModify=((Item=(Type=Bone,Name="foot_l"),Transform=(Rotation=(X=0,Y=0,Z=0,W=1),Translation=(X=1,Y=0,Z=0),Scale3D=(X=1,Y=1,Z=1))))
    Weight=@n11.OutputParam
   )
   n12 = &RigUnit_SphericalPoseReader(
    ActiveRegionScaleFactors=(PositiveWidth=0.5,NegativeWidth=0.5,PositiveHeight=0.5,NegativeHeight=0.5)
    ActiveRegionSize=0.1
    Debug=(bDrawDebug=false,bDraw2D=False,bDrawLocalAxes=False,DebugScale=25,DebugSegments=20,DebugThickness=0.25)
    DriverAxis=(X=1,Y=0,Z=0)
    DriverItem=(Type=Bone,Name="foot_r")
    FalloffRegionScaleFactors=(PositiveWidth=1,NegativeWidth=1,PositiveHeight=1,NegativeHeight=1)
    FalloffSize=0.3
    FlipHeightScaling=False
    FlipWidthScaling=False
    OptionalParentItem=(Type=Bone,Name="None")
    RotationOffset=(X=-100,Y=0,Z=45)
    ExecutePin=@n10.B
   )
   > &RigUnit_ModifyTransforms(
    ItemToModify=((Item=(Type=Bone,Name="calf_twist_01_r"),Transform=(Rotation=(X=0,Y=0,Z=0,W=1),Translation=(X=-1.5,Y=0,Z=0),Scale3D=(X=1,Y=1,Z=1))))
    ItemToModify.0.Transform.Rotation=&RigVMFunction_MathQuaternionFromEuler(Euler=(X=12.5,Y=-5,Z=-15), RotationOrder=ZXY).Result
    Weight=@n12.OutputParam
   )
   n13 = &RigVMFunction_Sequence(ExecuteContext=@RigVMFunction_Sequence_1.D)
   &RigUnit_RotationConstraintLocalSpaceOffset(
    AdvancedSettings=(InterpolationType=Average,RotationOrderForFilter=XZY)
    Child=(Type=Bone,Name="dyn_ball_l_1")
    Filter=(bX=True,bY=True,bZ=True)
    Parents=((Item=(Type=Bone,Name="ball_l"),Weight=1))
    Weight=0.75
    ExecutePin=@n13.A
   )
   > &RigUnit_RotationConstraintLocalSpaceOffset(
    AdvancedSettings=(InterpolationType=Average,RotationOrderForFilter=XZY)
    Child=(Type=Bone,Name="dyn_ball_l_2")
    Filter=(bX=True,bY=True,bZ=True)
    Parents=((Item=(Type=Bone,Name="ball_l"),Weight=1))
    Weight=0.75
   )
   > &RigUnit_RotationConstraintLocalSpaceOffset(
    AdvancedSettings=(InterpolationType=Average,RotationOrderForFilter=XZY)
    Child=(Type=Bone,Name="dyn_ball_l_3")
    Filter=(bX=True,bY=True,bZ=True)
    Parents=((Item=(Type=Bone,Name="ball_l"),Weight=1))
    Weight=0.45
   )
   > &RigUnit_RotationConstraintLocalSpaceOffset(
    AdvancedSettings=(InterpolationType=Average,RotationOrderForFilter=XZY)
    Child=(Type=Bone,Name="dyn_ball_l_4")
    Filter=(bX=True,bY=True,bZ=True)
    Parents=((Item=(Type=Bone,Name="ball_l"),Weight=1))
    Weight=0.25
   )
   &RigUnit_RotationConstraintLocalSpaceOffset(
    AdvancedSettings=(InterpolationType=Average,RotationOrderForFilter=XZY)
    Child=(Type=Bone,Name="dyn_ball_r_1")
    Filter=(bX=True,bY=True,bZ=True)
    Parents=((Item=(Type=Bone,Name="ball_r"),Weight=1))
    Weight=0.75
    ExecutePin=@n13.B
   )
   > &RigUnit_RotationConstraintLocalSpaceOffset(
    AdvancedSettings=(InterpolationType=Average,RotationOrderForFilter=XZY)
    Child=(Type=Bone,Name="dyn_ball_r_2")
    Filter=(bX=True,bY=True,bZ=True)
    Parents=((Item=(Type=Bone,Name="ball_r"),Weight=1))
    Weight=0.75
   )
   > &RigUnit_RotationConstraintLocalSpaceOffset(
    AdvancedSettings=(InterpolationType=Average,RotationOrderForFilter=XZY)
    Child=(Type=Bone,Name="dyn_ball_r_3")
    Filter=(bX=True,bY=True,bZ=True)
    Parents=((Item=(Type=Bone,Name="ball_r"),Weight=1))
    Weight=0.45
   )
   > &RigUnit_RotationConstraintLocalSpaceOffset(
    AdvancedSettings=(InterpolationType=Average,RotationOrderForFilter=XZY)
    Child=(Type=Bone,Name="dyn_ball_r_4")
    Filter=(bX=True,bY=True,bZ=True)
    Parents=((Item=(Type=Bone,Name="ball_r"),Weight=1))
    Weight=0.25
   )
   RigVMFunction_Sequence_5 = RigVMAggregateNode(ExecuteContext=@RigVMFunction_Sequence_1_2.B)
   n14 = &RigVMFunction_Sequence(ExecuteContext=@RigVMFunction_Sequence_5.A)
   n15 = &RigUnit_SphericalPoseReader(
    ActiveRegionScaleFactors=(PositiveWidth=0.5,NegativeWidth=0.5,PositiveHeight=0.5,NegativeHeight=0.5)
    ActiveRegionSize=0.1
    Debug=(bDrawDebug=false,bDraw2D=False,bDrawLocalAxes=False,DebugScale=25,DebugSegments=20,DebugThickness=0.25)
    DriverAxis=(X=0,Y=1,Z=0)
    DriverItem=(Type=Bone,Name="hand_r")
    FalloffRegionScaleFactors=(PositiveWidth=1,NegativeWidth=1,PositiveHeight=1,NegativeHeight=1)
    FalloffSize=0.3
    FlipHeightScaling=False
    FlipWidthScaling=False
    OptionalParentItem=(Type=Bone,Name="None")
    RotationOffset=(X=20,Y=-10,Z=0)
    ExecutePin=@n14.A
   )
   n16 = RigVMVariableNode() {variable=FE}
   > &RigUnit_ModifyTransforms(
    ItemToModify=((Item=(Type=Bone,Name="dyn_finger_base_r"),Transform=(Rotation=(X=0,Y=0,Z=0,W=1),Translation=(X=0,Y=0,Z=0),Scale3D=(X=1,Y=1,Z=1))))
    Weight=&RigVMFunction_MathDoubleMul(B=@n15.OutputParam, A=&RigVMFunction_MathDoubleSub(A=1, B=@n16.Value).Result).Result
    ItemToModify.0.Transform.Rotation=&RigVMFunction_MathQuaternionFromEuler(Euler=(X=10,Y=0,Z=-10), RotationOrder=ZXY).Result
   )
   n17 = &RigUnit_SphericalPoseReader(
    ActiveRegionScaleFactors=(PositiveWidth=0.5,NegativeWidth=0.5,PositiveHeight=0.5,NegativeHeight=0.5)
    ActiveRegionSize=0.1
    Debug=(bDrawDebug=false,bDraw2D=False,bDrawLocalAxes=False,DebugScale=25,DebugSegments=20,DebugThickness=0.25)
    DriverAxis=(X=0,Y=1,Z=0)
    DriverItem=(Type=Bone,Name="hand_r")
    FalloffRegionScaleFactors=(PositiveWidth=1,NegativeWidth=1,PositiveHeight=1,NegativeHeight=1)
    FalloffSize=0.3
    FlipHeightScaling=False
    FlipWidthScaling=False
    OptionalParentItem=(Type=Bone,Name="None")
    RotationOffset=(X=60,Y=45,Z=0)
    ExecutePin=@n14.B
   )
   n18 = RigVMVariableNode() {variable=FE}
   > &RigUnit_ModifyTransforms(
    ItemToModify=((Item=(Type=Bone,Name="dyn_finger_base_r"),Transform=(Rotation=(X=0,Y=0,Z=0,W=1),Translation=(X=-1,Y=0,Z=0),Scale3D=(X=1,Y=1,Z=1))))
    Weight=&RigVMFunction_MathDoubleMul(B=@n17.OutputParam, A=&RigVMFunction_MathDoubleSub(A=1, B=@n18.Value).Result).Result
    ItemToModify.0.Transform.Rotation=&RigVMFunction_MathQuaternionFromEuler(Euler=(X=0,Y=-12.5,Z=-45), RotationOrder=ZXY).Result
   )
   RigVMFunction_Sequence_1_1_2_1_2 = RigVMAggregateNode(ExecuteContext=@RigVMFunction_Sequence_5.B)
   &RigUnit_ModifyTransforms(
    ItemToModify=((Item=(Type=Bone,Name="dyn_finger_base_r"),Transform=(Rotation=(X=0.0004,Y=-0.061,Z=-0.2215,W=0.9733),Translation=(X=0,Y=0,Z=0),Scale3D=(X=1,Y=1,Z=1))))
    WeightMinimum=0
    ExecutePin=@RigVMFunction_Sequence_1_1_2_1_2.A
    Weight=&RigVMFunction_MathDoubleSub(A=1, B=RigVMVariableNode() {variable=FE_Idle}.Value).Result
   )
   > n19 = &RigUnit_SphericalPoseReader(
    ActiveRegionScaleFactors=(PositiveWidth=0.5,NegativeWidth=0.5,PositiveHeight=0.5,NegativeHeight=0.5)
    ActiveRegionSize=0.1
    Debug=(bDrawDebug=false,bDraw2D=False,bDrawLocalAxes=False,DebugScale=25,DebugSegments=20,DebugThickness=0.25)
    DriverAxis=(X=0,Y=1,Z=0)
    DriverItem=(Type=Bone,Name="hand_l")
    FalloffRegionScaleFactors=(PositiveWidth=1,NegativeWidth=1,PositiveHeight=1,NegativeHeight=1)
    FalloffSize=0.2
    FlipHeightScaling=False
    FlipWidthScaling=False
    OptionalParentItem=(Type=Bone,Name="None")
    RotationOffset=(X=125,Y=0,Z=30)
   )
   n20 = &RigVMFunction_Sequence(ExecuteContext=@n19.ExecutePin)
   n21 = RigVMVariableNode() {variable=FE}
   n22 = &RigVMFunction_AlphaInterp(
    InterpSpeedDecreasing=100
    InterpSpeedIncreasing=100
    Scale=1
    bClampResult=False
    bInterpResult=False
    Value=&RigVMFunction_MathDoubleMul(B=@n19.OutputParam, A=&RigVMFunction_MathDoubleSub(A=1, B=@n21.Value).Result).Result
   )
   &RigUnit_ModifyTransforms(
    ItemToModify=((Item=(Type=Bone,Name="dyn_finger_base_l"),Transform=(Rotation=(X=0,Y=0,Z=0,W=1),Translation=(X=1,Y=0,Z=0),Scale3D=(X=1,Y=1,Z=1))))
    Weight=@n22.Result
    ItemToModify.0.Transform.Rotation=&RigVMFunction_MathQuaternionFromEuler(Euler=(X=0,Y=-10,Z=-60), RotationOrder=ZXY).Result
    ExecutePin=@n20.A
   )
   &RigUnit_ModifyTransforms(
    ItemToModify=((Item=(Type=Bone,Name="dyn_finger_base_l"),Transform=(Rotation=(X=0,Y=0,Z=0,W=1),Translation=(X=1.5,Y=0,Z=0),Scale3D=(X=1,Y=1,Z=1))))
    Weight=&RigVMFunction_MathDoubleMul(A=@n22.Result, B=RigVMVariableNode() {variable=is_crouch}.Value).Result
    ItemToModify.0.Transform.Rotation=&RigVMFunction_MathQuaternionFromEuler(Euler=(X=-45,Y=0,Z=60), RotationOrder=ZXY).Result
    ExecutePin=@n20.B
   )
   n23 = &RigUnit_SphericalPoseReader(
    ActiveRegionScaleFactors=(PositiveWidth=0.25,NegativeWidth=0.25,PositiveHeight=0.25,NegativeHeight=1)
    ActiveRegionSize=0.1
    Debug=(bDrawDebug=false,bDraw2D=False,bDrawLocalAxes=False,DebugScale=25,DebugSegments=20,DebugThickness=0.25)
    DriverAxis=(X=0,Y=1,Z=0)
    DriverItem=(Type=Bone,Name="hand_l")
    FalloffRegionScaleFactors=(PositiveWidth=1,NegativeWidth=1,PositiveHeight=1,NegativeHeight=1)
    FalloffSize=0.35
    FlipHeightScaling=False
    FlipWidthScaling=False
    OptionalParentItem=(Type=Bone,Name="None")
    RotationOffset=(X=0,Y=30,Z=30)
    ExecutePin=@RigVMFunction_Sequence_1_1_2_1_2.B
   )
   n24 = &RigVMFunction_Sequence(ExecuteContext=@n23.ExecutePin)
   n25 = RigVMVariableNode() {variable=FE}
   n26 = &RigVMFunction_AlphaInterp(
    InterpSpeedDecreasing=100
    InterpSpeedIncreasing=100
    Scale=1
    bClampResult=False
    bInterpResult=False
    Value=&RigVMFunction_MathDoubleMul(B=@n23.OutputParam, A=&RigVMFunction_MathDoubleSub(A=1, B=@n25.Value).Result).Result
   )
   &RigUnit_ModifyTransforms(
    ItemToModify=((Item=(Type=Bone,Name="dyn_finger_base_l"),Transform=(Rotation=(X=0,Y=0,Z=0,W=1),Translation=(X=1,Y=0,Z=0),Scale3D=(X=1,Y=1,Z=1))))
    Weight=@n26.Result
    ItemToModify.0.Transform.Rotation=&RigVMFunction_MathQuaternionFromEuler(Euler=(X=5,Y=-30,Z=-75), RotationOrder=ZXY).Result
    ExecutePin=@n24.A
   )
   Subtract = &RigVMFunction_MathDoubleSub(A=1, B=@n26.Result)
   &RigUnit_ModifyTransforms(
    ItemToModify=((Item=(Type=Bone,Name="dyn_finger_base_l"),Transform=(Rotation=(X=0,Y=0,Z=0,W=1),Translation=(X=1,Y=0,Z=0),Scale3D=(X=1,Y=1,Z=1))))
    Weight=&RigVMFunction_MathDoubleMul(A=@Subtract.Result, B=RigVMVariableNode() {variable=is_crouch}.Value).Result
    ItemToModify.0.Transform.Rotation=&RigVMFunction_MathQuaternionFromEuler(Euler=(X=20,Y=-30,Z=-75), RotationOrder=ZXY).Result
    ExecutePin=@n24.B
   )
   n27 = &RigUnit_SphericalPoseReader(
    ActiveRegionScaleFactors=(PositiveWidth=0.25,NegativeWidth=0.5,PositiveHeight=0.5,NegativeHeight=0.5)
    ActiveRegionSize=0.1
    Debug=(bDrawDebug=false,bDraw2D=False,bDrawLocalAxes=False,DebugScale=25,DebugSegments=20,DebugThickness=0.25)
    DriverAxis=(X=1,Y=0,Z=0)
    DriverItem=(Type=Bone,Name="hand_l")
    FalloffRegionScaleFactors=(PositiveWidth=1,NegativeWidth=1,PositiveHeight=1,NegativeHeight=1)
    FalloffSize=0.2
    FlipHeightScaling=False
    FlipWidthScaling=False
    OptionalParentItem=(Type=Bone,Name="None")
    RotationOffset=(X=10,Y=-25,Z=0)
    ExecutePin=@RigVMFunction_Sequence_1_1_2_1_2.D
   )
   n28 = &RigVMFunction_Sequence(ExecuteContext=@n27.ExecutePin)
   n29 = RigVMVariableNode() {variable=FE}
   AlphaInterp = &RigVMFunction_AlphaInterp(
    InterpSpeedDecreasing=100
    InterpSpeedIncreasing=100
    Scale=1
    bClampResult=False
    bInterpResult=False
    Value=&RigVMFunction_MathDoubleMul(B=@n27.OutputParam, A=&RigVMFunction_MathDoubleSub(A=1, B=@n29.Value).Result).Result
   )
   &RigUnit_ModifyTransforms(
    ItemToModify=((Item=(Type=Bone,Name="dyn_finger_base_l"),Transform=(Rotation=(X=0,Y=0,Z=0,W=1),Translation=(X=0,Y=0,Z=0),Scale3D=(X=1,Y=1,Z=1))))
    Weight=@AlphaInterp.Result
    ItemToModify.0.Transform.Rotation=&RigVMFunction_MathQuaternionFromEuler(Euler=(X=10,Y=20,Z=-10), RotationOrder=ZXY).Result
    ExecutePin=@n28.A
   )
   n30 = &RigVMFunction_Sequence(ExecuteContext=@n28.B)
   n31 = &RigVMFunction_MathDoubleMul(A=@AlphaInterp.Result, B=RigVMVariableNode() {variable=is_crouch}.Value)
   &RigUnit_ModifyTransforms(
    ItemToModify=((Item=(Type=Bone,Name="dyn_finger_base_l"),Transform=(Rotation=(X=0,Y=0,Z=0,W=1),Translation=(X=0,Y=2,Z=0),Scale3D=(X=1,Y=1,Z=1))))
    Weight=@n31.Result
    ItemToModify.0.Transform.Rotation=&RigVMFunction_MathQuaternionFromEuler(Euler=(X=20,Y=10,Z=0), RotationOrder=ZXY).Result
    ExecutePin=@n30.A
   )
   n32 = &RigUnit_SphericalPoseReader(
    ActiveRegionScaleFactors=(PositiveWidth=1,NegativeWidth=1,PositiveHeight=1,NegativeHeight=1)
    ActiveRegionSize=0.1
    Debug=(bDrawDebug=false,bDraw2D=False,bDrawLocalAxes=False,DebugScale=25,DebugSegments=20,DebugThickness=0.25)
    DriverAxis=(X=0,Y=1,Z=0)
    DriverItem=(Type=Bone,Name="hand_l")
    FalloffRegionScaleFactors=(PositiveWidth=1,NegativeWidth=1,PositiveHeight=1,NegativeHeight=1)
    FalloffSize=0.45
    FlipHeightScaling=False
    FlipWidthScaling=False
    OptionalParentItem=(Type=Bone,Name="None")
    RotationOffset=(X=-30,Y=-60,Z=180)
    ExecutePin=@RigVMFunction_Sequence_1_1_2_1_2.C
   )
   n33 = &RigVMFunction_Sequence(ExecuteContext=@n32.ExecutePin)
   n34 = RigVMVariableNode() {variable=FE}
   n35 = &RigVMFunction_AlphaInterp(
    InterpSpeedDecreasing=100
    InterpSpeedIncreasing=100
    Scale=1
    bClampResult=False
    bInterpResult=False
    Value=&RigVMFunction_MathDoubleMul(B=@n32.OutputParam, A=&RigVMFunction_MathDoubleSub(A=1, B=@n34.Value).Result).Result
   )
   n36 = RigVMVariableNode() {variable=is_fly}
   &RigUnit_ModifyTransforms(
    ItemToModify=((Item=(Type=Bone,Name="dyn_finger_base_l"),Transform=(Rotation=(X=0,Y=0,Z=0,W=1),Translation=(X=0,Y=0,Z=0),Scale3D=(X=1,Y=1,Z=1))))
    Weight=&RigVMFunction_MathDoubleMul(A=@n35.Result, B=&RigVMFunction_MathDoubleSub(A=1, B=@n36.Value).Result).Result
    ItemToModify.0.Transform.Rotation=&RigVMFunction_MathQuaternionFromEuler(Euler=(X=10,Y=-15,Z=-60), RotationOrder=ZXY).Result
    ExecutePin=@n33.A
   )
   &RigUnit_ModifyTransforms(
    ItemToModify=((Item=(Type=Bone,Name="dyn_finger_base_l"),Transform=(Rotation=(X=0,Y=0,Z=0,W=1),Translation=(X=-1,Y=0,Z=0),Scale3D=(X=1,Y=1,Z=1))))
    Weight=&RigVMFunction_MathDoubleMul(A=@n35.Result, B=RigVMVariableNode() {variable=is_crouch}.Value).Result
    ItemToModify.0.Transform.Rotation=&RigVMFunction_MathQuaternionFromEuler(Euler=(X=-35,Y=0,Z=80), RotationOrder=ZXY).Result
    ExecutePin=@n33.B
   )
   n37 = &RigUnit_SphericalPoseReader(
    ActiveRegionScaleFactors=(PositiveWidth=0.5,NegativeWidth=0.5,PositiveHeight=0.25,NegativeHeight=0.5)
    ActiveRegionSize=0.1
    Debug=(bDrawDebug=false,bDraw2D=False,bDrawLocalAxes=False,DebugScale=25,DebugSegments=20,DebugThickness=0.25)
    DriverAxis=(X=0,Y=1,Z=0)
    DriverItem=(Type=Bone,Name="hand_l")
    FalloffRegionScaleFactors=(PositiveWidth=1,NegativeWidth=1,PositiveHeight=1,NegativeHeight=1)
    FalloffSize=0.35
    FlipHeightScaling=False
    FlipWidthScaling=False
    OptionalParentItem=(Type=Bone,Name="None")
    RotationOffset=(X=-60,Y=0,Z=20)
    ExecutePin=@RigVMFunction_Sequence_1_1_2_1_2.E
   )
   n38 = &RigVMFunction_Sequence(ExecuteContext=@n37.ExecutePin)
   n39 = RigVMVariableNode() {variable=FE}
   n40 = &RigVMFunction_AlphaInterp(
    InterpSpeedDecreasing=100
    InterpSpeedIncreasing=100
    Scale=1
    bClampResult=False
    bInterpResult=False
    Value=&RigVMFunction_MathDoubleMul(B=@n37.OutputParam, A=&RigVMFunction_MathDoubleSub(A=1, B=@n39.Value).Result).Result
   )
   &RigUnit_ModifyTransforms(
    ItemToModify=((Item=(Type=Bone,Name="dyn_finger_base_l"),Transform=(Rotation=(X=0,Y=0,Z=0,W=1),Translation=(X=0,Y=0,Z=0),Scale3D=(X=1,Y=1,Z=1))))
    Weight=@n40.Result
    ItemToModify.0.Transform.Rotation=&RigVMFunction_MathQuaternionFromEuler(Euler=(X=20,Y=-20,Z=-60), RotationOrder=ZXY).Result
    ExecutePin=@n38.A
   )
   &RigUnit_ModifyTransforms(
    ItemToModify=((Item=(Type=Bone,Name="dyn_finger_base_l"),Transform=(Rotation=(X=0,Y=0,Z=0,W=1),Translation=(X=0,Y=0,Z=0),Scale3D=(X=1,Y=1,Z=1))))
    Weight=&RigVMFunction_MathDoubleMul(A=@n40.Result, B=RigVMVariableNode() {variable=is_crouch}.Value).Result
    ItemToModify.0.Transform.Rotation=&RigVMFunction_MathQuaternionFromEuler(Euler=(X=-10,Y=10,Z=70), RotationOrder=ZXY).Result
    ExecutePin=@n38.B
   )
   n41 = &RigUnit_SphericalPoseReader(
    ActiveRegionScaleFactors=(PositiveWidth=0.5,NegativeWidth=0.5,PositiveHeight=0.5,NegativeHeight=0.5)
    ActiveRegionSize=0.1
    Debug=(bDrawDebug=false,bDraw2D=False,bDrawLocalAxes=False,DebugScale=25,DebugSegments=20,DebugThickness=0.25)
    DriverAxis=(X=0,Y=1,Z=0)
    DriverItem=(Type=Bone,Name="hand_r")
    FalloffRegionScaleFactors=(PositiveWidth=1,NegativeWidth=1,PositiveHeight=1,NegativeHeight=1)
    FalloffSize=0.3
    FlipHeightScaling=False
    FlipWidthScaling=False
    OptionalParentItem=(Type=Bone,Name="None")
    RotationOffset=(X=15,Y=75,Z=0)
    ExecutePin=@RigVMFunction_Sequence_5.C
   )
   n42 = RigVMVariableNode() {variable=FE}
   Multiply = &RigVMFunction_MathDoubleMul(
    A=&RigVMFunction_MathDoubleMul(B=@n41.OutputParam, A=&RigVMFunction_MathDoubleSub(A=1, B=@n42.Value).Result).Result
    B=RigVMVariableNode() {variable=is_fly}.Value
   )
   > &RigUnit_ModifyTransforms(
    ItemToModify=((Item=(Type=Bone,Name="dyn_finger_base_r"),Transform=(Rotation=(X=0,Y=0,Z=0,W=1),Translation=(X=0,Y=0,Z=0),Scale3D=(X=1,Y=1,Z=1))))
    Weight=@Multiply.Result
    ItemToModify.0.Transform.Rotation=&RigVMFunction_MathQuaternionFromEuler(Euler=(X=0,Y=0,Z=-60), RotationOrder=ZXY).Result
   )
   n43 = &RigVMFunction_Sequence(ExecuteContext=@RigVMFunction_Sequence_1_2.D)
   RigVMFunction_Sequence_2_1 = RigVMAggregateNode(ExecuteContext=@n43.A)
   n44 = &RigUnit_SphericalPoseReader(
    ActiveRegionScaleFactors=(PositiveWidth=0.5,NegativeWidth=0.5,PositiveHeight=0.5,NegativeHeight=0.5)
    ActiveRegionSize=0.1
    Debug=(bDrawDebug=false,bDraw2D=False,bDrawLocalAxes=False,DebugScale=25,DebugSegments=20,DebugThickness=0.25)
    DriverAxis=(X=-1,Y=0,Z=0)
    DriverItem=(Type=Bone,Name="upperarm_r")
    FalloffRegionScaleFactors=(PositiveWidth=1,NegativeWidth=1,PositiveHeight=1,NegativeHeight=1)
    FalloffSize=0.25
    FlipHeightScaling=False
    FlipWidthScaling=False
    OptionalParentItem=(Type=Bone,Name="None")
    RotationOffset=(X=5,Y=125,Z=0)
    ExecutePin=@RigVMFunction_Sequence_2_1.A
   )
   > &RigUnit_ModifyTransforms(
    ItemToModify=((Item=(Type=Bone,Name="dyn_shoulder_armor_r"),Transform=(Rotation=(X=0,Y=0,Z=0,W=1),Translation=(X=0,Y=0,Z=0),Scale3D=(X=1,Y=1,Z=1))))
    ItemToModify.0.Transform.Rotation=&RigVMFunction_MathQuaternionFromEuler(Euler=(X=-10,Y=-7.5,Z=0), RotationOrder=ZXY).Result
    Weight=@n44.OutputParam
   )
   > n45 = &RigUnit_SphericalPoseReader(
    ActiveRegionScaleFactors=(PositiveWidth=0.5,NegativeWidth=0.5,PositiveHeight=0.5,NegativeHeight=0.5)
    ActiveRegionSize=0.1
    Debug=(bDrawDebug=false,bDraw2D=False,bDrawLocalAxes=False,DebugScale=25,DebugSegments=20,DebugThickness=0.25)
    DriverAxis=(X=-1,Y=0,Z=0)
    DriverItem=(Type=Bone,Name="clavicle_r")
    FalloffRegionScaleFactors=(PositiveWidth=1,NegativeWidth=1,PositiveHeight=1,NegativeHeight=1)
    FalloffSize=0.25
    FlipHeightScaling=False
    FlipWidthScaling=False
    OptionalParentItem=(Type=Bone,Name="None")
    RotationOffset=(X=5,Y=135,Z=0)
   )
   > &RigUnit_ModifyTransforms(
    ItemToModify=((Item=(Type=Bone,Name="dyn_shoulder_armor_r"),Transform=(Rotation=(X=0,Y=0,Z=0,W=1),Translation=(X=-5,Y=0,Z=0),Scale3D=(X=1,Y=1,Z=1))))
    ItemToModify.0.Transform.Rotation=&RigVMFunction_MathQuaternionFromEuler(Euler=(X=0,Y=5,Z=0), RotationOrder=ZXY).Result
    Weight=@n45.OutputParam
   )
   n46 = &RigUnit_SphericalPoseReader(
    ActiveRegionScaleFactors=(PositiveWidth=0.5,NegativeWidth=0.5,PositiveHeight=0.5,NegativeHeight=0.5)
    ActiveRegionSize=0.1
    Debug=(bDrawDebug=false,bDraw2D=False,bDrawLocalAxes=False,DebugScale=25,DebugSegments=20,DebugThickness=0.25)
    DriverAxis=(X=-1,Y=0,Z=0)
    DriverItem=(Type=Bone,Name="upperarm_r")
    FalloffRegionScaleFactors=(PositiveWidth=1,NegativeWidth=1,PositiveHeight=1,NegativeHeight=1)
    FalloffSize=0.25
    FlipHeightScaling=False
    FlipWidthScaling=False
    OptionalParentItem=(Type=Bone,Name="None")
    RotationOffset=(X=-20,Y=25,Z=0)
    ExecutePin=@RigVMFunction_Sequence_2_1.B
   )
   > &RigUnit_ModifyTransforms(
    ItemToModify=((Item=(Type=Bone,Name="dyn_shoulder_armor_r"),Transform=(Rotation=(X=0,Y=0,Z=0,W=1),Translation=(X=0,Y=0,Z=0),Scale3D=(X=1,Y=1,Z=1))))
    ItemToModify.0.Transform.Rotation=&RigVMFunction_MathQuaternionFromEuler(Euler=(X=10,Y=20,Z=0), RotationOrder=ZXY).Result
    Weight=@n46.OutputParam
   )
   n47 = &RigUnit_SphericalPoseReader(
    ActiveRegionScaleFactors=(PositiveWidth=0.5,NegativeWidth=0.5,PositiveHeight=0.5,NegativeHeight=0.5)
    ActiveRegionSize=0.1
    Debug=(bDrawDebug=false,bDraw2D=False,bDrawLocalAxes=False,DebugScale=25,DebugSegments=20,DebugThickness=0.25)
    DriverAxis=(X=-1,Y=0,Z=0)
    DriverItem=(Type=Bone,Name="upperarm_r")
    FalloffRegionScaleFactors=(PositiveWidth=1,NegativeWidth=1,PositiveHeight=1,NegativeHeight=1)
    FalloffSize=0.35
    FlipHeightScaling=False
    FlipWidthScaling=False
    OptionalParentItem=(Type=Bone,Name="None")
    RotationOffset=(X=-125,Y=0,Z=25)
    ExecutePin=@RigVMFunction_Sequence_2_1.C
   )
   > &RigUnit_ModifyTransforms(
    ItemToModify=((Item=(Type=Bone,Name="dyn_shoulder_armor_r"),Transform=(Rotation=(X=0,Y=0,Z=0,W=1),Translation=(X=-2,Y=1,Z=-1),Scale3D=(X=1,Y=1,Z=1))))
    ItemToModify.0.Transform.Rotation=&RigVMFunction_MathQuaternionFromEuler(Euler=(X=45,Y=-10,Z=0), RotationOrder=ZXY).Result
    Weight=@n47.OutputParam
   )
   RigVMFunction_Sequence_2 = RigVMAggregateNode(ExecuteContext=@n43.B)
   n48 = &RigUnit_SphericalPoseReader(
    ActiveRegionScaleFactors=(PositiveWidth=0.5,NegativeWidth=0.5,PositiveHeight=0.5,NegativeHeight=0.5)
    ActiveRegionSize=0.1
    Debug=(bDrawDebug=false,bDraw2D=False,bDrawLocalAxes=False,DebugScale=25,DebugSegments=20,DebugThickness=0.25)
    DriverAxis=(X=1,Y=0,Z=0)
    DriverItem=(Type=Bone,Name="upperarm_l")
    FalloffRegionScaleFactors=(PositiveWidth=1,NegativeWidth=1,PositiveHeight=1,NegativeHeight=1)
    FalloffSize=0.3
    FlipHeightScaling=False
    FlipWidthScaling=False
    OptionalParentItem=(Type=Bone,Name="None")
    RotationOffset=(X=-5,Y=-65,Z=0)
    ExecutePin=@RigVMFunction_Sequence_2.A
   )
   > &RigUnit_ModifyTransforms(
    ItemToModify=((Item=(Type=Bone,Name="dyn_shoulder_armor_l"),Transform=(Rotation=(X=0,Y=0,Z=0,W=1),Translation=(X=1,Y=-2,Z=-1),Scale3D=(X=1,Y=1,Z=1))))
    ItemToModify.0.Transform.Rotation=&RigVMFunction_MathQuaternionFromEuler(Euler=(X=-10,Y=-20,Z=0), RotationOrder=ZXY).Result
    Weight=@n48.OutputParam
   )
   > n49 = &RigUnit_SphericalPoseReader(
    ActiveRegionScaleFactors=(PositiveWidth=0.5,NegativeWidth=0.5,PositiveHeight=0.5,NegativeHeight=0.5)
    ActiveRegionSize=0.1
    Debug=(bDrawDebug=false,bDraw2D=False,bDrawLocalAxes=False,DebugScale=25,DebugSegments=20,DebugThickness=0.25)
    DriverAxis=(X=1,Y=0,Z=0)
    DriverItem=(Type=Bone,Name="clavicle_l")
    FalloffRegionScaleFactors=(PositiveWidth=1,NegativeWidth=1,PositiveHeight=1,NegativeHeight=1)
    FalloffSize=0.2
    FlipHeightScaling=False
    FlipWidthScaling=False
    OptionalParentItem=(Type=Bone,Name="None")
    RotationOffset=(X=0,Y=-40,Z=0)
   )
   > &RigUnit_ModifyTransforms(
    ItemToModify=((Item=(Type=Bone,Name="dyn_shoulder_armor_l"),Transform=(Rotation=(X=0,Y=0,Z=0,W=1),Translation=(X=5,Y=0,Z=0),Scale3D=(X=1,Y=1,Z=1))))
    ItemToModify.0.Transform.Rotation=&RigVMFunction_MathQuaternionFromEuler(Euler=(X=0,Y=5,Z=0), RotationOrder=ZXY).Result
    Weight=@n49.OutputParam
   )
   n50 = &RigUnit_SphericalPoseReader(
    ActiveRegionScaleFactors=(PositiveWidth=0.25,NegativeWidth=0.25,PositiveHeight=0.25,NegativeHeight=0.25)
    ActiveRegionSize=0.1
    Debug=(bDrawDebug=false,bDraw2D=False,bDrawLocalAxes=False,DebugScale=25,DebugSegments=20,DebugThickness=0.25)
    DriverAxis=(X=1,Y=0,Z=0)
    DriverItem=(Type=Bone,Name="upperarm_l")
    FalloffRegionScaleFactors=(PositiveWidth=1,NegativeWidth=1,PositiveHeight=1,NegativeHeight=1)
    FalloffSize=0.25
    FlipHeightScaling=False
    FlipWidthScaling=False
    OptionalParentItem=(Type=Bone,Name="None")
    RotationOffset=(X=20,Y=210,Z=0)
    ExecutePin=@RigVMFunction_Sequence_2.B
   )
   > &RigUnit_ModifyTransforms(
    ItemToModify=((Item=(Type=Bone,Name="dyn_shoulder_armor_l"),Transform=(Rotation=(X=0,Y=0,Z=0,W=1),Translation=(X=0,Y=0,Z=0),Scale3D=(X=1,Y=1,Z=1))))
    ItemToModify.0.Transform.Rotation=&RigVMFunction_MathQuaternionFromEuler(Euler=(X=5,Y=20,Z=0), RotationOrder=ZXY).Result
    Weight=@n50.OutputParam
   )
   n51 = &RigUnit_SphericalPoseReader(
    ActiveRegionScaleFactors=(PositiveWidth=0.5,NegativeWidth=0.5,PositiveHeight=0.5,NegativeHeight=0.5)
    ActiveRegionSize=0.1
    Debug=(bDrawDebug=false,bDraw2D=False,bDrawLocalAxes=False,DebugScale=25,DebugSegments=20,DebugThickness=0.25)
    DriverAxis=(X=1,Y=0,Z=0)
    DriverItem=(Type=Bone,Name="upperarm_l")
    FalloffRegionScaleFactors=(PositiveWidth=1,NegativeWidth=1,PositiveHeight=1,NegativeHeight=1)
    FalloffSize=0.2
    FlipHeightScaling=False
    FlipWidthScaling=False
    OptionalParentItem=(Type=Bone,Name="None")
    RotationOffset=(X=60,Y=0,Z=-20)
    ExecutePin=@RigVMFunction_Sequence_2.C
   )
   > &RigUnit_ModifyTransforms(
    ItemToModify=((Item=(Type=Bone,Name="dyn_shoulder_armor_l"),Transform=(Rotation=(X=0,Y=0,Z=0,W=1),Translation=(X=2,Y=0,Z=0),Scale3D=(X=1,Y=1,Z=1))))
    ItemToModify.0.Transform.Rotation=&RigVMFunction_MathQuaternionFromEuler(Euler=(X=20,Y=0,Z=-5), RotationOrder=ZXY).Result
    Weight=@n51.OutputParam
   )
   &RigUnit_ParentConstraint(
    AdvancedSettings=(RotationOrderForFilter=XZY)
    Child=(Type=Bone,Name="spike_01")
    Parents="((Item=(Type=Bone,Name=\"spike_01_r\"),Weight=0.0059),(Item=(Type=Bone,Name=\"neck_02\"),Weight=0.0078),(Item=(Type=Bone,Name=\"spike_01\"),Weight=0.013),(Item=(Type=Bone,Name=\"spine_02\"),Weight=0.0485),(Item=(Type=Bone,Name=\"spine_01\"),Weight=0.0502),(Item=(Typ ...(truncated)"
    Weight=1
    ExecutePin=@RigVMFunction_Sequence_1_2.E
   )
   > &RigUnit_ParentConstraint(
    AdvancedSettings=(RotationOrderForFilter=XZY)
    Child=(Type=Bone,Name="spike_01_l")
    Parents="((Item=(Type=Bone,Name=\"spike_01\"),Weight=0.0032),(Item=(Type=Bone,Name=\"neck_02\"),Weight=0.0066),(Item=(Type=Bone,Name=\"spike_01_l\"),Weight=0.0292),(Item=(Type=Bone,Name=\"spine_02\"),Weight=0.0482),(Item=(Type=Bone,Name=\"spine_04\"),Weight=0.0501),(Item=(Ty ...(truncated)"
    Weight=1
   )
   > &RigUnit_ParentConstraint(
    AdvancedSettings=(RotationOrderForFilter=XZY)
    Child=(Type=Bone,Name="spike_01_r")
    Parents="((Item=(Type=Bone,Name=\"spike_01\"),Weight=0.0017),(Item=(Type=Bone,Name=\"neck_02\"),Weight=0.0069),(Item=(Type=Bone,Name=\"spike_01_r\"),Weight=0.0406),(Item=(Type=Bone,Name=\"spine_02\"),Weight=0.0484),(Item=(Type=Bone,Name=\"spine_04\"),Weight=0.0489),(Item=(Ty ...(truncated)"
    Weight=1
   )
   > &RigUnit_ParentConstraint(
    AdvancedSettings=(RotationOrderForFilter=XZY)
    Child=(Type=Bone,Name="spike_02")
    Parents="((Item=(Type=Bone,Name=\"clavicle_r\"),Weight=0.0006),(Item=(Type=Bone,Name=\"spine_01\"),Weight=0.0336),(Item=(Type=Bone,Name=\"spine_02\"),Weight=0.0359),(Item=(Type=Bone,Name=\"dyn_pelvis\"),Weight=0.0836),(Item=(Type=Bone,Name=\"spine_04\"),Weight=0.1682),(Item= ...(truncated)"
    Weight=1
   )
   > &RigUnit_ParentConstraint(
    AdvancedSettings=(RotationOrderForFilter=XZY)
    Child=(Type=Bone,Name="spike_02_l")
    Parents="((Item=(Type=Bone,Name=\"spike_02_l\"),Weight=0.007),(Item=(Type=Bone,Name=\"clavicle_l\"),Weight=0.01),(Item=(Type=Bone,Name=\"spine_01\"),Weight=0.0345),(Item=(Type=Bone,Name=\"spine_02\"),Weight=0.0404),(Item=(Type=Bone,Name=\"dyn_pelvis\"),Weight=0.0822),(Item=( ...(truncated)"
    Weight=1
   )
   > &RigUnit_ParentConstraint(
    AdvancedSettings=(RotationOrderForFilter=XZY)
    Child=(Type=Bone,Name="spike_02_r")
    Parents="((Item=(Type=Bone,Name=\"spike_02_r\"),Weight=0.007),(Item=(Type=Bone,Name=\"clavicle_r\"),Weight=0.01),(Item=(Type=Bone,Name=\"spine_01\"),Weight=0.0345),(Item=(Type=Bone,Name=\"spine_02\"),Weight=0.0404),(Item=(Type=Bone,Name=\"dyn_pelvis\"),Weight=0.0822),(Item=( ...(truncated)"
    Weight=1
   )
   > &RigUnit_ParentConstraint(
    AdvancedSettings=(RotationOrderForFilter=XZY)
    Child=(Type=Bone,Name="spike_03")
    Parents="((Item=(Type=Bone,Name=\"spine_05\"),Weight=0.0753),(Item=(Type=Bone,Name=\"spine_01\"),Weight=0.0994),(Item=(Type=Bone,Name=\"dyn_pelvis\"),Weight=0.1207),(Item=(Type=Bone,Name=\"spine_02\"),Weight=0.1316),(Item=(Type=Bone,Name=\"spine_04\"),Weight=0.1733),(Item=(T ...(truncated)"
    Weight=1
   )
   > &RigUnit_ParentConstraint(
    AdvancedSettings=(RotationOrderForFilter=XZY)
    Child=(Type=Bone,Name="spike_03_l")
    Parents="((Item=(Type=Bone,Name=\"spine_05\"),Weight=0.078),(Item=(Type=Bone,Name=\"spine_01\"),Weight=0.1003),(Item=(Type=Bone,Name=\"dyn_pelvis\"),Weight=0.1208),(Item=(Type=Bone,Name=\"spine_02\"),Weight=0.1352),(Item=(Type=Bone,Name=\"spine_04\"),Weight=0.1692),(Item=(Ty ...(truncated)"
    Weight=1
   )
   > &RigUnit_ParentConstraint(
    AdvancedSettings=(RotationOrderForFilter=XZY)
    Child=(Type=Bone,Name="spike_03_r")
    Parents="((Item=(Type=Bone,Name=\"spine_05\"),Weight=0.078),(Item=(Type=Bone,Name=\"spine_01\"),Weight=0.1003),(Item=(Type=Bone,Name=\"dyn_pelvis\"),Weight=0.1208),(Item=(Type=Bone,Name=\"spine_02\"),Weight=0.1352),(Item=(Type=Bone,Name=\"spine_04\"),Weight=0.1692),(Item=(Ty ...(truncated)"
    Weight=1
   )
   > &RigUnit_ParentConstraint(
    AdvancedSettings=(RotationOrderForFilter=XZY)
    Child=(Type=Bone,Name="spike_04")
    Parents="((Item=(Type=Bone,Name=\"spike_04\"),Weight=0.0002),(Item=(Type=Bone,Name=\"spine_04\"),Weight=0.0854),(Item=(Type=Bone,Name=\"spine_01\"),Weight=0.1217),(Item=(Type=Bone,Name=\"dyn_pelvis\"),Weight=0.1951),(Item=(Type=Bone,Name=\"spine_02\"),Weight=0.2831),(Item=(T ...(truncated)"
    Weight=1
   )
   > &RigUnit_ParentConstraint(
    AdvancedSettings=(RotationOrderForFilter=XZY)
    Child=(Type=Bone,Name="spike_04_l")
    Parents=((Item=(Type=Bone,Name="spine_04"),Weight=0.0761),(Item=(Type=Bone,Name="spine_01"),Weight=0.1265),(Item=(Type=Bone,Name="dyn_pelvis"),Weight=0.1954),(Item=(Type=Bone,Name="spine_02"),Weight=0.2892),(Item=(Type=Bone,Name="spine_03"),Weight=0.3128))
    Weight=1
   )
   > &RigUnit_ParentConstraint(
    AdvancedSettings=(RotationOrderForFilter=XZY)
    Child=(Type=Bone,Name="spike_04_r")
    Parents=((Item=(Type=Bone,Name="spine_04"),Weight=0.0761),(Item=(Type=Bone,Name="spine_01"),Weight=0.1265),(Item=(Type=Bone,Name="dyn_pelvis"),Weight=0.1954),(Item=(Type=Bone,Name="spine_02"),Weight=0.2892),(Item=(Type=Bone,Name="spine_03"),Weight=0.3128))
    Weight=1
   )
   > RigVMFunction_Sequence_3 = RigVMAggregateNode()
   n52 = RigVMVariableNode() {variable=dyn_spike_ButterPlate}
   &RigUnit_ModifyTransforms(
    ItemToModify=((Item=(Type=Bone,Name="spike_01"),Transform=(Rotation=(X=0,Y=0,Z=0,W=1),Translation=(X=0,Y=5.5,Z=0),Scale3D=(X=1,Y=1,Z=1))))
    ExecutePin=@RigVMFunction_Sequence_3.A
    Weight=@n52.Value
   )
   n53 = RigVMVariableNode() {variable=dyn_spike}
   > &RigUnit_ModifyTransforms(
    ItemToModify=((Item=(Type=Bone,Name="spike_01_l"),Transform=(Rotation=(X=0,Y=0,Z=0,W=1),Translation=(X=0,Y=1,Z=0),Scale3D=(X=1,Y=1,Z=1))))
    Weight=@n53.Value
   )
   > &RigUnit_ModifyTransforms(
    ItemToModify=((Item=(Type=Bone,Name="spike_01_r"),Transform=(Rotation=(X=0,Y=0,Z=0,W=1),Translation=(X=0,Y=1,Z=0),Scale3D=(X=1,Y=1,Z=1))))
    Weight=@n53.Value
   )
   &RigUnit_ModifyTransforms(
    ItemToModify=((Item=(Type=Bone,Name="spike_02"),Transform=(Rotation=(X=0,Y=0,Z=0,W=1),Translation=(X=0,Y=12,Z=0),Scale3D=(X=1,Y=1,Z=1))))
    ExecutePin=@RigVMFunction_Sequence_3.B
    Weight=@n52.Value
   )
   > &RigUnit_ModifyTransforms(
    ItemToModify=((Item=(Type=Bone,Name="spike_02_l"),Transform=(Rotation=(X=0,Y=0,Z=0,W=1),Translation=(X=0,Y=7.5,Z=0),Scale3D=(X=1,Y=1,Z=1))))
    Weight=@n53.Value
   )
   > &RigUnit_ModifyTransforms(
    ItemToModify=((Item=(Type=Bone,Name="spike_02_r"),Transform=(Rotation=(X=0,Y=0,Z=0,W=1),Translation=(X=0,Y=7.5,Z=0),Scale3D=(X=1,Y=1,Z=1))))
    Weight=@n53.Value
   )
   &RigUnit_ModifyTransforms(
    ItemToModify=((Item=(Type=Bone,Name="spike_03"),Transform=(Rotation=(X=0,Y=0,Z=0,W=1),Translation=(X=0,Y=12,Z=0),Scale3D=(X=1,Y=1,Z=1))))
    ExecutePin=@RigVMFunction_Sequence_3.C
    Weight=@n52.Value
   )
   > &RigUnit_ModifyTransforms(
    ItemToModify=((Item=(Type=Bone,Name="spike_03_l"),Transform=(Rotation=(X=0,Y=0,Z=0,W=1),Translation=(X=0,Y=7,Z=-1),Scale3D=(X=1,Y=1,Z=1))))
    Weight=@n53.Value
   )
   > &RigUnit_ModifyTransforms(
    ItemToModify=((Item=(Type=Bone,Name="spike_03_r"),Transform=(Rotation=(X=0,Y=0,Z=0,W=1),Translation=(X=0,Y=7,Z=1),Scale3D=(X=1,Y=1,Z=1))))
    Weight=@n53.Value
   )
   &RigUnit_ModifyTransforms(
    ItemToModify=((Item=(Type=Bone,Name="spike_04"),Transform=(Rotation=(X=0,Y=0,Z=0,W=1),Translation=(X=0,Y=12,Z=0),Scale3D=(X=1,Y=1,Z=1))))
    ExecutePin=@RigVMFunction_Sequence_3.D
    Weight=@n52.Value
   )
   > &RigUnit_ModifyTransforms(
    ItemToModify=((Item=(Type=Bone,Name="spike_04_l"),Transform=(Rotation=(X=0,Y=0,Z=0,W=1),Translation=(X=0,Y=7.5,Z=-1),Scale3D=(X=1,Y=1,Z=1))))
    Weight=@n53.Value
   )
   > &RigUnit_ModifyTransforms(
    ItemToModify=((Item=(Type=Bone,Name="spike_04_r"),Transform=(Rotation=(X=0,Y=0,Z=0,W=1),Translation=(X=0,Y=7.5,Z=1),Scale3D=(X=1,Y=1,Z=1))))
    Weight=@n53.Value
   )
   n54 = &RigVMFunction_Sequence(ExecuteContext=@RigVMFunction_Sequence_1_2.C)
   n55 = &RigVMFunction_Sequence(ExecuteContext=@n54.A)
   n56 = &RigVMFunction_Sequence(ExecuteContext=@n55.A)
   n57 = &RigVMFunction_Sequence(ExecuteContext=@n56.A)
   n58 = RigVMVariableNode() {variable=FE}
   n59 = &RigVMFunction_MathDoubleClamp(Value=&RigVMFunction_MathDoubleSub(B=&RigVMFunction_MathDoubleSub(A=1, B=@n58.Value).Result, A=RigVMVariableNode() {variable=FE_backpack}.Value).Result)
   &RigUnit_ModifyTransforms(
    ItemToModify=((Item=(Type=Bone,Name="index_01_l"),Transform=(Rotation=(X=0,Y=0,Z=0,W=1),Translation=(X=0,Y=0,Z=0),Scale3D=(X=1,Y=1,Z=1))))
    ItemToModify.0.Transform.Rotation=&RigVMFunction_MathQuaternionFromEuler(Euler=(X=-15,Y=120,Z=-90), RotationOrder=ZXY).Result
    ExecutePin=@n57.A
    Weight=@n59.Result
   )
   > &RigUnit_ModifyTransforms(
    ItemToModify=((Item=(Type=Bone,Name="index_02_l"),Transform=(Rotation=(X=0,Y=0,Z=0,W=1),Translation=(X=0,Y=0,Z=0),Scale3D=(X=1,Y=1,Z=1))))
    ItemToModify.0.Transform.Rotation=&RigVMFunction_MathQuaternionFromEuler(Euler=(X=0,Y=0,Z=-35), RotationOrder=ZXY).Result
    Weight=@n59.Result
   )
   > &RigUnit_ModifyTransforms(
    ItemToModify=((Item=(Type=Bone,Name="index_03_l"),Transform=(Rotation=(X=0,Y=0,Z=0,W=1),Translation=(X=0,Y=0,Z=0),Scale3D=(X=1,Y=1,Z=1))))
    ItemToModify.0.Transform.Rotation=&RigVMFunction_MathQuaternionFromEuler(Euler=(X=0,Y=0,Z=-15), RotationOrder=ZXY).Result
    Weight=@n59.Result
   )
   &RigUnit_ModifyTransforms(
    ItemToModify=((Item=(Type=Bone,Name="middle_01_l"),Transform=(Rotation=(X=0,Y=0,Z=0,W=1),Translation=(X=0,Y=0,Z=0),Scale3D=(X=1,Y=1,Z=1))))
    ItemToModify.0.Transform.Rotation=&RigVMFunction_MathQuaternionFromEuler(Euler=(X=0,Y=-30,Z=-15), RotationOrder=ZXY).Result
    ExecutePin=@n57.B
    Weight=@n59.Result
   )
   > &RigUnit_ModifyTransforms(
    ItemToModify=((Item=(Type=Bone,Name="middle_02_l"),Transform=(Rotation=(X=0,Y=0,Z=0,W=1),Translation=(X=0,Y=0,Z=0),Scale3D=(X=1,Y=1,Z=1))))
    ItemToModify.0.Transform.Rotation=&RigVMFunction_MathQuaternionFromEuler(Euler=(X=0,Y=20,Z=-35), RotationOrder=ZXY).Result
    Weight=@n59.Result
   )
   > &RigUnit_ModifyTransforms(
    ItemToModify=((Item=(Type=Bone,Name="middle_03_l"),Transform=(Rotation=(X=0,Y=0,Z=0,W=1),Translation=(X=0,Y=0,Z=0),Scale3D=(X=1,Y=1,Z=1))))
    ItemToModify.0.Transform.Rotation=&RigVMFunction_MathQuaternionFromEuler(Euler=(X=0,Y=0,Z=-15), RotationOrder=ZXY).Result
    Weight=@n59.Result
   )
   n60 = &RigVMFunction_Sequence(ExecuteContext=@n56.B)
   &RigUnit_ModifyTransforms(
    ItemToModify=((Item=(Type=Bone,Name="ring_01_l"),Transform=(Rotation=(X=0,Y=0,Z=0,W=1),Translation=(X=0,Y=0,Z=0),Scale3D=(X=1,Y=1,Z=1))))
    ItemToModify.0.Transform.Rotation=&RigVMFunction_MathQuaternionFromEuler(Euler=(X=45,Y=60,Z=30), RotationOrder=ZXY).Result
    ExecutePin=@n60.A
    Weight=@n59.Result
   )
   > &RigUnit_ModifyTransforms(
    ItemToModify=((Item=(Type=Bone,Name="ring_02_l"),Transform=(Rotation=(X=0,Y=0,Z=0,W=1),Translation=(X=0,Y=0,Z=0),Scale3D=(X=1,Y=1,Z=1))))
    ItemToModify.0.Transform.Rotation=&RigVMFunction_MathQuaternionFromEuler(Euler=(X=0,Y=0,Z=0), RotationOrder=ZXY).Result
    Weight=@n59.Result
   )
   > &RigUnit_ModifyTransforms(
    ItemToModify=((Item=(Type=Bone,Name="ring_03_l"),Transform=(Rotation=(X=0,Y=0,Z=0,W=1),Translation=(X=0,Y=0,Z=0),Scale3D=(X=1,Y=1,Z=1))))
    ItemToModify.0.Transform.Rotation=&RigVMFunction_MathQuaternionFromEuler(Euler=(X=0,Y=0,Z=0), RotationOrder=ZXY).Result
    Weight=@n59.Result
   )
   &RigUnit_ModifyTransforms(
    ItemToModify=((Item=(Type=Bone,Name="pinky_01_l"),Transform=(Rotation=(X=0,Y=0,Z=0,W=1),Translation=(X=0,Y=0,Z=0),Scale3D=(X=1,Y=1,Z=1))))
    ItemToModify.0.Transform.Rotation=&RigVMFunction_MathQuaternionFromEuler(Euler=(X=-35,Y=0,Z=15), RotationOrder=ZXY).Result
    ExecutePin=@n60.B
    Weight=@n59.Result
   )
   > &RigUnit_ModifyTransforms(
    ItemToModify=((Item=(Type=Bone,Name="pinky_02_l"),Transform=(Rotation=(X=0,Y=0,Z=0,W=1),Translation=(X=0,Y=0,Z=0),Scale3D=(X=1,Y=1,Z=1))))
    ItemToModify.0.Transform.Rotation=&RigVMFunction_MathQuaternionFromEuler(Euler=(X=0,Y=0,Z=0), RotationOrder=ZXY).Result
    Weight=@n59.Result
   )
   > &RigUnit_ModifyTransforms(
    ItemToModify=((Item=(Type=Bone,Name="pinky_03_l"),Transform=(Rotation=(X=0,Y=0,Z=0,W=1),Translation=(X=0,Y=0,Z=0),Scale3D=(X=1,Y=1,Z=1))))
    ItemToModify.0.Transform.Rotation=&RigVMFunction_MathQuaternionFromEuler(Euler=(X=0,Y=0,Z=0), RotationOrder=ZXY).Result
    Weight=@n59.Result
   )
   n61 = &RigVMFunction_Sequence(ExecuteContext=@n55.B)
   n62 = &RigVMFunction_Sequence(ExecuteContext=@n61.A)
   n63 = RigVMVariableNode() {variable=FE_backpack}
   Clamp = &RigVMFunction_MathDoubleClamp(Value=&RigVMFunction_MathDoubleSub(B=&RigVMFunction_MathDoubleSub(A=1, B=@n63.Value).Result, A=RigVMVariableNode() {variable=FE}.Value).Result)
   &RigUnit_ModifyTransforms(
    ItemToModify=((Item=(Type=Bone,Name="index_01_r"),Transform=(Rotation=(X=0,Y=0,Z=0,W=1),Translation=(X=0,Y=0,Z=0),Scale3D=(X=1,Y=1,Z=1))))
    ItemToModify.0.Transform.Rotation=&RigVMFunction_MathQuaternionFromEuler(Euler=(X=-90,Y=90,Z=-60), RotationOrder=ZXY).Result
    ExecutePin=@n62.A
    Weight=@Clamp.Result
   )
   > &RigUnit_ModifyTransforms(
    ItemToModify=((Item=(Type=Bone,Name="index_02_r"),Transform=(Rotation=(X=0,Y=0,Z=0,W=1),Translation=(X=0,Y=0,Z=0),Scale3D=(X=1,Y=1,Z=1))))
    ItemToModify.0.Transform.Rotation=&RigVMFunction_MathQuaternionFromEuler(Euler=(X=0,Y=20,Z=-60), RotationOrder=ZXY).Result
    Weight=@Clamp.Result
   )
   > &RigUnit_ModifyTransforms(
    ItemToModify=((Item=(Type=Bone,Name="index_03_r"),Transform=(Rotation=(X=0,Y=0,Z=0,W=1),Translation=(X=0,Y=0,Z=0),Scale3D=(X=1,Y=1,Z=1))))
    ItemToModify.0.Transform.Rotation=&RigVMFunction_MathQuaternionFromEuler(Euler=(X=0,Y=0,Z=-20), RotationOrder=ZXY).Result
    Weight=@Clamp.Result
   )
   &RigUnit_ModifyTransforms(
    ItemToModify=((Item=(Type=Bone,Name="middle_01_r"),Transform=(Rotation=(X=0,Y=0,Z=0,W=1),Translation=(X=0,Y=0,Z=0),Scale3D=(X=1,Y=1,Z=1))))
    ItemToModify.0.Transform.Rotation=&RigVMFunction_MathQuaternionFromEuler(Euler=(X=35,Y=-60,Z=-30), RotationOrder=ZXY).Result
    ExecutePin=@n62.B
    Weight=@Clamp.Result
   )
   > &RigUnit_ModifyTransforms(
    ItemToModify=((Item=(Type=Bone,Name="middle_02_r"),Transform=(Rotation=(X=0,Y=0,Z=0,W=1),Translation=(X=0,Y=0,Z=0),Scale3D=(X=1,Y=1,Z=1))))
    ItemToModify.0.Transform.Rotation=&RigVMFunction_MathQuaternionFromEuler(Euler=(X=0,Y=20,Z=-35), RotationOrder=ZXY).Result
    Weight=@Clamp.Result
   )
   > &RigUnit_ModifyTransforms(
    ItemToModify=((Item=(Type=Bone,Name="middle_03_r"),Transform=(Rotation=(X=0,Y=0,Z=0,W=1),Translation=(X=0,Y=0,Z=0),Scale3D=(X=1,Y=1,Z=1))))
    ItemToModify.0.Transform.Rotation=&RigVMFunction_MathQuaternionFromEuler(Euler=(X=0,Y=0,Z=-20), RotationOrder=ZXY).Result
    Weight=@Clamp.Result
   )
   n64 = &RigVMFunction_Sequence(ExecuteContext=@n61.B)
   &RigUnit_ModifyTransforms(
    ItemToModify=((Item=(Type=Bone,Name="ring_01_r"),Transform=(Rotation=(X=0,Y=0,Z=0,W=1),Translation=(X=0,Y=0,Z=0),Scale3D=(X=1,Y=1,Z=1))))
    ItemToModify.0.Transform.Rotation=&RigVMFunction_MathQuaternionFromEuler(Euler=(X=60,Y=90,Z=30), RotationOrder=ZXY).Result
    ExecutePin=@n64.A
    Weight=@Clamp.Result
   )
   > &RigUnit_ModifyTransforms(
    ItemToModify=((Item=(Type=Bone,Name="ring_02_r"),Transform=(Rotation=(X=0,Y=0,Z=0,W=1),Translation=(X=0,Y=0,Z=0),Scale3D=(X=1,Y=1,Z=1))))
    ItemToModify.0.Transform.Rotation=&RigVMFunction_MathQuaternionFromEuler(Euler=(X=0,Y=0,Z=0), RotationOrder=ZXY).Result
    Weight=@Clamp.Result
   )
   > &RigUnit_ModifyTransforms(
    ItemToModify=((Item=(Type=Bone,Name="ring_03_r"),Transform=(Rotation=(X=0,Y=0,Z=0,W=1),Translation=(X=0,Y=0,Z=0),Scale3D=(X=1,Y=1,Z=1))))
    ItemToModify.0.Transform.Rotation=&RigVMFunction_MathQuaternionFromEuler(Euler=(X=0,Y=0,Z=0), RotationOrder=ZXY).Result
    Weight=@Clamp.Result
   )
   &RigUnit_ModifyTransforms(
    ItemToModify=((Item=(Type=Bone,Name="pinky_01_r"),Transform=(Rotation=(X=0,Y=0,Z=0,W=1),Translation=(X=0,Y=0,Z=0),Scale3D=(X=1,Y=1,Z=1))))
    ItemToModify.0.Transform.Rotation=&RigVMFunction_MathQuaternionFromEuler(Euler=(X=-60,Y=-60,Z=0), RotationOrder=ZXY).Result
    ExecutePin=@n64.B
    Weight=@Clamp.Result
   )
   > &RigUnit_ModifyTransforms(
    ItemToModify=((Item=(Type=Bone,Name="pinky_02_r"),Transform=(Rotation=(X=0,Y=0,Z=0,W=1),Translation=(X=0,Y=0,Z=0),Scale3D=(X=1,Y=1,Z=1))))
    ItemToModify.0.Transform.Rotation=&RigVMFunction_MathQuaternionFromEuler(Euler=(X=0,Y=0,Z=-45), RotationOrder=ZXY).Result
    Weight=@Clamp.Result
   )
   > &RigUnit_ModifyTransforms(
    ItemToModify=((Item=(Type=Bone,Name="pinky_03_r"),Transform=(Rotation=(X=0,Y=0,Z=0,W=1),Translation=(X=0,Y=0,Z=0),Scale3D=(X=1,Y=1,Z=1))))
    ItemToModify.0.Transform.Rotation=&RigVMFunction_MathQuaternionFromEuler(Euler=(X=0,Y=0,Z=-15), RotationOrder=ZXY).Result
    Weight=@Clamp.Result
   )
   RigVMFunction_ControlFlowBranch = &RigVMFunction_ControlFlowBranch(ExecuteContext=@n54.B, Condition=RigVMVariableNode() {variable=Enable Finger Retarget}.Value)
   n65 = &RigVMFunction_Sequence(ExecuteContext=@RigVMFunction_ControlFlowBranch.True)
   n66 = RigVMDispatchNode(ExecuteContext=@n65.A, Array=RigVMVariableNode() {variable=Fingers_Up}.Value) {template="DISPATCH_RigVMDispatch_ArrayIterator(in Array,out Element,out Index,out Count,out Ratio)"}
   > Map_Fingers_Up = RigVMFunctionReferenceNode(
    Inverse=False
    Target=RigVMDispatchNode(Index=@n66.Index, Array=RigVMVariableNode() {variable=Remap Fingers_Up}.Value) {template="DISPATCH_RigVMDispatch_ArrayGetAtIndex(in Array,in Index,out Element)"}.Element.Keys
    Source=@n66.Element.Keys
   )
   n67 = RigVMDispatchNode(ExecuteContext=@n65.B, Array=RigVMVariableNode() {variable=Fingers_Down}.Value) {template="DISPATCH_RigVMDispatch_ArrayIterator(in Array,out Element,out Index,out Count,out Ratio)"}
   > Map_Fingers_Down = RigVMFunctionReferenceNode(
    Inverse=False
    Target=RigVMDispatchNode(Index=@n67.Index, Array=RigVMVariableNode() {variable=Remap Fingers_Down}.Value) {template="DISPATCH_RigVMDispatch_ArrayGetAtIndex(in Array,in Index,out Element)"}.Element.Keys
    Source=@n67.Element.Keys
   )
   n68 = &RigVMFunction_Sequence(ExecuteContext=())
   &RigUnit_RotationConstraintLocalSpaceOffset(
    AdvancedSettings=(InterpolationType=Average,RotationOrderForFilter=XZY)
    Child=(Type=Bone,Name="dyn_finger_4_1_r")
    Filter=(bX=True,bY=True,bZ=True)
    Parents=((Item=(Type=Bone,Name="thumb_01_r"),Weight=1))
    Weight=1
    ExecutePin=@n68.A
   )
   > &RigUnit_RotationConstraintLocalSpaceOffset(
    AdvancedSettings=(InterpolationType=Average,RotationOrderForFilter=XZY)
    Child=(Type=Bone,Name="dyn_finger_4_2_r")
    Filter=(bX=True,bY=True,bZ=True)
    Parents=((Item=(Type=Bone,Name="thumb_02_r"),Weight=1))
    Weight=1
   )
   > &RigUnit_RotationConstraintLocalSpaceOffset(
    AdvancedSettings=(InterpolationType=Average,RotationOrderForFilter=XZY)
    Child=(Type=Bone,Name="dyn_finger_4_3_r")
    Filter=(bX=True,bY=True,bZ=True)
    Parents=((Item=(Type=Bone,Name="thumb_03_r"),Weight=1))
    Weight=1
   )
   > &RigUnit_RotationConstraintLocalSpaceOffset(
    AdvancedSettings=(InterpolationType=Average,RotationOrderForFilter=XZY)
    Child=(Type=Bone,Name="dyn_finger_3_1_r")
    Filter=(bX=True,bY=True,bZ=True)
    Parents=((Item=(Type=Bone,Name="thumb_01_r"),Weight=1))
    Weight=1
   )
   > &RigUnit_RotationConstraintLocalSpaceOffset(
    AdvancedSettings=(InterpolationType=Average,RotationOrderForFilter=XZY)
    Child=(Type=Bone,Name="dyn_finger_3_2_r")
    Filter=(bX=True,bY=True,bZ=True)
    Parents=((Item=(Type=Bone,Name="thumb_02_r"),Weight=1))
    Weight=1
   )
   > &RigUnit_RotationConstraintLocalSpaceOffset(
    AdvancedSettings=(InterpolationType=Average,RotationOrderForFilter=XZY)
    Child=(Type=Bone,Name="dyn_finger_3_3_r")
    Filter=(bX=True,bY=True,bZ=True)
    Parents=((Item=(Type=Bone,Name="thumb_03_r"),Weight=1))
    Weight=1
   )
   > &RigUnit_RotationConstraintLocalSpaceOffset(
    AdvancedSettings=(InterpolationType=Average,RotationOrderForFilter=XZY)
    Child=(Type=Bone,Name="dyn_finger_2_1_r")
    Filter=(bX=true,bY=true,bZ=true)
    Parents=((Item=(Type=Bone,Name="middle_01_r"),Weight=1))
    Weight=1
   )
   > &RigUnit_RotationConstraintLocalSpaceOffset(
    AdvancedSettings=(InterpolationType=Average,RotationOrderForFilter=XZY)
    Child=(Type=Bone,Name="dyn_finger_2_2_r")
    Filter=(bX=false,bY=true,bZ=false)
    Parents=((Item=(Type=Bone,Name="middle_02_r"),Weight=1))
    Weight=1
   )
   > &RigUnit_RotationConstraintLocalSpaceOffset(
    AdvancedSettings=(InterpolationType=Average,RotationOrderForFilter=XZY)
    Child=(Type=Bone,Name="dyn_finger_2_3_r")
    Filter=(bX=false,bY=True,bZ=false)
    Parents=((Item=(Type=Bone,Name="middle_03_r"),Weight=1))
    Weight=1
   )
   > &RigUnit_RotationConstraintLocalSpaceOffset(
    AdvancedSettings=(InterpolationType=Average,RotationOrderForFilter=XZY)
    Child=(Type=Bone,Name="dyn_finger_1_1_r")
    Filter=(bX=true,bY=true,bZ=true)
    Parents=((Item=(Type=Bone,Name="pinky_01_r"),Weight=1))
    Weight=1
   )
   > &RigUnit_RotationConstraintLocalSpaceOffset(
    AdvancedSettings=(InterpolationType=Average,RotationOrderForFilter=XZY)
    Child=(Type=Bone,Name="dyn_finger_1_2_r")
    Filter=(bX=false,bY=true,bZ=false)
    Parents=((Item=(Type=Bone,Name="pinky_02_r"),Weight=1))
    Weight=1
   )
   > n69 = &RigUnit_RotationConstraintLocalSpaceOffset(
    AdvancedSettings=(InterpolationType=Average,RotationOrderForFilter=XZY)
    Child=(Type=Bone,Name="dyn_finger_1_3_r")
    Filter=(bX=false,bY=True,bZ=false)
    Parents=((Item=(Type=Bone,Name="pinky_03_r"),Weight=1))
    Weight=1
   )
   n70 = &RigVMFunction_Sequence(ExecuteContext=@n69.ExecutePin)
   n71 = &RigUnit_SphericalPoseReader(
    ActiveRegionScaleFactors=(PositiveWidth=0.5,NegativeWidth=0.5,PositiveHeight=0.5,NegativeHeight=0.5)
    ActiveRegionSize=0.1
    Debug=(bDrawDebug=true,bDraw2D=False,bDrawLocalAxes=False,DebugScale=25,DebugSegments=20,DebugThickness=0.25)
    DriverAxis=(X=0,Y=1,Z=0)
    DriverItem=(Type=Bone,Name="hand_r")
    FalloffRegionScaleFactors=(PositiveWidth=1,NegativeWidth=1,PositiveHeight=1,NegativeHeight=1)
    FalloffSize=0.3
    FlipHeightScaling=False
    FlipWidthScaling=False
    OptionalParentItem=(Type=Bone,Name="None")
    RotationOffset=(X=20,Y=-10,Z=0)
    ExecutePin=@n70.A
   )
   > &RigUnit_ModifyTransforms(
    ItemToModify=((Item=(Type=Bone,Name="dyn_finger_base_r"),Transform=(Rotation=(X=0,Y=0,Z=0,W=1),Translation=(X=0,Y=0,Z=0),Scale3D=(X=1,Y=1,Z=1))))
    ItemToModify.0.Transform.Rotation=&RigVMFunction_MathQuaternionFromEuler(Euler=(X=10,Y=0,Z=-25), RotationOrder=ZXY).Result
    Weight=@n71.OutputParam
   )
   RigVMFunction_Sequence_1_1_1 = RigVMAggregateNode(ExecuteContext=@n68.B)
   &RigUnit_RotationConstraintLocalSpaceOffset(
    AdvancedSettings=(InterpolationType=Average,RotationOrderForFilter=XZY)
    Child=(Type=Bone,Name="dyn_finger_4_1_l")
    Filter=(bX=True,bY=True,bZ=True)
    Parents=((Item=(Type=Bone,Name="thumb_01_l"),Weight=1))
    Weight=1
    ExecutePin=@RigVMFunction_Sequence_1_1_1.A
   )
   > &RigUnit_RotationConstraintLocalSpaceOffset(
    AdvancedSettings=(InterpolationType=Average,RotationOrderForFilter=XZY)
    Child=(Type=Bone,Name="dyn_finger_4_2_l")
    Filter=(bX=True,bY=True,bZ=True)
    Parents=((Item=(Type=Bone,Name="thumb_02_l"),Weight=1))
    Weight=1
   )
   > &RigUnit_RotationConstraintLocalSpaceOffset(
    AdvancedSettings=(InterpolationType=Average,RotationOrderForFilter=XZY)
    Child=(Type=Bone,Name="dyn_finger_4_3_l")
    Filter=(bX=True,bY=True,bZ=True)
    Parents=((Item=(Type=Bone,Name="thumb_03_l"),Weight=1))
    Weight=1
   )
   > &RigUnit_RotationConstraintLocalSpaceOffset(
    AdvancedSettings=(InterpolationType=Average,RotationOrderForFilter=XZY)
    Child=(Type=Bone,Name="dyn_finger_3_1_l")
    Filter=(bX=True,bY=True,bZ=True)
    Parents=((Item=(Type=Bone,Name="thumb_01_l"),Weight=1))
    Weight=1
   )
   > &RigUnit_RotationConstraintLocalSpaceOffset(
    AdvancedSettings=(InterpolationType=Average,RotationOrderForFilter=XZY)
    Child=(Type=Bone,Name="dyn_finger_3_2_l")
    Filter=(bX=True,bY=True,bZ=True)
    Parents=((Item=(Type=Bone,Name="thumb_02_l"),Weight=1))
    Weight=1
   )
   > &RigUnit_RotationConstraintLocalSpaceOffset(
    AdvancedSettings=(InterpolationType=Average,RotationOrderForFilter=XZY)
    Child=(Type=Bone,Name="dyn_finger_3_3_l")
    Filter=(bX=True,bY=True,bZ=True)
    Parents=((Item=(Type=Bone,Name="thumb_03_l"),Weight=1))
    Weight=1
   )
   > &RigUnit_RotationConstraintLocalSpaceOffset(
    AdvancedSettings=(InterpolationType=Average,RotationOrderForFilter=XZY)
    Child=(Type=Bone,Name="dyn_finger_2_1_l")
    Filter=(bX=true,bY=true,bZ=true)
    Parents=((Item=(Type=Bone,Name="middle_01_l"),Weight=1))
    Weight=1
   )
   > &RigUnit_RotationConstraintLocalSpaceOffset(
    AdvancedSettings=(InterpolationType=Average,RotationOrderForFilter=XZY)
    Child=(Type=Bone,Name="dyn_finger_2_2_l")
    Filter=(bX=false,bY=true,bZ=false)
    Parents=((Item=(Type=Bone,Name="middle_02_l"),Weight=1))
    Weight=1
   )
   > &RigUnit_RotationConstraintLocalSpaceOffset(
    AdvancedSettings=(InterpolationType=Average,RotationOrderForFilter=XZY)
    Child=(Type=Bone,Name="dyn_finger_2_3_l")
    Filter=(bX=false,bY=True,bZ=false)
    Parents=((Item=(Type=Bone,Name="middle_03_l"),Weight=1))
    Weight=1
   )
   > &RigUnit_RotationConstraintLocalSpaceOffset(
    AdvancedSettings=(InterpolationType=Average,RotationOrderForFilter=XZY)
    Child=(Type=Bone,Name="dyn_finger_1_1_l")
    Filter=(bX=true,bY=true,bZ=true)
    Parents=((Item=(Type=Bone,Name="pinky_01_l"),Weight=1))
    Weight=1
   )
   > &RigUnit_RotationConstraintLocalSpaceOffset(
    AdvancedSettings=(InterpolationType=Average,RotationOrderForFilter=XZY)
    Child=(Type=Bone,Name="dyn_finger_1_2_l")
    Filter=(bX=false,bY=true,bZ=false)
    Parents=((Item=(Type=Bone,Name="pinky_02_l"),Weight=1))
    Weight=1
   )
   > n72 = &RigUnit_RotationConstraintLocalSpaceOffset(
    AdvancedSettings=(InterpolationType=Average,RotationOrderForFilter=XZY)
    Child=(Type=Bone,Name="dyn_finger_1_3_l")
    Filter=(bX=false,bY=True,bZ=false)
    Parents=((Item=(Type=Bone,Name="pinky_03_l"),Weight=1))
    Weight=1
   )
   n73 = &RigVMFunction_Sequence(ExecuteContext=@n72.ExecutePin)
   n74 = &RigUnit_SphericalPoseReader(
    ActiveRegionScaleFactors=(PositiveWidth=0.5,NegativeWidth=0.5,PositiveHeight=0.5,NegativeHeight=0.5)
    ActiveRegionSize=0.1
    Debug=(bDrawDebug=true,bDraw2D=False,bDrawLocalAxes=False,DebugScale=25,DebugSegments=20,DebugThickness=0.25)
    DriverAxis=(X=0,Y=1,Z=0)
    DriverItem=(Type=Bone,Name="hand_l")
    FalloffRegionScaleFactors=(PositiveWidth=1,NegativeWidth=1,PositiveHeight=1,NegativeHeight=1)
    FalloffSize=0.2
    FlipHeightScaling=False
    FlipWidthScaling=False
    OptionalParentItem=(Type=Bone,Name="None")
    RotationOffset=(X=125,Y=0,Z=30)
    ExecutePin=@n73.A
   )
   > &RigUnit_ModifyTransforms(
    ItemToModify=((Item=(Type=Bone,Name="dyn_finger_base_l"),Transform=(Rotation=(X=0,Y=0,Z=0,W=1),Translation=(X=0,Y=0,Z=0),Scale3D=(X=1,Y=1,Z=1))))
    ItemToModify.0.Transform.Rotation=&RigVMFunction_MathQuaternionFromEuler(Euler=(X=30,Y=-20,Z=0), RotationOrder=ZXY).Result
    Weight=@n74.OutputParam
   )
   n75 = &RigUnit_SphericalPoseReader(
    ActiveRegionScaleFactors=(PositiveWidth=0.5,NegativeWidth=0.5,PositiveHeight=0.5,NegativeHeight=0.5)
    ActiveRegionSize=0.1
    Debug=(bDrawDebug=true,bDraw2D=False,bDrawLocalAxes=False,DebugScale=25,DebugSegments=20,DebugThickness=0.25)
    DriverAxis=(X=0,Y=1,Z=0)
    DriverItem=(Type=Bone,Name="hand_l")
    FalloffRegionScaleFactors=(PositiveWidth=1,NegativeWidth=1,PositiveHeight=1,NegativeHeight=1)
    FalloffSize=0.2
    FlipHeightScaling=False
    FlipWidthScaling=False
    OptionalParentItem=(Type=Bone,Name="None")
    RotationOffset=(X=0,Y=35,Z=30)
    ExecutePin=@n73.B
   )
   > &RigUnit_ModifyTransforms(
    ItemToModify=((Item=(Type=Bone,Name="dyn_finger_base_l"),Transform=(Rotation=(X=0,Y=0,Z=0,W=1),Translation=(X=0,Y=0,Z=0),Scale3D=(X=1,Y=1,Z=1))))
    ItemToModify.0.Transform.Rotation=&RigVMFunction_MathQuaternionFromEuler(Euler=(X=10,Y=-30,Z=-60), RotationOrder=ZXY).Result
    Weight=@n75.OutputParam
   )
   n76 = &RigUnit_SphericalPoseReader(
    ActiveRegionScaleFactors=(PositiveWidth=0.5,NegativeWidth=0.5,PositiveHeight=0.5,NegativeHeight=0.5)
    ActiveRegionSize=0.1
    Debug=(bDrawDebug=true,bDraw2D=False,bDrawLocalAxes=False,DebugScale=25,DebugSegments=20,DebugThickness=0.25)
    DriverAxis=(X=0,Y=1,Z=0)
    DriverItem=(Type=Bone,Name="hand_r")
    ExecutePin=()
    FalloffRegionScaleFactors=(PositiveWidth=1,NegativeWidth=1,PositiveHeight=1,NegativeHeight=1)
    FalloffSize=0.3
    FlipHeightScaling=False
    FlipWidthScaling=False
    OptionalParentItem=(Type=Bone,Name="None")
    RotationOffset=(X=60,Y=45,Z=0)
   )
   > &RigUnit_ModifyTransforms(
    ItemToModify=((Item=(Type=Bone,Name="dyn_finger_base_r"),Transform=(Rotation=(X=0,Y=0,Z=0,W=1),Translation=(X=0,Y=0,Z=0),Scale3D=(X=1,Y=1,Z=1))))
    ItemToModify.0.Transform.Rotation=&RigVMFunction_MathQuaternionFromEuler(Euler=(X=0,Y=0,Z=0), RotationOrder=ZXY).Result
    Weight=@n76.OutputParam
   )
   RigVMDispatchNode(Index=@n4.Index, Array=RigVMVariableNode() {variable=Remap Fingers}.Value) {template="DISPATCH_RigVMDispatch_ArrayGetAtIndex(in Array,in Index,out Element)"}
   &RigUnit_ModifyTransforms(
    ExecutePin=()
    ItemToModify=((Item=(Type=Bone,Name="dyn_finger_base_l"),Transform=(Rotation=(X=0,Y=0,Z=0,W=1),Translation=(X=0,Y=0,Z=0),Scale3D=(X=1,Y=1,Z=1))))
    ItemToModify.0.Transform.Rotation=&RigVMFunction_MathQuaternionFromEuler(Euler=(X=-20,Y=-10,Z=0), RotationOrder=ZXY).Result
    Weight=&RigVMFunction_MathDoubleSub(A=1, B=@n31.Result).Result
   )
   &RigVMFunction_MathQuaternionFromEuler(Euler=(X=-75,Y=0,Z=90), RotationOrder=ZXY)
   &RigVMFunction_MathQuaternionFromEuler(Euler=(X=0,Y=-10,Z=-45), RotationOrder=ZXY)
   &RigVMFunction_MathQuaternionFromEuler(Euler=(X=0,Y=0,Z=0), RotationOrder=ZXY)
   RigVMFunction_Sequence_1_1_1_1 = RigVMAggregateNode(ExecuteContext=())
   &RigVMFunction_Sequence(ExecuteContext=())
   &RigUnit_RotationConstraintLocalSpaceOffset(
    AdvancedSettings=(InterpolationType=Average,RotationOrderForFilter=XZY)
    Child=(Type=Bone,Name="dyn_finger_base_l")
    ExecutePin=()
    Filter=(bX=True,bY=True,bZ=True)
    Parents=((Item=(Type=Bone,Name="hand_l"),Weight=1))
    Weight=1
   )
   RigVMVariableNode() {variable=FE_backpack}
  ]}
]
hierarchy [
 root {loc=<0 0 0>, rot=<0 0 0 1>}
 pelvis {parent=root, loc=<0 -0.7321 92.4218>, rot=<0.0127 0.707 -0.0127 -0.707>}
 spine_01 {parent=pelvis, loc=<5.6002 -0.5128 0>, rot=<0 0 0.034 -0.9994>}
 spine_02 {parent=spine_01, loc=<8.0753 0.0139 0>, rot=<0 0 0.0238 -0.9997>}
 spine_03 {parent=spine_02, loc=<7.0393 0 0>, rot=<0 0 0.0149 -0.9999>}
 spine_04 {parent=spine_03, loc=<8.1024 0 0>, rot=<0 0 -0.0132 -0.9999>}
 spine_05 {parent=spine_04, loc=<10.0146 0 0>, rot=<0 0 -0.0015 -1>}
 attach_backpack {parent=spine_05, loc=<-0.123 -15.1423 0>, rot=<-0.0093 -0.707 0.0093 -0.707>}
 clavicle_l {parent=spine_05, loc=<2.5856 -0.2866 -4.3777>, rot=<0.0465 -0.7289 -0.009 -0.6829>}
 upperarm_l {parent=clavicle_l, loc=<18.2264 0.1342 -0.001>, rot=<0.0253 -0.2693 0.0013 -0.9627>}
 lowerarm_l {parent=upperarm_l, loc=<29.0071 0.1686 -1.239>, rot=<0.0391 -0.0099 -0.2453 -0.9686>}
 hand_l {parent=lowerarm_l, loc=<25.8607 0.222 1.0274>, rot=<-0.7193 -0.0988 0.1193 -0.6773>}
 thumb_01_l {parent=hand_l, loc=<1.3118 -0.3191 -2.459>, rot=<0.6546 -0.4042 0.0664 -0.6354>}
 thumb_02_l {parent=thumb_01_l, loc=<5.9336 -0.0448 0.0401>, rot=<-0.0037 0.0213 0.298 -0.9543>}
 thumb_03_l {parent=thumb_02_l, loc=<2.2223 0.3682 -0.0907>, rot=<0 0 0 1>}
 index_metacarpal_l {parent=hand_l, loc=<1.6535 0.0183 -2.96>, rot=<0.0158 -0.0124 -0.0173 -0.9996>}
 index_01_l {parent=index_metacarpal_l, loc=<8.3008 0.0357 -0.0524>, rot=<-0.0044 0.0485 0.0924 -0.9945>}
 index_02_l {parent=index_01_l, loc=<3.3587 -0.3004 0.0452>, rot=<-0.002 0.0027 0.3225 -0.9466>}
 index_03_l {parent=index_02_l, loc=<2.5578 -0.123 -0.0252>, rot=<-0.002 0.0067 0.2558 -0.9667>}
 middle_metacarpal_l {parent=hand_l, loc=<2.1856 -0.1725 -0.5984>, rot=<0.006 0.0168 -0.0329 -0.9993>}
 middle_01_l {parent=middle_metacarpal_l, loc=<7.3275 -0.2447 0.0129>, rot=<-0.0208 0.0042 0.0858 -0.9961>}
 middle_02_l {parent=middle_01_l, loc=<4.0123 0.1401 0.1175>, rot=<-0.0218 0.0258 0.3512 -0.9357>}
 middle_03_l {parent=middle_02_l, loc=<2.6077 0.0534 0.1373>, rot=<0.0082 0.0188 0.1708 -0.9851>}
 ring_metacarpal_l {parent=hand_l, loc=<2.2074 -0.3625 1.3638>, rot=<-0.0284 0.0722 -0.0458 -0.9959>}
 ring_01_l {parent=ring_metacarpal_l, loc=<7.0295 -0.6197 -0.1446>, rot=<-0.0489 -0.0252 0.1892 -0.9804>}
 ring_02_l {parent=ring_01_l, loc=<3.6174 0.4647 -0.1027>, rot=<0 0 0.2806 -0.9598>}
 ring_03_l {parent=ring_02_l, loc=<2.4152 0.114 0.0199>, rot=<0 0 0.1189 -0.9929>}
 pinky_metacarpal_l {parent=hand_l, loc=<1.7616 -0.9753 3.3814>, rot=<-0.0397 0.0814 -0.069 -0.9935>}
 pinky_01_l {parent=pinky_metacarpal_l, loc=<7.1709 -0.4831 -0.1094>, rot=<-0.0654 -0.0035 0.2242 -0.9723>}
 pinky_02_l {parent=pinky_01_l, loc=<2.8274 0.0893 -0.1043>, rot=<0.0037 0.0011 0.2854 -0.9584>}
 pinky_03_l {parent=pinky_02_l, loc=<2.2046 0.0307 -0.0078>, rot=<-0.0002 0.0098 0.1445 -0.9895>}
 dyn_finger_base_l {parent=hand_l, loc=<0 0 0>, rot=<0 0 0 1>}
 dyn_finger_1_1_l {parent=dyn_finger_base_l, loc=<1.2197 3.3837 3.1165>, rot=<-0.2111 0.1239 -0.6087 -0.7547>}
 dyn_finger_1_2_l {parent=dyn_finger_1_1_l, loc=<4.5604 0 0>, rot=<0 0 0.3204 -0.9473>}
 dyn_finger_1_3_l {parent=dyn_finger_1_2_l, loc=<5.7175 0 0>, rot=<0 0 0 1>}
 dyn_finger_2_1_l {parent=dyn_finger_base_l, loc=<0.9836 3.7532 -1.87>, rot=<0.0928 -0.1711 -0.6685 -0.7178>}
 dyn_finger_2_2_l {parent=dyn_finger_2_1_l, loc=<4.5639 0 0>, rot=<0 0 0.3493 -0.937>}
 dyn_finger_2_3_l {parent=dyn_finger_2_2_l, loc=<5.7692 0 0>, rot=<0 0 0 1>}
 dyn_finger_3_1_l {parent=dyn_finger_base_l, loc=<3.6969 -2.2642 2.6768>, rot=<-0.0301 0.124 0.3121 -0.9414>}
 dyn_finger_3_2_l {parent=dyn_finger_3_1_l, loc=<4.2339 0 0>, rot=<0 0 -0.2697 -0.963>}
 dyn_finger_3_3_l {parent=dyn_finger_3_2_l, loc=<5.8019 0 0>, rot=<0 0 0 1>}
 dyn_finger_4_1_l {parent=dyn_finger_base_l, loc=<3.2101 -1.9681 -2.4892>, rot=<-0.1215 -0.2681 0.3448 -0.8913>}
 dyn_finger_4_2_l {parent=dyn_finger_4_1_l, loc=<4.415 0 0>, rot=<0 0 -0.3022 -0.9532>}
 dyn_finger_4_3_l {parent=dyn_finger_4_2_l, loc=<5.6852 0 0>, rot=<0 0 0 1>}
 weapon_l {parent=hand_l, loc=<-0.1974 -3.4111 0.3811>, rot=<0 0 0 1>}
 hand_attach_l {parent=hand_l, loc=<8.5 -3 0>, rot=<0.7071 0.7071 0 0>}
 lowerarm_twist_02_l {parent=lowerarm_l, loc=<9.6199 0 0>, rot=<0 0 0 -1>}
 lowerarm_twist_01_l {parent=lowerarm_l, loc=<17.8549 0 0>, rot=<0 0 0 -1>}
 upperarm_twist_01_l {parent=upperarm_l, loc=<7.9796 0.0843 -0.6195>, rot=<0 0 0 -1>}
 upperarm_twist_02_l {parent=upperarm_l, loc=<19.5059 0.0843 -0.6195>, rot=<0 0 0 -1>}
 dyn_shoulder_armor_l {parent=clavicle_l, loc=<7.9302 0 7.149>, rot=<0 0 0 1>}
 neck_01 {parent=spine_05, loc=<10.815 1.9196 0>, rot=<0 0 -0.1864 -0.9825>}
 neck_02 {parent=neck_01, loc=<4.7347 0.3045 0>, rot=<0 0 0.0363 -0.9993>}
 head {parent=neck_02, loc=<5.7632 0.5552 0>, rot=<0 0 0.115 -0.9934>}
 attach_cape {parent=spine_05, loc=<9.1509 -13.3783 0>, rot=<0.0616 -0.7044 -0.0616 -0.7044>}
 clavicle_r {parent=spine_05, loc=<2.5855 -0.2866 4.3777>, rot=<0.7289 0.0465 -0.6829 0.009>}
 upperarm_r {parent=clavicle_r, loc=<-18.2264 -0.1342 0.001>, rot=<0.0253 -0.2693 0.0013 -0.9627>}
 lowerarm_r {parent=upperarm_r, loc=<-29.0071 -0.1686 1.239>, rot=<0.0391 -0.0099 -0.2453 -0.9686>}
 hand_r {parent=lowerarm_r, loc=<-25.8607 -0.222 -1.0274>, rot=<-0.7193 -0.0988 0.1193 -0.6773>}
 thumb_01_r {parent=hand_r, loc=<-1.3118 0.3191 2.459>, rot=<0.6546 -0.4042 0.0664 -0.6354>}
 thumb_02_r {parent=thumb_01_r, loc=<-5.9336 0.0448 -0.0401>, rot=<-0.0037 0.0213 0.298 -0.9543>}
 thumb_03_r {parent=thumb_02_r, loc=<-2.2223 -0.3682 0.0907>, rot=<0.0118 -0.013 0.0066 -0.9998>}
 index_metacarpal_r {parent=hand_r, loc=<-1.6535 -0.0183 2.96>, rot=<0.0158 -0.0124 -0.0173 -0.9996>}
 index_01_r {parent=index_metacarpal_r, loc=<-8.3008 -0.0357 0.0524>, rot=<-0.0044 0.0485 0.0924 -0.9945>}
 index_02_r {parent=index_01_r, loc=<-3.3587 0.3004 -0.0452>, rot=<-0.002 0.0027 0.3225 -0.9466>}
 index_03_r {parent=index_02_r, loc=<-2.5578 0.123 0.0252>, rot=<-0.002 0.0067 0.2558 -0.9667>}
 middle_metacarpal_r {parent=hand_r, loc=<-2.1856 0.1725 0.5984>, rot=<0.006 0.0168 -0.0329 -0.9993>}
 middle_01_r {parent=middle_metacarpal_r, loc=<-7.3275 0.2447 -0.0129>, rot=<-0.0208 0.0042 0.0858 -0.9961>}
 middle_02_r {parent=middle_01_r, loc=<-4.0123 -0.1401 -0.1175>, rot=<-0.0218 0.0258 0.3512 -0.9357>}
 middle_03_r {parent=middle_02_r, loc=<-2.6077 -0.0534 -0.1373>, rot=<0.0082 0.0188 0.1708 -0.9851>}
 ring_metacarpal_r {parent=hand_r, loc=<-2.2074 0.3625 -1.3638>, rot=<-0.0284 0.0722 -0.0458 -0.9959>}
 ring_01_r {parent=ring_metacarpal_r, loc=<-7.0295 0.6197 0.1446>, rot=<-0.0489 -0.0252 0.1892 -0.9804>}
 ring_02_r {parent=ring_01_r, loc=<-3.6174 -0.4647 0.1027>, rot=<0 0 0.2806 -0.9598>}
 ring_03_r {parent=ring_02_r, loc=<-2.4152 -0.114 -0.0199>, rot=<0 0 0.1189 -0.9929>}
 pinky_metacarpal_r {parent=hand_r, loc=<-1.7616 0.9753 -3.3814>, rot=<-0.0397 0.0814 -0.069 -0.9935>}
 pinky_01_r {parent=pinky_metacarpal_r, loc=<-7.1709 0.4831 0.1094>, rot=<-0.0654 -0.0035 0.2242 -0.9723>}
 pinky_02_r {parent=pinky_01_r, loc=<-2.8274 -0.0893 0.1043>, rot=<0.0037 0.0011 0.2854 -0.9584>}
 pinky_03_r {parent=pinky_02_r, loc=<-2.2046 -0.0307 0.0078>, rot=<-0.0002 0.0098 0.1445 -0.9895>}
 dyn_finger_base_r {parent=hand_r, loc=<0 0 0>, rot=<0 0 0 1>}
 dyn_finger_1_1_r {parent=dyn_finger_base_r, loc=<-1.2194 -3.3839 -3.1164>, rot=<-0.2111 0.1239 -0.6087 -0.7547>}
 dyn_finger_1_2_r {parent=dyn_finger_1_1_r, loc=<-4.5605 0 0>, rot=<0 0 0.3204 -0.9473>}
 dyn_finger_1_3_r {parent=dyn_finger_1_2_r, loc=<-5.7178 0.0007 -0.0004>, rot=<0 0 0 1>}
 dyn_finger_2_1_r {parent=dyn_finger_base_r, loc=<-0.9837 -3.7531 1.87>, rot=<0.0928 -0.1711 -0.6685 -0.7178>}
 dyn_finger_2_2_r {parent=dyn_finger_2_1_r, loc=<-4.5638 0.0002 0>, rot=<0 0 0.3493 -0.937>}
 dyn_finger_2_3_r {parent=dyn_finger_2_2_r, loc=<-5.7692 -0.0001 0>, rot=<0 0 0 1>}
 dyn_finger_3_1_r {parent=dyn_finger_base_r, loc=<-3.6969 2.2642 -2.6768>, rot=<-0.0301 0.124 0.3121 -0.9414>}
 dyn_finger_3_2_r {parent=dyn_finger_3_1_r, loc=<-4.2338 0 0>, rot=<0 0 -0.2697 -0.963>}
 dyn_finger_3_3_r {parent=dyn_finger_3_2_r, loc=<-5.802 0 0>, rot=<0 0 0 1>}
 dyn_finger_4_1_r {parent=dyn_finger_base_r, loc=<-3.2101 1.968 2.4892>, rot=<-0.1215 -0.2681 0.3448 -0.8913>}
 dyn_finger_4_2_r {parent=dyn_finger_4_1_r, loc=<-4.415 0 0>, rot=<0 0 -0.3022 -0.9532>}
 dyn_finger_4_3_r {parent=dyn_finger_4_2_r, loc=<-5.6852 0 -0.0001>, rot=<0 0 0 1>}
 hand_attach_r {parent=hand_r, loc=<-8.5 3 0>, rot=<0 0 -0.7071 -0.7071>}
 weapon_r {parent=hand_r, loc=<0.1974 3.4111 -0.3811>, rot=<0 0 0 1>}
 lowerarm_twist_02_r {parent=lowerarm_r, loc=<-9.6199 0 0>, rot=<0 0 0 -1>}
 lowerarm_twist_01_r {parent=lowerarm_r, loc=<-17.8549 0 0>, rot=<0 0 0 -1>}
 upperarm_twist_01_r {parent=upperarm_r, loc=<-7.9796 -0.0843 0.6195>, rot=<0 0 0 -1>}
 upperarm_twist_02_r {parent=upperarm_r, loc=<-19.5059 -0.0843 0.6195>, rot=<0 0 0 -1>}
 dyn_shoulder_armor_r {parent=clavicle_r, loc=<-7.9302 0 -7.1488>, rot=<0 0 0 1>}
 spike_01_l {parent=spine_05, loc=<13.4968 -14.2103 -5>, rot=<-0.3007 0 0 -0.9537>}
 spike_01 {parent=spine_05, loc=<13.7604 -15.1569 0>, rot=<0 0 0 1>}
 spike_01_r {parent=spine_05, loc=<13.4968 -14.2103 5>, rot=<0.3007 0 0 -0.9537>}
 spike_02_l {parent=spine_04, loc=<16.263 -16.0919 -5>, rot=<-0.2588 -0.0004 -0.0014 -0.9659>}
 spike_02 {parent=spine_04, loc=<16.4007 -18.0597 0>, rot=<0 0 -0.0015 -1>}
 spike_02_r {parent=spine_04, loc=<16.263 -16.0919 5>, rot=<0.2588 0.0004 -0.0014 -0.9659>}
 spike_03_l {parent=spine_03, loc=<16.1061 -19.8373 -5.6>, rot=<-0.342 -0.005 -0.0138 -0.9396>}
 spike_03 {parent=spine_03, loc=<15.9353 -21.0605 0>, rot=<0 0 -0.0147 -0.9999>}
 spike_03_r {parent=spine_03, loc=<16.1061 -19.8373 5.6>, rot=<0.342 0.005 -0.0138 -0.9396>}
 spike_04_r {parent=spine_02, loc=<13.8689 -21.3188 5.6>, rot=<0.342 -0.0001 0.0002 -0.9397>}
 spike_04 {parent=spine_02, loc=<13.6618 -22.5364 0>, rot=<0 0 0.0002 -1>}
 spike_04_l {parent=spine_02, loc=<13.8689 -21.3188 -5.6>, rot=<-0.342 0.0001 0.0002 -0.9397>}
 thigh_l {parent=pelvis, loc=<-1.8716 -0.0109 -11.9936>, rot=<0.0012 0.0572 0.0155 -0.9982>}
 calf_l {parent=thigh_l, loc=<-41.9178 -0.2246 0.0926>, rot=<0.0006 -0.0161 -0.0391 -0.9991>}
 foot_l {parent=calf_l, loc=<-38.7057 1.2438 0.2452>, rot=<0.0975 -0.0474 0.0317 -0.9936>}
 ball_l {parent=foot_l, loc=<-6.1614 13.1469 0.6807>, rot=<0.0001 -0.0001 -0.7063 -0.7079>}
 dyn_ball_l_1 {parent=foot_l, loc=<-7.9295 4.1991 9.15>, rot=<-0.3084 0.3091 -0.6354 -0.6369>}
 dyn_ball_l_2 {parent=foot_l, loc=<-7.9193 8.5702 4.0947>, rot=<-0.1094 0.1096 -0.6978 -0.6994>}
 dyn_ball_l_3 {parent=foot_l, loc=<-7.9202 8.1428 -1.9487>, rot=<0.0783 -0.0785 -0.7019 -0.7036>}
 dyn_ball_l_4 {parent=foot_l, loc=<-7.9274 5.0638 -7.9342>, rot=<0.2783 -0.2789 -0.6492 -0.6507>}
 calf_twist_02_l {parent=calf_l, loc=<-15.9268 -0.0967 0.0397>, rot=<0.0237 0.0095 -0.0035 -0.9997>}
 calf_twist_01_l {parent=calf_l, loc=<-32.9335 -0.0967 0.0397>, rot=<0.0237 0.0095 -0.0035 -0.9997>}
 thigh_twist_01_l {parent=thigh_l, loc=<-21.6303 0 0>, rot=<0.0013 0 0 -1>}
 dyn_knee_l {parent=thigh_l, loc=<-41.9178 -0.2246 0.0926>, rot=<0.0006 -0.0161 -0.0391 -0.9991>}
 dyn_hip_l {parent=thigh_l, loc=<0 0 0>, rot=<0 0 0 1>}
 thigh_r {parent=pelvis, loc=<-1.8716 -0.0109 11.9936>, rot=<0.0572 -0.0012 0.9982 0.0155>}
 calf_r {parent=thigh_r, loc=<41.9178 0.2246 -0.0926>, rot=<0.0006 -0.0161 -0.0391 -0.9991>}
 foot_r {parent=calf_r, loc=<38.7057 -1.2438 -0.2452>, rot=<0.0975 -0.0474 0.0317 -0.9936>}
 ball_r {parent=foot_r, loc=<6.1614 -13.1469 -0.6807>, rot=<0.0001 -0.0001 -0.7063 -0.7079>}
 dyn_ball_r_4 {parent=foot_r, loc=<7.9274 -5.0638 7.9342>, rot=<0.2783 -0.2789 -0.6492 -0.6507>}
 dyn_ball_r_3 {parent=foot_r, loc=<7.9202 -8.1428 1.9487>, rot=<0.0783 -0.0785 -0.7019 -0.7036>}
 dyn_ball_r_2 {parent=foot_r, loc=<7.9193 -8.5702 -4.0947>, rot=<-0.1094 0.1096 -0.6978 -0.6994>}
 dyn_ball_r_1 {parent=foot_r, loc=<7.9295 -4.1991 -9.15>, rot=<-0.3084 0.3091 -0.6354 -0.6369>}
 calf_twist_02_r {parent=calf_r, loc=<15.9268 0.0967 -0.0397>, rot=<0.0237 0.0095 -0.0035 -0.9997>}
 calf_twist_01_r {parent=calf_r, loc=<32.9335 0.0967 -0.0397>, rot=<0.0237 0.0095 -0.0035 -0.9997>}
 thigh_twist_01_r {parent=thigh_r, loc=<21.6303 0 0>, rot=<0.0013 0 0 -1>}
 dyn_knee_r {parent=thigh_r, loc=<41.9178 0.2246 -0.0926>, rot=<0.0006 -0.0161 -0.0391 -0.9991>}
 dyn_hip_r {parent=thigh_r, loc=<0 0 0>, rot=<0 0 0 1>}
 tail1 {parent=pelvis, loc=<-5.4929 -11.4259 0>, rot=<0.2422 -0.6643 0.6643 -0.2422>}
 tail2 {parent=tail1, loc=<24.6384 0 0>, rot=<0 0.0793 0 -0.9968>}
 tail3 {parent=tail2, loc=<26.1413 0 0>, rot=<0 0.1348 0 -0.9909>}
 tail4 {parent=tail3, loc=<25.2556 0 0>, rot=<0 0 0 1>}
 dyn_pelvis {parent=pelvis, loc=<0 0 0>, rot=<0 0 0 1>}
 ik_hand_root {parent=root, loc=<0 0 0>, rot=<0 0 0 1>}
 ik_hand_gun {parent=ik_hand_root, loc=<-64.462 4.8638 101.1961>, rot=<0.6084 0.1793 0.4613 -0.6205>}
 ik_hand_r {parent=ik_hand_gun, loc=<0 0 0>, rot=<0 0 0 1>}
 ik_hand_l {parent=ik_hand_gun, loc=<65.7708 101.9258 43.6659>, rot=<-0.5102 -0.7906 -0.3387 0>}
]
variables [
 Enable Finger Retarget {type=bool, default=True}
 FE {type=double, default=0}
 FE_Idle {type=double, default=0}
 FE_backpack {type=double, default=0}
 Fingers {
  type="TArray<FRigElementKeyCollection>"
  default="((Keys=((Type=Bone,Name=\"index_01_l\"),(Type=Bone,Name=\"ring_02_l\"),(Type=Bone,Name=\"index_03_l\"))),(Keys=((Type=Bone,Name=\"middle_01_l\"),(Type=Bone,Name=\"ring_02_l\"),(Type=Bone,Name=\"middle_03_l\"))),(Keys=((Type=Bone,Name=\"ring_01_l\"),(Type=Bone,Name=\"ring ...(truncated)"
 }
 Fingers_Down {
  type="TArray<FRigElementKeyCollection>"
  default="((Keys=((Type=Bone,Name=\"ring_01_l\"),(Type=Bone,Name=\"pinky_02_l\"),(Type=Bone,Name=\"ring_03_l\"))),(Keys=((Type=Bone,Name=\"pinky_01_l\"),(Type=Bone,Name=\"pinky_02_l\"),(Type=Bone,Name=\"pinky_03_l\"))),(Keys=((Type=Bone,Name=\"ring_01_r\"),(Type=Bone,Name=\"ring_0 ...(truncated)"
 }
 Fingers_Up {
  type="TArray<FRigElementKeyCollection>"
  default="((Keys=((Type=Bone,Name=\"index_01_l\"),(Type=Bone,Name=\"index_02_l\"),(Type=Bone,Name=\"index_03_l\"))),(Keys=((Type=Bone,Name=\"middle_01_l\"),(Type=Bone,Name=\"middle_02_l\"),(Type=Bone,Name=\"middle_03_l\"))),(Keys=((Type=Bone,Name=\"index_01_r\"),(Type=Bone,Name=\" ...(truncated)"
 }
 Get Delta Ref_Result {type=FEulerTransform, default=(Location=(X=0,Y=0,Z=0),Rotation=(Pitch=0,Yaw=0,Roll=0),Scale=(X=1,Y=1,Z=1))}
 NewVar {type=bool, default=False}
 Remap Fingers {
  type="TArray<FRigElementKeyCollection>"
  default="((Keys=((Type=Bone,Name=\"dyn_finger_2_1_l\"),(Type=Bone,Name=\"dyn_finger_2_2_l\"),(Type=Bone,Name=\"dyn_finger_2_3_l\"))),(Keys=((Type=Bone,Name=\"dyn_finger_1_1_l\"),(Type=Bone,Name=\"dyn_finger_1_2_l\"),(Type=Bone,Name=\"dyn_finger_1_3_l\"))),(Keys=((Type=Bone,Nam ...(truncated)"
 }
 Remap Fingers_Down {
  type="TArray<FRigElementKeyCollection>"
  default="((Keys=((Type=Bone,Name=\"dyn_finger_3_1_l\"),(Type=Bone,Name=\"dyn_finger_3_2_l\"),(Type=Bone,Name=\"dyn_finger_3_3_l\"))),(Keys=((Type=Bone,Name=\"dyn_finger_4_1_l\"),(Type=Bone,Name=\"dyn_finger_4_2_l\"),(Type=Bone,Name=\"dyn_finger_4_3_l\"))),(Keys=((Type=Bone,Nam ...(truncated)"
 }
 Remap Fingers_Up {
  type="TArray<FRigElementKeyCollection>"
  default="((Keys=((Type=Bone,Name=\"dyn_finger_2_1_l\"),(Type=Bone,Name=\"dyn_finger_2_2_l\"),(Type=Bone,Name=\"dyn_finger_2_3_l\"))),(Keys=((Type=Bone,Name=\"dyn_finger_1_1_l\"),(Type=Bone,Name=\"dyn_finger_1_2_l\"),(Type=Bone,Name=\"dyn_finger_1_3_l\"))),(Keys=((Type=Bone,Nam ...(truncated)"
 }
 dyn_spike {type=double, default=0}
 dyn_spike_ButterPlate {type=double, default=0}
 is_crouch {type=double, default=0}
 is_fly {type=double, default=0}
]
