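# Post-process style AnimBlueprint for the M_LRG_ButterPlate character body, adding secondary motion and pose fixes to tail, spikes, fingers and armor.
# It copies the parent mesh pose into a Control Rig, then applies bone offsets.
# Two AnimDynamics tail chains read cached 'Presim', blending on emote state.
# Alpha-driven additive transforms then adjust for gliding, skydiving, sprinting, sliding and frontend poses, and shrink spikes under back bling.
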AnimBlueprint /BRCosmetics/Characters/Player/Male/Large/Bodies/M_LRG_ButterPlate/Meshes/M_LRG_ButterPlate_AnimBP
target_skeleton=&/M_LRG_ButterPlate_Skeleton
graphs [
 AnimGraph {nodes [
   n1 = Get Is Tactical Sprinting() -> (Is Tactical Sprinting:double)
   n2 = "Transform (Modify) Bone - Bone: tail4"(Alpha=@n1, Translation="0,0,0", Rotation="0,0,-20", Scale="1,1,1") {
    rotationMode=BMM_Additive
    scaleMode=BMM_Additive
    translationMode=BMM_Additive
    alphaBoolBlend {blendOption=Linear}
    alphaScaleBiasClamp {bInterpResult=true, interpSpeedDecreasing=1, interpSpeedIncreasing=1}
    boneToModify {boneName=tail4}
    rotation {roll=-20}
   }
   "Transform (Modify) Bone - Bone: tail3"(ComponentPose=@n2, Alpha=@n1, Translation="0,0,0", Rotation="0,0,-20", Scale="1,1,1") {
    rotationMode=BMM_Additive
    scaleMode=BMM_Additive
    translationMode=BMM_Additive
    alphaBoolBlend {blendOption=Linear}
    alphaScaleBiasClamp {bInterpResult=true, interpSpeedDecreasing=1, interpSpeedIncreasing=1}
    boneToModify {boneName=tail3}
    rotation {roll=-20}
   }
   n3 = Get In Vehicle Sim Alpha() -> (In Vehicle Sim Alpha:double)
   n4 = Anim Dynamics - Bone: tail1(ComponentPose=Local To Component(LocalPose=Use cached pose 'Presim'()), Alpha=@n3, ExternalForce="0,0,0") {
    angularDampingOverride=1
    angularSpringConstant=5000
    bAngularSpring=true
    bOverrideAngularDamping=true
    bUseGravityOverride=true
    gravityScale=0.35
    simulationSpace=RootRelative
    alphaBoolBlend {blendOption=Linear}
    alphaScaleBiasClamp {bInterpResult=true, interpSpeedDecreasing=1, interpSpeedIncreasing=1}
    boundBone {boneName=tail1}
    componentLinearAccScale=<1 1 1>
    componentLinearVelScale=<1 1 1>
    gravityOverride {z=-1500}
    physicsBodyDefinitions [{
      boundBone {boneName=tail1}
      boxExtents=<23 23 23>
      constraintSetup {angularLimitsMax=<0 30 30>, angularLimitsMin=<-30 -30 -30>, angularTarget=<1 0 0>}
      localJointOffset=<1 0 0>
     }]
   }
   n5 = Anim Dynamics - Bone: tail2(ComponentPose=@n4, Alpha=@n3, ExternalForce="0,0,0") {
    angularDampingOverride=1
    angularSpringConstant=5000
    bAngularSpring=true
    bOverrideAngularDamping=true
    bUseGravityOverride=true
    gravityScale=0.35
    simulationSpace=RootRelative
    alphaBoolBlend {blendOption=Linear}
    alphaScaleBiasClamp {bInterpResult=true, interpSpeedDecreasing=1, interpSpeedIncreasing=1}
    boundBone {boneName=tail2}
    componentLinearAccScale=<1 1 1>
    componentLinearVelScale=<1 1 1>
    gravityOverride {z=-1500}
    physicsBodyDefinitions [{
      boundBone {boneName=tail2}
      boxExtents=<23 23 23>
      constraintSetup {angularLimitsMax=<0 30 30>, angularLimitsMin=<-30 -30 -30>, angularTarget=<1 0 0>}
      localJointOffset=<1 0 0>
     }]
   }
   n6 = Anim Dynamics - Bone: tail3(ComponentPose=@n5, Alpha=@n3, ExternalForce="0,0,0") {
    angularDampingOverride=1
    angularSpringConstant=5000
    bAngularSpring=true
    bOverrideAngularDamping=true
    bUseGravityOverride=true
    gravityScale=0.35
    simulationSpace=RootRelative
    alphaBoolBlend {blendOption=Linear}
    alphaScaleBiasClamp {bInterpResult=true, interpSpeedDecreasing=1, interpSpeedIncreasing=1}
    boundBone {boneName=tail3}
    componentLinearAccScale=<1 1 1>
    componentLinearVelScale=<1 1 1>
    gravityOverride {z=-1500}
    physicsBodyDefinitions [{
      boundBone {boneName=tail3}
      boxExtents=<23 23 23>
      constraintSetup {angularLimitsMax=<0 30 30>, angularLimitsMin=<-30 -30 -30>, angularTarget=<1 0 0>}
      localJointOffset=<1 0 0>
     }]
   }
   n7 = Anim Dynamics - Bone: tail4(ComponentPose=@n6, Alpha=@n3, ExternalForce="0,0,0") {
    angularDampingOverride=1
    angularSpringConstant=5000
    bAngularSpring=true
    bOverrideAngularDamping=true
    bUseGravityOverride=true
    gravityScale=0.35
    simulationSpace=RootRelative
    alphaBoolBlend {blendOption=Linear}
    alphaScaleBiasClamp {bInterpResult=true, interpSpeedDecreasing=1, interpSpeedIncreasing=1}
    boundBone {boneName=tail4}
    componentLinearAccScale=<1 1 1>
    componentLinearVelScale=<1 1 1>
    gravityOverride {z=-1500}
    physicsBodyDefinitions [{
      boundBone {boneName=tail4}
      boxExtents=<23 23 23>
      constraintSetup {angularLimitsMax=<0 30 30>, angularLimitsMin=<-30 -30 -30>, angularTarget=<1 0 0>}
      localJointOffset=<1 0 0>
     }]
   }
   n8 = Get in vehicle scale alpha() -> (in vehicle scale alpha:double)
   n9 = "Transform (Modify) Bone - Bone: tail1"(ComponentPose=@n7, Alpha=@n8, Translation="0,0,0", Rotation="0,0,20", Scale="0.1,0.1,0.1") {
    rotationMode=BMM_Additive
    scaleMode=BMM_Additive
    translationMode=BMM_Additive
    alphaBoolBlend {blendOption=Linear}
    alphaScaleBiasClamp {bInterpResult=true, interpSpeedDecreasing=1, interpSpeedIncreasing=1}
    boneToModify {boneName=tail1}
    rotation {roll=20}
    scale=<0.1 0.1 0.1>
   }
   n10 = "Transform (Modify) Bone - Bone: tail2"(ComponentPose=@n9, Alpha=@n8, Translation="0,0,0", Rotation="0,0,20", Scale="0.1,0.1,0.1") {
    rotationMode=BMM_Additive
    scaleMode=BMM_Additive
    translationMode=BMM_Additive
    alphaBoolBlend {blendOption=Linear}
    alphaScaleBiasClamp {bInterpResult=true, interpSpeedDecreasing=1, interpSpeedIncreasing=1}
    boneToModify {boneName=tail2}
    rotation {roll=20}
    scale=<0.1 0.1 0.1>
   }
   n11 = "Transform (Modify) Bone - Bone: tail3"(ComponentPose=@n10, Alpha=@n8, Translation="0,0,0", Rotation="0,0,20", Scale="0.1,0.1,0.1") {
    rotationMode=BMM_Additive
    scaleMode=BMM_Additive
    translationMode=BMM_Additive
    alphaBoolBlend {blendOption=Linear}
    alphaScaleBiasClamp {bInterpResult=true, interpSpeedDecreasing=1, interpSpeedIncreasing=1}
    boneToModify {boneName=tail3}
    rotation {roll=20}
    scale=<0.1 0.1 0.1>
   }
   n12 = "Transform (Modify) Bone - Bone: tail4"(ComponentPose=@n11, Alpha=@n8, Translation="0,0,0", Rotation="0,0,20", Scale="0.1,0.1,0.1") {
    rotationMode=BMM_Additive
    scaleMode=BMM_Additive
    translationMode=BMM_Additive
    alphaBoolBlend {blendOption=Linear}
    alphaScaleBiasClamp {bInterpResult=true, interpSpeedDecreasing=1, interpSpeedIncreasing=1}
    boneToModify {boneName=tail4}
    rotation {roll=20}
    scale=<0.1 0.1 0.1>
   }
   n13 = Component To Local(ComponentPose=@n12)
   n14 = Get animDynAlpha() -> (animDynAlpha:double)
   n15 = Anim Dynamics - Bone: tail1(ComponentPose=Local To Component(LocalPose=Use cached pose 'Presim'()), Alpha=@n14, ExternalForce="0,0,0") {
    angularDampingOverride=1
    angularSpringConstant=5000
    bAngularSpring=true
    bOverrideAngularDamping=true
    bUseGravityOverride=true
    gravityScale=0.35
    simulationSpace=RootRelative
    alphaBoolBlend {blendOption=Linear}
    alphaScaleBiasClamp {bInterpResult=true, interpSpeedDecreasing=1, interpSpeedIncreasing=1}
    boundBone {boneName=tail1}
    componentLinearAccScale=<1 1 1>
    componentLinearVelScale=<1 1 1>
    gravityOverride {z=-1500}
    physicsBodyDefinitions [{
      boundBone {boneName=tail1}
      boxExtents=<23 23 23>
      constraintSetup {angularLimitsMax=<0 30 30>, angularLimitsMin=<-30 -30 -30>, angularTarget=<1 0 0>}
      localJointOffset=<1 0 0>
     }]
   }
   n16 = Anim Dynamics - Bone: tail2(ComponentPose=@n15, Alpha=@n14, ExternalForce="0,0,0") {
    angularDampingOverride=1
    angularSpringConstant=5000
    bAngularSpring=true
    bOverrideAngularDamping=true
    bUseGravityOverride=true
    gravityScale=0.35
    simulationSpace=RootRelative
    alphaBoolBlend {blendOption=Linear}
    alphaScaleBiasClamp {bInterpResult=true, interpSpeedDecreasing=1, interpSpeedIncreasing=1}
    boundBone {boneName=tail2}
    componentLinearAccScale=<1 1 1>
    componentLinearVelScale=<1 1 1>
    gravityOverride {z=-1500}
    physicsBodyDefinitions [{
      boundBone {boneName=tail2}
      boxExtents=<23 23 23>
      constraintSetup {angularLimitsMax=<0 30 30>, angularLimitsMin=<-30 -30 -30>, angularTarget=<1 0 0>}
      localJointOffset=<1 0 0>
     }]
   }
   n17 = Anim Dynamics - Bone: tail3(ComponentPose=@n16, Alpha=@n14, ExternalForce="0,0,0") {
    angularDampingOverride=1
    angularSpringConstant=5000
    bAngularSpring=true
    bOverrideAngularDamping=true
    bUseGravityOverride=true
    gravityScale=0.35
    simulationSpace=RootRelative
    alphaBoolBlend {blendOption=Linear}
    alphaScaleBiasClamp {bInterpResult=true, interpSpeedDecreasing=1, interpSpeedIncreasing=1}
    boundBone {boneName=tail3}
    componentLinearAccScale=<1 1 1>
    componentLinearVelScale=<1 1 1>
    gravityOverride {z=-1500}
    physicsBodyDefinitions [{
      boundBone {boneName=tail3}
      boxExtents=<23 23 23>
      constraintSetup {angularLimitsMax=<0 30 30>, angularLimitsMin=<-30 -30 -30>, angularTarget=<1 0 0>}
      localJointOffset=<1 0 0>
     }]
   }
   n18 = Anim Dynamics - Bone: tail4(ComponentPose=@n17, Alpha=@n14, ExternalForce="0,0,0") {
    angularDampingOverride=1
    angularSpringConstant=5000
    bAngularSpring=true
    bOverrideAngularDamping=true
    bUseGravityOverride=true
    gravityScale=0.35
    simulationSpace=RootRelative
    alphaBoolBlend {blendOption=Linear}
    alphaScaleBiasClamp {bInterpResult=true, interpSpeedDecreasing=1, interpSpeedIncreasing=1}
    boundBone {boneName=tail4}
    componentLinearAccScale=<1 1 1>
    componentLinearVelScale=<1 1 1>
    gravityOverride {z=-1500}
    physicsBodyDefinitions [{
      boundBone {boneName=tail4}
      boxExtents=<23 23 23>
      constraintSetup {angularLimitsMax=<0 30 30>, angularLimitsMin=<-30 -30 -30>, angularTarget=<1 0 0>}
      localJointOffset=<1 0 0>
     }]
   }
   n19 = Get animDynAlpha()
   n20 = "Transform (Modify) Bone - Bone: tail1"(ComponentPose=@n18, Alpha=@n19, Translation="0,0,0", Rotation="0,0,20", Scale="1,1,1") {
    rotationMode=BMM_Additive
    scaleMode=BMM_Additive
    translationMode=BMM_Additive
    alphaBoolBlend {blendOption=Linear}
    alphaScaleBiasClamp {bInterpResult=true, interpSpeedDecreasing=1, interpSpeedIncreasing=1}
    boneToModify {boneName=tail1}
    rotation {roll=20}
   }
   n21 = "Transform (Modify) Bone - Bone: tail2"(ComponentPose=@n20, Alpha=@n19, Translation="0,0,0", Rotation="0,0,20", Scale="1,1,1") {
    rotationMode=BMM_Additive
    scaleMode=BMM_Additive
    translationMode=BMM_Additive
    alphaBoolBlend {blendOption=Linear}
    alphaScaleBiasClamp {bInterpResult=true, interpSpeedDecreasing=1, interpSpeedIncreasing=1}
    boneToModify {boneName=tail2}
    rotation {roll=20}
   }
   n22 = "Transform (Modify) Bone - Bone: tail3"(ComponentPose=@n21, Alpha=@n19, Translation="0,0,0", Rotation="0,0,20", Scale="1,1,1") {
    rotationMode=BMM_Additive
    scaleMode=BMM_Additive
    translationMode=BMM_Additive
    alphaBoolBlend {blendOption=Linear}
    alphaScaleBiasClamp {bInterpResult=true, interpSpeedDecreasing=1, interpSpeedIncreasing=1}
    boneToModify {boneName=tail3}
    rotation {roll=20}
   }
   n23 = "Transform (Modify) Bone - Bone: tail4"(ComponentPose=@n22, Alpha=@n19, Translation="0,0,0", Rotation="0,0,20", Scale="1,1,1") {
    rotationMode=BMM_Additive
    scaleMode=BMM_Additive
    translationMode=BMM_Additive
    alphaBoolBlend {blendOption=Linear}
    alphaScaleBiasClamp {bInterpResult=true, interpSpeedDecreasing=1, interpSpeedIncreasing=1}
    boneToModify {boneName=tail4}
    rotation {roll=20}
   }
   n24 = Component To Local(ComponentPose=@n23)
   n25 = Get bIsPlayingEmote()
   n26 = Get IsHangGlider() -> (IsHangGlider:bool)
   n27 = Anim Dynamics - Bone: tail2!(
    ComponentPose=Local To Component(LocalPose=Blend Poses by bool(BlendPose_0=@n13, BlendPose_1=@n24, bActiveValue=@n25) {blendType=Linear})
    bAlphaBoolEnabled:bool=@n26
    ExternalForce="0,0,0"
   ) -> (Pose) {
    alphaInputType=Bool
    angularDampingOverride=1
    angularSpringConstant=5000
    bAngularSpring=true
    bOverrideAngularDamping=true
    bUseGravityOverride=true
    gravityScale=0.35
    simulationSpace=RootRelative
    alphaBoolBlend {blendOption=Linear}
    boundBone {boneName=tail2}
    componentLinearAccScale=<1 1 1>
    componentLinearVelScale=<1 1 1>
    gravityOverride {z=-1500}
    physicsBodyDefinitions [{
      boundBone {boneName=tail2}
      boxExtents=<30 30 30>
      constraintSetup {angularLimitsMax=<0 30 30>, angularLimitsMin=<-30 -30 -30>, angularTarget=<1 0 0>}
      localJointOffset=<1 0 0>
     }]
   }
   n28 = Anim Dynamics - Bone: tail3!(ComponentPose=@n27, bAlphaBoolEnabled:bool=@n26, ExternalForce="0,0,0") -> (Pose) {
    alphaInputType=Bool
    angularDampingOverride=1
    angularSpringConstant=5000
    bAngularSpring=true
    bOverrideAngularDamping=true
    bUseGravityOverride=true
    gravityScale=0.35
    simulationSpace=RootRelative
    alphaBoolBlend {blendOption=Linear}
    boundBone {boneName=tail3}
    componentLinearAccScale=<1 1 1>
    componentLinearVelScale=<1 1 1>
    gravityOverride {z=-1500}
    physicsBodyDefinitions [{
      boundBone {boneName=tail3}
      boxExtents=<30 30 30>
      constraintSetup {angularLimitsMax=<0 30 30>, angularLimitsMin=<-30 -30 -30>, angularTarget=<1 0 0>}
      localJointOffset=<1 0 0>
     }]
   }
   n29 = Anim Dynamics - Bone: tail4!(ComponentPose=@n28, bAlphaBoolEnabled:bool=@n26, ExternalForce="0,0,0") -> (Pose) {
    alphaInputType=Bool
    angularDampingOverride=1
    angularSpringConstant=5000
    bAngularSpring=true
    bOverrideAngularDamping=true
    bUseGravityOverride=true
    gravityScale=0.35
    simulationSpace=RootRelative
    alphaBoolBlend {blendOption=Linear}
    boundBone {boneName=tail4}
    componentLinearAccScale=<1 1 1>
    componentLinearVelScale=<1 1 1>
    gravityOverride {z=-1500}
    physicsBodyDefinitions [{
      boundBone {boneName=tail4}
      boxExtents=<30 30 30>
      constraintSetup {angularLimitsMax=<0 30 30>, angularLimitsMin=<-30 -30 -30>, angularTarget=<1 0 0>}
      localJointOffset=<1 0 0>
     }]
   }
   n30 = "Transform (Modify) Bone - Bone: tail1"!(ComponentPose=@n29, bAlphaBoolEnabled:bool=@n26, Translation="0,0,0", Rotation="0,0,30", Scale="1,1,1") -> (Pose) {
    alphaInputType=Bool
    rotationMode=BMM_Additive
    scaleMode=BMM_Additive
    translationMode=BMM_Additive
    alphaBoolBlend {blendOption=Linear}
    boneToModify {boneName=tail1}
    rotation {roll=30}
   }
   n31 = "Transform (Modify) Bone - Bone: tail2"!(ComponentPose=@n30, bAlphaBoolEnabled:bool=@n26, Translation="0,0,0", Rotation="0,0,10", Scale="1,1,1") -> (Pose) {
    alphaInputType=Bool
    rotationMode=BMM_Additive
    scaleMode=BMM_Additive
    translationMode=BMM_Additive
    alphaBoolBlend {blendOption=Linear}
    boneToModify {boneName=tail2}
    rotation {roll=10}
   }
   n32 = "Transform (Modify) Bone - Bone: tail3"!(ComponentPose=@n31, bAlphaBoolEnabled:bool=@n26, Translation="0,0,0", Rotation="0,0,10", Scale="1,1,1") -> (Pose) {
    alphaInputType=Bool
    rotationMode=BMM_Additive
    scaleMode=BMM_Additive
    translationMode=BMM_Additive
    alphaBoolBlend {blendOption=Linear}
    boneToModify {boneName=tail3}
    rotation {roll=10}
   }
   n33 = "Transform (Modify) Bone - Bone: tail4"!(ComponentPose=@n32, bAlphaBoolEnabled:bool=@n26, Translation="0,0,0", Rotation="0,0,10", Scale="1,1,1") -> (Pose) {
    alphaInputType=Bool
    rotationMode=BMM_Additive
    scaleMode=BMM_Additive
    translationMode=BMM_Additive
    alphaBoolBlend {blendOption=Linear}
    boneToModify {boneName=tail4}
    rotation {roll=10}
   }
   n34 = "Transform (Modify) Bone - Bone: tail1"!(ComponentPose=@n33, Alpha:float=@n1, Translation="0,0,20", Rotation="0,0,10", Scale="1,1,1") -> (Pose) {
    rotationMode=BMM_Additive
    scaleMode=BMM_Additive
    translationMode=BMM_Additive
    alphaBoolBlend {blendOption=Linear}
    alphaScaleBiasClamp {bInterpResult=true, interpSpeedDecreasing=1, interpSpeedIncreasing=1}
    boneToModify {boneName=tail1}
    rotation {roll=10}
    translation {z=20}
   }
   n35 = "Transform (Modify) Bone - Bone: dyn_pelvis"(ComponentPose=@n34, Alpha=@n1, Translation="0,0,5", Rotation="0,0,5", Scale="1,1,1") {
    rotationMode=BMM_Additive
    scaleMode=BMM_Additive
    translationMode=BMM_Additive
    alphaBoolBlend {blendOption=Linear}
    alphaScaleBiasClamp {bInterpResult=true, interpSpeedDecreasing=1, interpSpeedIncreasing=1}
    boneToModify {boneName=dyn_pelvis}
    rotation {roll=5}
    translation {z=5}
   }
   n36 = Get isSkyDiveDivingFloat() -> (isSkyDiveDivingFloat:double)
   n37 = "Transform (Modify) Bone - Bone: tail1"(
    ComponentPose=@n35
    Alpha=@n36
    Translation="0,0,0"
    Rotation=Get Dyn_tai1_Rotation_Sky_Diving() -> (Dyn_tai1_Rotation_Sky_Diving:Rotator)
    Scale="1,1,1"
   ) {
    rotationMode=BMM_Additive
    scaleMode=BMM_Additive
    translationMode=BMM_Additive
    alphaBoolBlend {blendOption=Linear}
    alphaScaleBiasClamp {bInterpResult=true, interpSpeedIncreasing=0.7}
    boneToModify {boneName=tail1}
   }
   n38 = "Transform (Modify) Bone - Bone: tail2"!(
    ComponentPose=@n37
    Alpha:float=@n36
    Translation="0,0,0"
    Rotation=Get Dyn_tai2_Rotation_Sky_Diving() -> (Dyn_tai2_Rotation_Sky_Diving:Rotator)
    Scale="1,1,1"
   ) -> (Pose) {
    rotationMode=BMM_Additive
    scaleMode=BMM_Additive
    translationMode=BMM_Additive
    alphaBoolBlend {blendOption=Linear}
    alphaScaleBiasClamp {bInterpResult=true, interpSpeedIncreasing=0.7}
    boneToModify {boneName=tail2}
   }
   n39 = "Transform (Modify) Bone - Bone: tail3"!(
    ComponentPose=@n38
    Alpha:float=@n36
    Translation="0,0,0"
    Rotation=Get Dyn_tai3_Rotation_Sky_Diving() -> (Dyn_tai3_Rotation_Sky_Diving:Rotator)
    Scale="1,1,1"
   ) -> (Pose) {
    rotationMode=BMM_Additive
    scaleMode=BMM_Additive
    translationMode=BMM_Additive
    alphaBoolBlend {blendOption=Linear}
    alphaScaleBiasClamp {bInterpResult=true, interpSpeedIncreasing=0.7}
    boneToModify {boneName=tail3}
   }
   n40 = "Transform (Modify) Bone - Bone: tail4"!(
    ComponentPose=@n39
    Alpha:float=@n36
    Translation="0,0,0"
    Rotation=Get Dyn_tai4_Rotation_Sky_Diving() -> (Dyn_tai4_Rotation_Sky_Diving:Rotator)
    Scale="1,1,1"
   ) -> (Pose) {
    rotationMode=BMM_Additive
    scaleMode=BMM_Additive
    translationMode=BMM_Additive
    alphaBoolBlend {blendOption=Linear}
    alphaScaleBiasClamp {bInterpResult=true, interpSpeedIncreasing=0.7}
    boneToModify {boneName=tail4}
   }
   n41 = Get Is Swim Sprinting() -> (Is Swim Sprinting:double)
   n42 = "Transform (Modify) Bone - Bone: tail1"(ComponentPose=@n40, Alpha=@n41, Translation="0,0,0", Rotation="0,0,-15", Scale="1,1,1") {
    rotationMode=BMM_Additive
    scaleMode=BMM_Additive
    translationMode=BMM_Additive
    alphaBoolBlend {blendOption=Linear}
    alphaScaleBiasClamp {bInterpResult=true, interpSpeedIncreasing=2}
    boneToModify {boneName=tail1}
    rotation {roll=-15}
   }
   n43 = "Transform (Modify) Bone - Bone: tail2"(ComponentPose=@n42, Alpha=@n41, Translation="0,0,0", Rotation="0,0,-15", Scale="1,1,1") {
    rotationMode=BMM_Additive
    scaleMode=BMM_Additive
    translationMode=BMM_Additive
    alphaBoolBlend {blendOption=Linear}
    alphaScaleBiasClamp {bInterpResult=true, interpSpeedIncreasing=2}
    boneToModify {boneName=tail2}
    rotation {roll=-15}
   }
   n44 = "Transform (Modify) Bone - Bone: tail3"(ComponentPose=@n43, Alpha=@n41, Translation="0,0,0", Rotation="0,0,-30", Scale="1,1,1") {
    rotationMode=BMM_Additive
    scaleMode=BMM_Additive
    translationMode=BMM_Additive
    alphaBoolBlend {blendOption=Linear}
    alphaScaleBiasClamp {bInterpResult=true, interpSpeedIncreasing=2}
    boneToModify {boneName=tail3}
    rotation {roll=-30}
   }
   n45 = "Transform (Modify) Bone - Bone: tail4"(ComponentPose=@n44, Alpha=@n41, Translation="0,0,0", Rotation="0,0,-30", Scale="1,1,1") {
    rotationMode=BMM_Additive
    scaleMode=BMM_Additive
    translationMode=BMM_Additive
    alphaBoolBlend {blendOption=Linear}
    alphaScaleBiasClamp {bInterpResult=true, interpSpeedIncreasing=2}
    boneToModify {boneName=tail4}
    rotation {roll=-30}
   }
   n46 = Get Is Knee Sliding Float() -> (Is Knee Sliding Float:double)
   n47 = "Constraint - tail1 (R)"(ComponentPose=@n45, Alpha=@n46, ConstraintWeights_0=1) {
    alphaBoolBlend {blendOption=Linear}
    boneToModify {boneName=tail1}
    constraintSetup [{transformType=Rotation, targetBone {boneName=root}}]
    constraintWeights [1]
   }
   n48 = "Transform (Modify) Bone - Bone: tail1"(ComponentPose=@n47, Alpha=@n46, Translation="0,0,0", Rotation="0,20,20", Scale="1,1,1") {
    rotationMode=BMM_Additive
    scaleMode=BMM_Additive
    translationMode=BMM_Additive
    alphaBoolBlend {blendOption=Linear}
    boneToModify {boneName=tail1}
    rotation {roll=20, yaw=20}
   }
   n49 = "Transform (Modify) Bone - Bone: dyn_pelvis"(ComponentPose=@n48, Alpha=@n46, Translation="0,0,1", Rotation="0,0,30", Scale="1,1,1") {
    rotationMode=BMM_Additive
    scaleMode=BMM_Additive
    translationMode=BMM_Additive
    alphaBoolBlend {blendOption=Linear}
    alphaScaleBiasClamp {bInterpResult=true}
    boneToModify {boneName=dyn_pelvis}
    rotation {roll=30}
    translation {z=1}
   }
   n50 = "Transform (Modify) Bone - Bone: tail2"(ComponentPose=@n49, Alpha=@n46, Translation="0,0,0", Rotation="0,0,30", Scale="1,1,1") {
    rotationMode=BMM_Additive
    scaleMode=BMM_Additive
    translationMode=BMM_Additive
    alphaBoolBlend {blendOption=Linear}
    alphaScaleBiasClamp {bInterpResult=true}
    boneToModify {boneName=tail2}
    rotation {roll=30}
   }
   n51 = "Transform (Modify) Bone - Bone: tail3"(ComponentPose=@n50, Alpha=@n46, Translation="0,0,0", Rotation="0,0,15", Scale="1,1,1") {
    rotationMode=BMM_Additive
    scaleMode=BMM_Additive
    translationMode=BMM_Additive
    alphaBoolBlend {blendOption=Linear}
    alphaScaleBiasClamp {bInterpResult=true}
    boneToModify {boneName=tail3}
    rotation {roll=15}
   }
   n52 = "Transform (Modify) Bone - Bone: tail4"(ComponentPose=@n51, Alpha=@n46, Translation="0,0,0", Rotation="0,0,15", Scale="1,1,1") {
    rotationMode=BMM_Additive
    scaleMode=BMM_Additive
    translationMode=BMM_Additive
    alphaBoolBlend {blendOption=Linear}
    alphaScaleBiasClamp {bInterpResult=true}
    boneToModify {boneName=tail4}
    rotation {roll=15}
   }
   n53 = Get Is Baseball Sliding float() -> (Is Baseball Sliding float:double)
   n54 = "Transform (Modify) Bone - Bone: tail1"(ComponentPose=@n52, Alpha=@n53, Translation="0,0,5", Rotation="0,0,60", Scale="1,1,1") {
    rotationMode=BMM_Additive
    scaleMode=BMM_Additive
    translationMode=BMM_Additive
    alphaBoolBlend {blendOption=Linear}
    alphaScaleBiasClamp {bInterpResult=true, interpSpeedIncreasing=3}
    boneToModify {boneName=tail1}
    rotation {roll=60}
    translation {z=5}
   }
   n55 = "Transform (Modify) Bone - Bone: dyn_pelvis"(ComponentPose=@n54, Alpha=@n53, Translation="0,0,5", Rotation="0,0,30", Scale="1,1,1") {
    rotationMode=BMM_Additive
    scaleMode=BMM_Additive
    translationMode=BMM_Additive
    alphaBoolBlend {blendOption=Linear}
    alphaScaleBiasClamp {bInterpResult=true, interpSpeedIncreasing=3}
    boneToModify {boneName=dyn_pelvis}
    rotation {roll=30}
    translation {z=5}
   }
   n56 = "Transform (Modify) Bone - Bone: tail2"(ComponentPose=@n55, Alpha=@n53, Translation="0,0,0", Rotation="0,0,30", Scale="1,1,1") {
    rotationMode=BMM_Additive
    scaleMode=BMM_Additive
    translationMode=BMM_Additive
    alphaBoolBlend {blendOption=Linear}
    alphaScaleBiasClamp {bInterpResult=true, interpSpeedIncreasing=3}
    boneToModify {boneName=tail2}
    rotation {roll=30}
   }
   n57 = "Transform (Modify) Bone - Bone: tail3"(ComponentPose=@n56, Alpha=@n53, Translation="0,0,0", Rotation="0,0,30", Scale="1,1,1") {
    rotationMode=BMM_Additive
    scaleMode=BMM_Additive
    translationMode=BMM_Additive
    alphaBoolBlend {blendOption=Linear}
    alphaScaleBiasClamp {bInterpResult=true, interpSpeedIncreasing=3}
    boneToModify {boneName=tail3}
    rotation {roll=30}
   }
   n58 = "Transform (Modify) Bone - Bone: tail4"(ComponentPose=@n57, Alpha=@n53, Translation="0,0,0", Rotation="0,0,30", Scale="1,1,1") {
    rotationMode=BMM_Additive
    scaleMode=BMM_Additive
    translationMode=BMM_Additive
    alphaBoolBlend {blendOption=Linear}
    alphaScaleBiasClamp {bInterpResult=true, interpSpeedIncreasing=3}
    boneToModify {boneName=tail4}
    rotation {roll=30}
   }
   n59 = Get isWearingBackbling()
   n60 = "Transform (Modify) Bone - Bone: spike_01"!(ComponentPose=@n58, bAlphaBoolEnabled:bool=@n59, Translation="0,0,0", Rotation="0,0,0", Scale="0.5,0.2,0.2") -> (Pose) {
    alphaInputType=Bool
    rotationMode=BMM_Additive
    scaleMode=BMM_Additive
    translationMode=BMM_Additive
    alphaBoolBlend {blendOption=Linear}
    alphaScaleBiasClamp {bInterpResult=true}
    boneToModify {boneName=spike_01}
    scale=<0.5 0.2 0.2>
   }
   n61 = "Transform (Modify) Bone - Bone: spike_01_l"!(ComponentPose=@n60, bAlphaBoolEnabled:bool=@n59, Translation="0,0,0", Rotation="0,0,0", Scale="0.5,0.2,0.2") -> (Pose) {
    alphaInputType=Bool
    rotationMode=BMM_Additive
    scaleMode=BMM_Additive
    translationMode=BMM_Additive
    alphaBoolBlend {blendOption=Linear}
    alphaScaleBiasClamp {bInterpResult=true}
    boneToModify {boneName=spike_01_l}
    scale=<0.5 0.2 0.2>
   }
   n62 = "Transform (Modify) Bone - Bone: spike_01_r"!(ComponentPose=@n61, bAlphaBoolEnabled:bool=@n59, Translation="0,0,0", Rotation="0,0,0", Scale="0.5,0.2,0.2") -> (Pose) {
    alphaInputType=Bool
    rotationMode=BMM_Additive
    scaleMode=BMM_Additive
    translationMode=BMM_Additive
    alphaBoolBlend {blendOption=Linear}
    alphaScaleBiasClamp {bInterpResult=true}
    boneToModify {boneName=spike_01_r}
    scale=<0.5 0.2 0.2>
   }
   n63 = "Transform (Modify) Bone - Bone: spike_02"!(ComponentPose=@n62, bAlphaBoolEnabled:bool=@n59, Translation="0,0,0", Rotation="0,0,0", Scale="0.2,0.2,0.2") -> (Pose) {
    alphaInputType=Bool
    rotationMode=BMM_Additive
    scaleMode=BMM_Additive
    translationMode=BMM_Additive
    alphaBoolBlend {blendOption=Linear}
    alphaScaleBiasClamp {bInterpResult=true}
    boneToModify {boneName=spike_02}
    scale=<0.2 0.2 0.2>
   }
   n64 = "Transform (Modify) Bone - Bone: spike_02_l"!(ComponentPose=@n63, bAlphaBoolEnabled:bool=@n59, Translation="0,0,0", Rotation="0,0,0", Scale="0.2,0.2,0.2") -> (Pose) {
    alphaInputType=Bool
    rotationMode=BMM_Additive
    scaleMode=BMM_Additive
    translationMode=BMM_Additive
    alphaBoolBlend {blendOption=Linear}
    alphaScaleBiasClamp {bInterpResult=true}
    boneToModify {boneName=spike_02_l}
    scale=<0.2 0.2 0.2>
   }
   n65 = "Transform (Modify) Bone - Bone: spike_02_r"!(ComponentPose=@n64, bAlphaBoolEnabled:bool=@n59, Translation="0,0,0", Rotation="0,0,0", Scale="0.2,0.2,0.2") -> (Pose) {
    alphaInputType=Bool
    rotationMode=BMM_Additive
    scaleMode=BMM_Additive
    translationMode=BMM_Additive
    alphaBoolBlend {blendOption=Linear}
    alphaScaleBiasClamp {bInterpResult=true}
    boneToModify {boneName=spike_02_r}
    scale=<0.2 0.2 0.2>
   }
   n66 = "Transform (Modify) Bone - Bone: spike_03"!(ComponentPose=@n65, bAlphaBoolEnabled:bool=@n59, Translation="0,0,0", Rotation="0,0,0", Scale="0.2,0.2,0.2") -> (Pose) {
    alphaInputType=Bool
    rotationMode=BMM_Additive
    scaleMode=BMM_Additive
    translationMode=BMM_Additive
    alphaBoolBlend {blendOption=Linear}
    alphaScaleBiasClamp {bInterpResult=true}
    boneToModify {boneName=spike_03}
    scale=<0.2 0.2 0.2>
   }
   n67 = "Transform (Modify) Bone - Bone: spike_03_l"!(ComponentPose=@n66, bAlphaBoolEnabled:bool=@n59, Translation="0,0,0", Rotation="0,0,0", Scale="0.2,0.2,0.2") -> (Pose) {
    alphaInputType=Bool
    rotationMode=BMM_Additive
    scaleMode=BMM_Additive
    translationMode=BMM_Additive
    alphaBoolBlend {blendOption=Linear}
    alphaScaleBiasClamp {bInterpResult=true}
    boneToModify {boneName=spike_03_l}
    scale=<0.2 0.2 0.2>
   }
   n68 = "Transform (Modify) Bone - Bone: spike_03_r"!(ComponentPose=@n67, bAlphaBoolEnabled:bool=@n59, Translation="0,0,0", Rotation="0,0,0", Scale="0.2,0.2,0.2") -> (Pose) {
    alphaInputType=Bool
    rotationMode=BMM_Additive
    scaleMode=BMM_Additive
    translationMode=BMM_Additive
    alphaBoolBlend {blendOption=Linear}
    alphaScaleBiasClamp {bInterpResult=true}
    boneToModify {boneName=spike_03_r}
    scale=<0.2 0.2 0.2>
   }
   n69 = "Transform (Modify) Bone - Bone: dyn_finger_1_1_l"!(ComponentPose=@n68, bAlphaBoolEnabled:bool=True, Translation="0,0,0", Rotation="0,0,0", Scale="0.65,1,1") -> (Pose) {
    alphaInputType=Bool
    rotationMode=BMM_Additive
    rotationSpace=BCS_BoneSpace
    scaleMode=BMM_Additive
    scaleSpace=BCS_BoneSpace
    translationMode=BMM_Additive
    translationSpace=BCS_BoneSpace
    alphaBoolBlend {blendOption=Linear}
    boneToModify {boneName=dyn_finger_1_1_l}
    scale {x=0.65}
   }
   n70 = "Transform (Modify) Bone - Bone: dyn_finger_2_1_l"!(ComponentPose=@n69, bAlphaBoolEnabled:bool=True, Translation="0,0,0", Rotation="0,0,0", Scale="0.65,1,1") -> (Pose) {
    alphaInputType=Bool
    rotationMode=BMM_Additive
    rotationSpace=BCS_BoneSpace
    scaleMode=BMM_Additive
    scaleSpace=BCS_BoneSpace
    translationMode=BMM_Additive
    translationSpace=BCS_BoneSpace
    alphaBoolBlend {blendOption=Linear}
    boneToModify {boneName=dyn_finger_2_1_l}
    scale {x=0.65}
   }
   n71 = "Transform (Modify) Bone - Bone: dyn_finger_3_1_l"!(ComponentPose=@n70, bAlphaBoolEnabled:bool=True, Translation="0,0,0", Rotation="0,0,0", Scale="0.65,1,1") -> (Pose) {
    alphaInputType=Bool
    rotationMode=BMM_Additive
    rotationSpace=BCS_BoneSpace
    scaleMode=BMM_Additive
    scaleSpace=BCS_BoneSpace
    translationMode=BMM_Additive
    translationSpace=BCS_BoneSpace
    alphaBoolBlend {blendOption=Linear}
    boneToModify {boneName=dyn_finger_3_1_l}
    scale {x=0.65}
   }
   n72 = "Transform (Modify) Bone - Bone: dyn_finger_4_1_l"!(ComponentPose=@n71, bAlphaBoolEnabled:bool=True, Translation="0,0,0", Rotation="0,0,0", Scale="0.65,1,1") -> (Pose) {
    alphaInputType=Bool
    rotationMode=BMM_Additive
    rotationSpace=BCS_BoneSpace
    scaleMode=BMM_Additive
    scaleSpace=BCS_BoneSpace
    translationMode=BMM_Additive
    translationSpace=BCS_BoneSpace
    alphaBoolBlend {blendOption=Linear}
    boneToModify {boneName=dyn_finger_4_1_l}
    scale {x=0.65}
   }
   n73 = "Transform (Modify) Bone - Bone: dyn_finger_base_l"!(ComponentPose=@n72, bAlphaBoolEnabled:bool=True, Translation="0,0,0", Rotation="0,20,0", Scale="0.92,0.92,0.92") -> (Pose) {
    alphaInputType=Bool
    rotationMode=BMM_Additive
    rotationSpace=BCS_BoneSpace
    scaleMode=BMM_Additive
    scaleSpace=BCS_BoneSpace
    alphaBoolBlend {blendOption=Linear}
    boneToModify {boneName=dyn_finger_base_l}
    rotation {yaw=20}
    scale=<0.92 0.92 0.92>
   }
   n74 = "Transform (Modify) Bone - Bone: dyn_finger_1_1_r"!(ComponentPose=@n73, bAlphaBoolEnabled:bool=True, Translation="0,0,0", Rotation="0,0,0", Scale="0.65,1,1") -> (Pose) {
    alphaInputType=Bool
    rotationMode=BMM_Additive
    rotationSpace=BCS_BoneSpace
    scaleMode=BMM_Additive
    scaleSpace=BCS_BoneSpace
    translationMode=BMM_Additive
    translationSpace=BCS_BoneSpace
    alphaBoolBlend {blendOption=Linear}
    boneToModify {boneName=dyn_finger_1_1_r}
    scale {x=0.65}
   }
   n75 = "Transform (Modify) Bone - Bone: dyn_finger_2_1_r"!(ComponentPose=@n74, bAlphaBoolEnabled:bool=True, Translation="0,0,0", Rotation="0,0,0", Scale="0.65,1,1") -> (Pose) {
    alphaInputType=Bool
    rotationMode=BMM_Additive
    rotationSpace=BCS_BoneSpace
    scaleMode=BMM_Additive
    scaleSpace=BCS_BoneSpace
    translationMode=BMM_Additive
    translationSpace=BCS_BoneSpace
    alphaBoolBlend {blendOption=Linear}
    boneToModify {boneName=dyn_finger_2_1_r}
    scale {x=0.65}
   }
   n76 = "Transform (Modify) Bone - Bone: dyn_finger_3_1_r"!(ComponentPose=@n75, bAlphaBoolEnabled:bool=True, Translation="0,0,0", Rotation="0,0,0", Scale="0.65,1,1") -> (Pose) {
    alphaInputType=Bool
    rotationMode=BMM_Additive
    rotationSpace=BCS_BoneSpace
    scaleMode=BMM_Additive
    scaleSpace=BCS_BoneSpace
    translationMode=BMM_Additive
    translationSpace=BCS_BoneSpace
    alphaBoolBlend {blendOption=Linear}
    boneToModify {boneName=dyn_finger_3_1_r}
    scale {x=0.65}
   }
   n77 = "Transform (Modify) Bone - Bone: dyn_finger_4_1_r"!(ComponentPose=@n76, bAlphaBoolEnabled:bool=True, Translation="0,0,0", Rotation="0,0,0", Scale="0.65,1,1") -> (Pose) {
    alphaInputType=Bool
    rotationMode=BMM_Additive
    rotationSpace=BCS_BoneSpace
    scaleMode=BMM_Additive
    scaleSpace=BCS_BoneSpace
    translationMode=BMM_Additive
    translationSpace=BCS_BoneSpace
    alphaBoolBlend {blendOption=Linear}
    boneToModify {boneName=dyn_finger_4_1_r}
    scale {x=0.65}
   }
   n78 = "Transform (Modify) Bone - Bone: dyn_finger_base_r"!(ComponentPose=@n77, bAlphaBoolEnabled:bool=True, Translation="0,0,0", Rotation="0,0,0", Scale="0.92,0.92,0.92") -> (Pose) {
    alphaInputType=Bool
    rotationMode=BMM_Additive
    rotationSpace=BCS_BoneSpace
    scaleMode=BMM_Additive
    scaleSpace=BCS_BoneSpace
    alphaBoolBlend {blendOption=Linear}
    boneToModify {boneName=dyn_finger_base_r}
    scale=<0.92 0.92 0.92>
   }
   n79 = Get Frontend_Backpack_Pose() -> (Frontend_Backpack_Pose:float)
   n80 = "Transform (Modify) Bone - Bone: spine_05"(ComponentPose=@n78, Alpha=@n79, Translation="0,0,0", Rotation="0,0,5", Scale="1,1,1") {rotationMode=BMM_Additive, alphaBoolBlend {blendOption=Linear}, boneToModify {boneName=spine_05}, rotation {roll=5}}
   n81 = "Transform (Modify) Bone - Bone: spine_04"(ComponentPose=@n80, Alpha=@n79, Translation="0,0,0", Rotation="0,0,5", Scale="1,1,1") {rotationMode=BMM_Additive, alphaBoolBlend {blendOption=Linear}, boneToModify {boneName=spine_04}, rotation {roll=5}}
   n82 = "Transform (Modify) Bone - Bone: spine_03"(ComponentPose=@n81, Alpha=@n79, Translation="0,0,0", Rotation="3,0,10", Scale="1,1,1") {rotationMode=BMM_Additive, alphaBoolBlend {blendOption=Linear}, boneToModify {boneName=spine_03}, rotation {pitch=3, roll=10}}
   n83 = "Transform (Modify) Bone - Bone: spine_02"(ComponentPose=@n82, Alpha=@n79, Translation="0,0,0", Rotation="3,0,10", Scale="1,1,1") {rotationMode=BMM_Additive, alphaBoolBlend {blendOption=Linear}, boneToModify {boneName=spine_02}, rotation {pitch=3, roll=10}}
   n84 = "Transform (Modify) Bone - Bone: spine_01"(ComponentPose=@n83, Alpha=@n79, Translation="0,0,0", Rotation="6,0,10", Scale="1,1,1") {rotationMode=BMM_Additive, alphaBoolBlend {blendOption=Linear}, boneToModify {boneName=spine_01}, rotation {pitch=6, roll=10}}
   n85 = "Transform (Modify) Bone - Bone: upperarm_l"(ComponentPose=@n84, Alpha=@n79, Translation="0,0,0", Rotation="-5,0,-25", Scale="1,1,1") {rotationMode=BMM_Additive, alphaBoolBlend {blendOption=Linear}, boneToModify {boneName=upperarm_l}, rotation {pitch=-5, roll=-25}}
   n86 = "Transform (Modify) Bone - Bone: upperarm_r"(ComponentPose=@n85, Alpha=@n79, Translation="0,0,0", Rotation="-5,0,-25", Scale="1,1,1") {rotationMode=BMM_Additive, alphaBoolBlend {blendOption=Linear}, boneToModify {boneName=upperarm_r}, rotation {pitch=-5, roll=-25}}
   n87 = "Transform (Modify) Bone - Bone: head"(ComponentPose=@n86, Alpha=@n79, Translation="0,0,0", Rotation="0,0,-30", Scale="1,1,1") {rotationMode=BMM_Additive, alphaBoolBlend {blendOption=Linear}, boneToModify {boneName=head}, rotation {roll=-30}}
   n88 = "Transform (Modify) Bone - Bone: neck_01"(ComponentPose=@n87, Alpha=@n79, Translation="0,0,0", Rotation="-5.113,15,-10", Scale="1,1,1") {rotationMode=BMM_Additive, alphaBoolBlend {blendOption=Linear}, boneToModify {boneName=neck_01}, rotation=<rot -5.113 15 -10>}
   Output Pose(Result=Component To Local(ComponentPose=@n88)) {layerGroup=None}
   n89 = Control Rig!(
    Source=Copy Pose From Mesh() {bCopyCurves=true, bUseAttachedParent=true}
    Alpha
    dyn_spike:double=Get Dyn_Alpha_Spike() -> (Dyn_Alpha_Spike:float)
    is_crouch:double=Get Dyn_is_crouch() -> (Dyn_is_crouch:double)
    dyn_spike_ButterPlate:double=Get Dyn Alpha Spike Butter Plate() -> (Dyn Alpha Spike Butter Plate:double)
    is_fly:double=Get Dyn Alfa Is Fly() -> (Dyn Alfa Is Fly:double)
    FE:double=Get FE_Alpha() -> (FE_Alpha:double)
    FE_backpack:double=Get Frontend_Backpack_Pose()
    FE_Idle:double=Get FE_Idle_Alpha() -> (FE_Idle_Alpha:double)
   ) -> (Pose) {
    alphaBoolBlend {blendOption=Linear}
    controlRigAssetReference {blueprintRigClass=&/M_LRG_ButterPlate_CtrlRig.M_LRG_ButterPlate_CtrlRig_C}
   }
   n90 = Get tail_cape_begin() -> (tail_cape_begin:float)
   n91 = "Transform (Modify) Bone - Bone: tail1"(ComponentPose=Local To Component(LocalPose=@n89), Alpha=@n90, Translation="0,0,0", Rotation="0,0,-40", Scale="1,1,1") {
    rotationMode=BMM_Additive
    alphaBoolBlend {blendOption=Linear}
    alphaScaleBiasClamp {bInterpResult=true}
    boneToModify {boneName=tail1}
    rotation {roll=-40}
   }
   n92 = "Transform (Modify) Bone - Bone: tail2"(ComponentPose=@n91, Alpha=@n90, Translation="0,0,0", Rotation="0,0,-30", Scale="1,1,1") {
    rotationMode=BMM_Additive
    alphaBoolBlend {blendOption=Linear}
    alphaScaleBiasClamp {bInterpResult=true}
    boneToModify {boneName=tail2}
    rotation {roll=-30}
   }
   n93 = "Transform (Modify) Bone - Bone: tail3"(ComponentPose=@n92, Alpha=@n90, Translation="0,0,0", Rotation="0,0,-20", Scale="1,1,1") {
    rotationMode=BMM_Additive
    alphaBoolBlend {blendOption=Linear}
    alphaScaleBiasClamp {bInterpResult=true}
    boneToModify {boneName=tail3}
    rotation {roll=-20}
   }
   n94 = "Transform (Modify) Bone - Bone: tail1"(ComponentPose=@n93, Alpha=1, Translation="0,0,0", Rotation="0,0,-40", Scale="1,1,1") {
    rotationMode=BMM_Additive
    alphaBoolBlend {blendOption=Linear}
    alphaScaleBiasClamp {bInterpResult=true, interpSpeedDecreasing=4, interpSpeedIncreasing=4}
    boneToModify {boneName=tail1}
    rotation {roll=-40}
   }
   n95 = "Transform (Modify) Bone - Bone: tail2"(ComponentPose=@n94, Alpha=1, Translation="0,0,0", Rotation="0,0,-30", Scale="1,1,1") {
    rotationMode=BMM_Additive
    alphaBoolBlend {blendOption=Linear}
    alphaScaleBiasClamp {bInterpResult=true, interpSpeedDecreasing=4, interpSpeedIncreasing=4}
    boneToModify {boneName=tail2}
    rotation {roll=-30}
   }
   n96 = "Transform (Modify) Bone - Bone: tail3"(ComponentPose=@n95, Alpha=1, Translation="0,0,0", Rotation="0,0,-20", Scale="1,1,1") {
    rotationMode=BMM_Additive
    alphaBoolBlend {blendOption=Linear}
    alphaScaleBiasClamp {bInterpResult=true, interpSpeedDecreasing=4, interpSpeedIncreasing=4}
    boneToModify {boneName=tail3}
    rotation {roll=-20}
   }
   n97 = "Transform (Modify) Bone - Bone: tail1"(ComponentPose=@n96, Alpha=1, Translation="0,0,0", Rotation="0,0,-5", Scale="1,1,1") {
    rotationMode=BMM_Additive
    alphaBoolBlend {blendOption=Linear}
    alphaScaleBiasClamp {bInterpResult=true, interpSpeedDecreasing=4, interpSpeedIncreasing=4}
    boneToModify {boneName=tail1}
    rotation {roll=-5}
   }
   n98 = "Transform (Modify) Bone - Bone: tail2"(ComponentPose=@n97, Alpha=1, Translation="0,0,0", Rotation="0,0,-5", Scale="1,1,1") {
    rotationMode=BMM_Additive
    alphaBoolBlend {blendOption=Linear}
    alphaScaleBiasClamp {bInterpResult=true, interpSpeedDecreasing=4, interpSpeedIncreasing=4}
    boneToModify {boneName=tail2}
    rotation {roll=-5}
   }
   n99 = "Transform (Modify) Bone - Bone: tail3"(ComponentPose=@n98, Alpha=1, Translation="0,0,0", Rotation="0,0,5", Scale="1,1,1") {
    rotationMode=BMM_Additive
    alphaBoolBlend {blendOption=Linear}
    alphaScaleBiasClamp {bInterpResult=true, interpSpeedDecreasing=4, interpSpeedIncreasing=4}
    boneToModify {boneName=tail3}
    rotation {roll=5}
   }
   n100 = Get Dyn_is_parachute() -> (Dyn_is_parachute:double)
   n101 = "Transform (Modify) Bone - Bone: dyn_shoulder_armor_l"(ComponentPose=@n99, Alpha=@n100, Translation="2,0,2", Rotation="-20,0,0", Scale="1,1,1") {
    rotationMode=BMM_Additive
    scaleMode=BMM_Additive
    translationMode=BMM_Additive
    alphaBoolBlend {blendOption=Linear}
    boneToModify {boneName=dyn_shoulder_armor_l}
    rotation {pitch=-20}
    translation {x=2, z=2}
   }
   n102 = "Transform (Modify) Bone - Bone: dyn_shoulder_armor_r"(ComponentPose=@n101, Alpha=@n100, Translation="-2,0,2", Rotation="0,0,0", Scale="1,1,1") {
    rotationMode=BMM_Additive
    scaleMode=BMM_Additive
    translationMode=BMM_Additive
    alphaBoolBlend {blendOption=Linear}
    boneToModify {boneName=dyn_shoulder_armor_r}
    translation {x=-2, z=2}
   }
   Save cached pose 'Presim'(Pose=Component To Local(ComponentPose=@n102)) {cacheName=Presim}
  ]}
 EventGraph {nodes [
   Event BlueprintUpdateAnimation()
   > n1 = Sequence!(execute) -> (then_0, then_1, then_2:Exec, then_3:Exec, then_4:Exec, then_5:Exec, then_6:Exec, then_7:Exec, then_8:Exec)
   > n2 = Sequence!(execute) -> (then_0, then_1, then_2, then_3, then_4, then_5)
   > Set Is Knee Sliding Float(Is Knee Sliding Float:double="To Float (Boolean)"(InBool=Get bIsKneeSliding())) -> (Output_Get:double)
   > Set Is Baseball Sliding float(Is Baseball Sliding float:double="To Float (Boolean)"(InBool=Get bIsBaseballSliding())) -> (Output_Get:double)
   Set Is Surface Swimming(execute=@n2.then_1, Is Surface Swimming:double="To Float (Boolean)"(InBool=Get bIsSurfaceSwimming())) -> (Output_Get:double)
   > Set Is Swim Sprinting(Is Swim Sprinting:double="To Float (Boolean)"(InBool=Get bIsSwimSprinting() -> (bIsSwimSprinting:bool))) -> (Output_Get:double)
   Set Is Tactical Sprinting(execute=@n2.then_2, Is Tactical Sprinting:double="To Float (Boolean)"(InBool=Get bIsTacticalSprinting())) -> (Output_Get:double)
   n3 = Branch(execute=@n2.then_3, Condition=Get bIsSkydiveDiveMode())
   > Set isSkyDiveDivingFloat(isSkyDiveDivingFloat:double=1) -> (Output_Get:double)
   > Set Dyn_tai1_Rotation_Sky_Diving(Dyn_tai1_Rotation_Sky_Diving:Rotator="0,0,-15") -> (Output_Get:Rotator)
   > Set Dyn_tai2_Rotation_Sky_Diving(Dyn_tai2_Rotation_Sky_Diving:Rotator="0,0,-15") -> (Output_Get:Rotator)
   > Set Dyn_tai3_Rotation_Sky_Diving(Dyn_tai3_Rotation_Sky_Diving:Rotator="0,0,-15") -> (Output_Get:Rotator)
   > Set Dyn_tai4_Rotation_Sky_Diving(Dyn_tai4_Rotation_Sky_Diving:Rotator="0,0,-30") -> (Output_Get:Rotator)
   n4 = Branch(execute=@n3.else, Condition=Get bIsSkydiveFloating())
   > Set isSkyDiveDivingFloat(isSkyDiveDivingFloat=1)
   > Set Dyn_tai1_Rotation_Sky_Diving(Dyn_tai1_Rotation_Sky_Diving="0,0,-10")
   > Set Dyn_tai2_Rotation_Sky_Diving(Dyn_tai2_Rotation_Sky_Diving="0,0,-20")
   > Set Dyn_tai3_Rotation_Sky_Diving(Dyn_tai3_Rotation_Sky_Diving="0,0,-30")
   > Set Dyn_tai4_Rotation_Sky_Diving(Dyn_tai4_Rotation_Sky_Diving="0,0,-30")
   Set isSkyDiveDivingFloat(execute=@n4.else, isSkyDiveDivingFloat=0)
   > Set Dyn_tai1_Rotation_Sky_Diving(Dyn_tai1_Rotation_Sky_Diving="0,0,-15")
   > Set Dyn_tai2_Rotation_Sky_Diving(Dyn_tai2_Rotation_Sky_Diving="0,0,-15")
   > Set Dyn_tai3_Rotation_Sky_Diving(Dyn_tai3_Rotation_Sky_Diving="0,0,-30")
   > Set Dyn_tai4_Rotation_Sky_Diving(Dyn_tai4_Rotation_Sky_Diving="0,0,-30")
   n5 = Set pawnVelocitySpeed(execute=@n2.then_4, pawnVelocitySpeed:double=Vector Length(A=Get PawnVelocity())) -> (Output_Get:double)
   n6 = Get bIsSkyDiving()
   n7 = SelectFloat(A=0, B=1, bPickA=AND Boolean(A=NOT Boolean(A=@n6), B="float > float"(A=@n5, B=2000) {is=title K2Node_CallFunction}))
   > n8 = Set pawnVelocitySpeedCutoff(pawnVelocitySpeedCutoff:double=@n7) -> (Output_Get:double)
   n9 = Branch(execute=@n2.then_5, Condition=Get bIsSliding())
   > Set animDynAlpha(animDynAlpha:double=float * float(A=@n8, B=0.1) {is=title K2Node_PromotableOperator}) -> (Output_Get:double)
   n10 = Branch(execute=@n9.else, Condition=Get IsHangGlider())
   > Set animDynAlpha(animDynAlpha=float * float(A=@n8, B=0.1))
   n11 = float * float(A=@n7, B=1)
   Set animDynAlpha(execute=@n10.else, animDynAlpha=@n11)
   > n12 = Branch(Condition=Get Dyn_Alpha_Spike_Emote() -> (Dyn_Alpha_Spike_Emote:bool))
   > Set animDynAlpha(animDynAlpha=0.5)
   Set animDynAlpha(execute=@n12.else, animDynAlpha=@n11)
   > n13 = Branch(Condition=Get bIsSkydiveFloating())
   > Set animDynAlpha(animDynAlpha=0.66)
   Set animDynAlpha(execute=@n13.else, animDynAlpha=@n11)
   n14 = Sequence!(execute=@n1.then_1) -> (then_0, then_1)
   n15 = CheckForBackbling()
   n16 = Branch(execute=@n14.then_1, Condition=OR Boolean!(A=@n15, B=@n15.bIsBackblingACape, C:bool=@n15.bIsBackblingLong) -> (ReturnValue))
   > Set Dyn_Alpha_Spike(Dyn_Alpha_Spike:float=1) -> (Output_Get:float)
   > n17 = Set NewVar(NewVar:bool=BackpackItemDefinitionHasAnyMetaTag!(Tags:Array of Gameplay Tag Structures=Make Array("[0]"=(TagName="Cosmetics.MetaTags.ButterPlate")), bRequiresEquipped:bool=false) -> (ReturnValue:bool)) -> (Output_Get:bool)
   > n18 = Branch(Condition=@n17)
   > Set Dyn Alpha Spike Butter Plate(Dyn Alpha Spike Butter Plate:double=0) -> (Output_Get:double)
   Set Dyn Alpha Spike Butter Plate(execute=@n18.else, Dyn Alpha Spike Butter Plate=1)
   Set Dyn_Alpha_Spike(execute=@n16.else, Dyn_Alpha_Spike=0)
   n19 = Branch(execute=@n1.then_2, Condition=Get bIsSkyDiving())
   > Set Dyn Alfa Is Fly(Dyn Alfa Is Fly:double=1) -> (Output_Get:double)
   > n20 = Branch(Condition=Get bIsParachuteOpened())
   > Set Dyn_is_parachute(Dyn_is_parachute:double=1) -> (Output_Get:double)
   Set Dyn_is_parachute(execute=@n20.else, Dyn_is_parachute=0)
   Set Dyn Alfa Is Fly(execute=@n19.else, Dyn Alfa Is Fly=0)
   n21 = Branch(execute=@n1.then_3, Condition=Get bIsCrouching())
   > Set Dyn_is_crouch(Dyn_is_crouch:double=1) -> (Output_Get:double)
   Set Dyn_is_crouch(execute=@n21.else, Dyn_is_crouch=0)
   n22 = Sequence!(execute=@n1.then_4) -> (then_0, then_1, then_2:Exec)
   n23 = GetCurveValue(CurveName=Frontend_Backpack_Pose)
   n24 = GetOwnerFrontEndAnimInstance()
   n25 = GetCurveValue(self=@n24, CurveName=emote_BuiltIn)
   > Set FE_Alpha(FE_Alpha:double="Clamp (Float)"(Value=float + float(A=@n25, B=@n23), Min=0, Max=1)) -> (Output_Get:double)
   > Set FE_Idle_Alpha(FE_Idle_Alpha:double=GetCurveValue(self=@n24, CurveName=ButterPlate_FE_Idle)) -> (Output_Get:double)
   n26 = GetCurrentWeapon!(self:Fort Pawn Object Reference=TryGetFortPawn()) -> (ReturnValue:Fort Weapon Object Reference)
   n27 = Branch(execute=@n22.then_2, Condition=IsCosmeticObjectDisplayedAsPreview!(CosmeticObject:Object Reference=@n26) -> (ReturnValue:bool))
   > Set Frontend Pickaxe Alpha(Frontend Pickaxe Alpha:double=1) -> (Output_Get:double)
   Set Frontend Pickaxe Alpha(execute=@n27.else, Frontend Pickaxe Alpha=0)
   n28 = NOT Boolean(A=Get bIsPlayingEmote())
   n29 = CheckForBackbling(LongBackblingThreshold=20)
   n30 = OR Boolean!(A=@n29.bIsBackblingACape, B=@n29.bIsBackblingLong) -> (ReturnValue)
   n31 = Get bIsInFrontEnd()
   Set FrontEnd_Tail Alpha(
    execute=[@n1.then_6, @n38.then_1]
    FrontEnd_Tail Alpha:float="To Float (Boolean)"(InBool=AND Boolean!(A=@n28, B=@n31, C:bool=@n30) -> (ReturnValue))
   ) -> (Output_Get:float)
   Branch(execute=@n1.then_7, Condition=Get bIsInFrontEnd())
   > Set Frontend_Backpack_Pose(Frontend_Backpack_Pose:float=GetCurveValue(CurveName=Frontend_Backpack_Pose)) -> (Output_Get:float)
   n32 = Branch(execute=@n1.then_8, Condition=Get bIsPlayingEmote())
   n33 = GetAnimInstance(self=GetOwnerSkeletalMeshComponent())
   > Is Valid(InputObject=@n33)
   n34 = OR Boolean(
    A=Montage_IsPlaying(self=@n33, Montage=/BRCosmetics/Animation/Game/MainPlayer/Emotes/Scrub_A_Dub/Emote_ScrubaDub_CLM_M)
    B=Montage_IsPlaying(self=@n33, Montage=/BRCosmetics/Animation/Game/MainPlayer/Emotes/Cauldron/Emote_Cauldron_CLM_M)
   )
   > Set in vehicle scale alpha(in vehicle scale alpha:double=SelectFloat(A=1, B=0, bPickA=@n34)) -> (Output_Get:double)
   > Set In Vehicle Sim Alpha(In Vehicle Sim Alpha:double=SelectFloat(A=0, B=1, bPickA=@n34)) -> (Output_Get:double)
   Set in vehicle scale alpha(execute=@n32.else, in vehicle scale alpha=0)
   > Set In Vehicle Sim Alpha(In Vehicle Sim Alpha=1)
   n35 = Branch(execute=@n1.then_5, Condition=Get bIsInFrontEnd())
   Set Is In First Person(
    execute=@n35.else
    Is In First Person:bool=Get bIsInFirstPerson!(self:Fort Player Anim Instance Object Reference=GetOwnerFortPlayerAnimInstance!(self:Custom Character Part Anim Instance Object Reference) -> (ReturnValue:Fort Player Anim Instance Object Reference)) -> (bIsInFirstPerson:bool)
   ) -> (Output_Get:bool)
   Event OnBeginPlayingEmote() {is=title K2Node_Event}
   > Set Dyn_Alpha_Spike_Emote(Dyn_Alpha_Spike_Emote:bool=true) -> (Output_Get:bool)
   Event OnEndPlayingEmote() {is=title K2Node_Event}
   > Set Dyn_Alpha_Spike_Emote(Dyn_Alpha_Spike_Emote=false)
   Event OnCharacterCustomizationLoaded()
   n36 = CheckForBackbling(LongBackblingThreshold=20)
   > n37 = Branch(Condition=OR Boolean(A=@n36, B=@n36.bIsBackblingLong))
   > n38 = Sequence!(execute) -> (then_0, then_1)
   > Set tail_cape_begin(tail_cape_begin:float=1) -> (Output_Get:float)
   > Delay(Duration=0.1)
   > Set tail_cape_begin(tail_cape_begin=0)
   Set tail_cape_begin(execute=@n37.else, tail_cape_begin=0)
   n39 = GetCoreAnimation!(self:Fort Weapon Object Reference=@n26) -> (ReturnValue:EFortWeaponCoreAnimation Enum)
   AND Boolean!(
    A="Equal (Enum)"(A:EFortWeaponCoreAnimation Enum=@n39, B:EFortWeaponCoreAnimation Enum=Melee)
    B=NOT Boolean(A="Equal (Enum)"(A=@n39, B=Unarmed))
   ) -> (ReturnValue)
  ]}
]
variables [
 animDynAlpha {type=number}
 Dyn Alfa Is Fly {type=number}
 Dyn Alpha Spike Butter Plate {type=number}
 Dyn_Alpha_Spike {type=number, default=1}
 Dyn_Alpha_Spike_Emote {type=boolean}
 Dyn_is_crouch {type=number}
 Dyn_is_parachute {type=number}
 Dyn_tai1_Rotation_Sky_Diving {type=Rotator, default=<rot 0 0 -30>}
 Dyn_tai2_Rotation_Sky_Diving {type=Rotator, default=<rot 0 0 -30>}
 Dyn_tai3_Rotation_Sky_Diving {type=Rotator, default=<rot 0 0 -30>}
 Dyn_tai4_Rotation_Sky_Diving {type=Rotator, default=<rot 0 0 -30>}
 FE_Alpha {type=number}
 FE_Idle_Alpha {type=number}
 Frontend Pickaxe Alpha {type=number}
 Frontend_Backpack_Pose {type=number}
 FrontEnd_Tail Alpha {type=number}
 in vehicle scale alpha {type=number}
 In Vehicle Sim Alpha {type=number}
 Is Baseball Sliding float {type=number}
 Is In First Person {type=boolean}
 Is Knee Sliding Float {type=number}
 Is Surface Swimming {type=number}
 Is Swim Sprinting {type=number}
 Is Tactical Sprinting {type=number}
 IsHangGlider {type=boolean}
 isSkyDiveDivingFloat {type=number}
 isWearingBackbling {type=boolean}
 isWearingCape {type=boolean}
 NewVar {type=boolean}
 pawnVelocitySpeed {type=number}
 pawnVelocitySpeedCutoff {type=number}
 tail_cape_begin {type=number, default=1}
]
