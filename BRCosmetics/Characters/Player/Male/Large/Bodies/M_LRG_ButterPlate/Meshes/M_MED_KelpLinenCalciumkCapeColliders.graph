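# Collision-only physics asset for a player character body, named as the cape colliders for KelpLinenCalciumk in the large male ButterPlate folder.
# Capsules cover the head, clavicles, pelvis, spine and feet, with paired capsules on several middle spine bones, plus a four-bone tail chain.
# Arm, thigh, calf and neck bodies exist but hold no shapes.
# No constraints are defined, so the bodies form no jointed ragdoll.
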PhysicsAsset /BRCosmetics/Characters/Player/Male/Large/Bodies/M_LRG_ButterPlate/Meshes/M_MED_KelpLinenCalciumkCapeColliders
bodies {
 calf_l []
 calf_r []
 clavicle_l [{shapeType=Capsule, length=6.5623, radius=7.5194, center=<3.9566 -1.0862 -2.1127>, rotation=<rot -77.5566 177.3915 7.7314>}]
 clavicle_r [{shapeType=Capsule, length=6.5623, radius=7.5194, center=<-3.9566 1.0862 2.1127>, rotation=<rot -77.5566 177.3915 -172.2686>}]
 foot_l [{shapeType=Capsule, length=11.0765, radius=8.0344, center=<-4.1811 6.0027 -0.338>, rotation=<rot 0.0388 1.4701 -89.3982>}]
 foot_r [{shapeType=Capsule, length=11.0765, radius=8.0344, center=<4.1811 -6.0027 0.338>, rotation=<rot 0.0388 1.4701 90.6018>}]
 head [{shapeType=Capsule, length=0, radius=18, center=<9.0697 0.0746 0>, rotation=<rot 0 -90 0>}]
 lowerarm_l []
 lowerarm_r []
 neck_02 []
 pelvis [{shapeType=Capsule, length=13.3542, radius=16.6441, center=<-6.7817 -11.0402 -0.1981>, rotation=<rot 0 -0.0001 90>}]
 spine_01 [{shapeType=Capsule, length=7.3888, radius=14.1138, center=<-8.9021 -0.8973 0>}]
 spine_02 [
  {shapeType=Capsule, length=12.2719, radius=15.2317, center=<-2.6885 -12.7463 0>, rotation=<rot 0 -0.0001 90>}
  {shapeType=Capsule, length=20.6875, radius=9.3125, center=<-1.8317 1.9362 0>, rotation=<rot 90 0 0>}
 ]
 spine_03 [
  {shapeType=Capsule, length=3.1863, radius=17.4854, center=<0.0573 -12.46 0>, rotation=<rot 0 -0.0002 90>}
  {shapeType=Capsule, length=20.7188, radius=9.2812, center=<-1.2163 0.549 0>, rotation=<rot 90 180 -175.0732>}
 ]
 spine_04 [
  {shapeType=Capsule, length=10.0076, radius=12.1416, center=<5.6234 -0.6779 0>}
  {shapeType=Capsule, length=10.0312, radius=9.2812, center=<-1.2326 1.308 0>, rotation=<rot 90 0 0>}
 ]
 spine_05 [{shapeType=Capsule, length=5.678, radius=14.6071, center=<0.6968 -11.267 0>, rotation=<rot 0 -0.0001 -90>}]
 tail1 [{
   shapeType=Capsule
   length=23.2021
   radius=17.6421
   shapeName=tail1_capsule
   center=<15.918 0.6644 7.4842>
   rotation=<rot 5.6478 -89.9999 96.2495>
  }]
 tail2 [{
   shapeType=Capsule
   length=30.8949
   radius=16.0761
   shapeName=tail2_capsule
   center=<15.6997 -0.3381 3.3424>
   rotation=<rot 0.0009 -91.2861 -80.3114>
  }]
 tail3 [{
   shapeType=Capsule
   length=10.8793
   radius=13.794
   shapeName=tail3_capsule
   center=<14.2558 -0.0169 1.9842>
   rotation=<rot -0.5683 -90 86.4196>
  }]
 tail4 [{
   shapeType=Capsule
   length=18.3187
   radius=10.1908
   shapeName=tail4_capsule
   center=<11.3595 -0.1127 3.0676>
   rotation=<rot 0 90.2369 -85.2009>
  }]
 thigh_l []
 thigh_r []
 upperarm_l []
 upperarm_r []
}
constraints []
